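# Physics asset for the female medium MissMissile body, set up for secondary motion of the blouse's hanging panels.
# Capsules on pelvis, spine, thighs, calves and feet form the body volume, with no head or arms.
# Center, back, front and side blouse chains hang from tiny base bodies constrained to the pelvis.
# Chain joints lock twist and limit swing; both second front segments add an angled capsule for wider coverage.
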PhysicsAsset /BRCosmetics/Characters/Player/Female/Medium/Bodies/F_MED_MissMissile/Meshes/F_MED_MissMissile_Physics
bodies {
 base_blouse_01_c [{shapeType=Capsule, length=0.505, radius=0.505, shapeName=base_blouse_01_c_capsule, rotation=<rot 0 -90 0>}]
 base_blouse_bk_01_l [{shapeType=Capsule, length=0.505, radius=0.505, shapeName=base_blouse_bk_01_l_capsule, rotation=<rot 0 -90 0>}]
 base_blouse_bk_01_r [{shapeType=Capsule, length=0.505, radius=0.505, shapeName=base_blouse_bk_01_r_capsule, rotation=<rot 0 -90 0>}]
 base_blouse_fr_main_01_l [{shapeType=Capsule, length=0.505, radius=0.505, shapeName=base_blouse_fr_main_01_l_capsule, rotation=<rot 0 -90 0>}]
 base_blouse_fr_main_01_r [{shapeType=Capsule, length=0.505, radius=0.505, shapeName=base_blouse_fr_main_01_r_capsule, rotation=<rot 0 -90 0>}]
 base_blouse_side_01_l [{shapeType=Capsule, length=0.505, radius=0.505, shapeName=base_blouse_side_01_l_capsule, rotation=<rot 0 -90 0>}]
 base_blouse_side_01_r [{shapeType=Capsule, length=0.505, radius=0.505, shapeName=base_blouse_side_01_r_capsule, rotation=<rot 0 -90 0>}]
 calf_l [{shapeType=Capsule, length=30, radius=9.5, shapeName=calf_l_capsule, center=<-21.6701 0 0>, rotation=<rot 90 0 0>}]
 calf_r [{shapeType=Capsule, length=30, radius=9.5, shapeName=calf_r_capsule, center=<21.6701 0 0>, rotation=<rot 90 180 0>}]
 dyn_blouse_01_c [{shapeType=Capsule, length=10, radius=6, shapeName=dyn_blouse_01_c_capsule, center=<5 0 4>, rotation=<rot 0 0 90>}]
 dyn_blouse_02_c [{shapeType=Capsule, length=15, radius=7, shapeName=dyn_blouse_02_c_capsule, center=<5 0 3>, rotation=<rot 0 0 90>}]
 dyn_blouse_03_c [{shapeType=Capsule, length=15, radius=7, shapeName=dyn_blouse_03_c_capsule, center=<6 0 4>, rotation=<rot 0 0 90>}]
 dyn_blouse_bk_01_l [{shapeType=Capsule, length=10, radius=6, shapeName=dyn_blouse_bk_01_l_capsule, center=<5 0 4>, rotation=<rot 0 0 90>}]
 dyn_blouse_bk_01_r [{shapeType=Capsule, length=10, radius=6, shapeName=dyn_blouse_bk_01_r_capsule, center=<5 0 4>, rotation=<rot 0 0 90>}]
 dyn_blouse_bk_02_l [{shapeType=Capsule, length=15, radius=7, shapeName=dyn_blouse_bk_02_l_capsule, center=<5 0 4>, rotation=<rot 0 0 90>}]
 dyn_blouse_bk_02_r [{shapeType=Capsule, length=15, radius=7, shapeName=dyn_blouse_bk_02_r_capsule, center=<5 0 4>, rotation=<rot 0 0 90>}]
 dyn_blouse_bk_03_l [{shapeType=Capsule, length=15, radius=7, shapeName=dyn_blouse_bk_03_l_capsule, center=<6 0 4>, rotation=<rot 0 0 90>}]
 dyn_blouse_bk_03_r [{shapeType=Capsule, length=15, radius=7, shapeName=dyn_blouse_bk_03_r_capsule, center=<6 0 4>, rotation=<rot 0 0 90>}]
 dyn_blouse_fr_01_l [{shapeType=Capsule, length=10, radius=6, shapeName=dyn_blouse_fr_01_l_capsule, center=<5 0 5>, rotation=<rot 0 0 90>}]
 dyn_blouse_fr_01_r [{shapeType=Capsule, length=10, radius=6, shapeName=dyn_blouse_fr_01_r_capsule, center=<5 -2.1267 5>, rotation=<rot 0 0 90>}]
 dyn_blouse_fr_02_l [
  {shapeType=Capsule, length=15, radius=7, shapeName=dyn_blouse_fr_02_l_capsule, center=<5 0 4>, rotation=<rot 0 0 90>}
  {
   shapeType=Capsule
   length=24.0085
   radius=5
   shapeName=dyn_blouse_fr_02_l_capsule
   center=<12.5639 7.0542 0.484>
   rotation=<rot -51.6634 -90.2488 92.7794>
  }
 ]
 dyn_blouse_fr_02_r [
  {shapeType=Capsule, length=15, radius=7, shapeName=dyn_blouse_fr_02_r_capsule, center=<5 -2.1267 4.5>, rotation=<rot 0 0 90>}
  {
   shapeType=Capsule
   length=24.0085
   radius=5
   shapeName=dyn_blouse_fr_02_r_capsule
   center=<12.5734 -7.0125 3>
   rotation=<rot -51.403 -95.2825 90.0232>
  }
 ]
 dyn_blouse_fr_03_l [{shapeType=Capsule, length=15, radius=7, shapeName=dyn_blouse_fr_03_l_capsule, center=<6 0 4>, rotation=<rot 0 0 90>}]
 dyn_blouse_fr_03_r [{shapeType=Capsule, length=15, radius=7, shapeName=dyn_blouse_fr_03_r_capsule, center=<6 -2.1267 4.5>, rotation=<rot 0 0 90>}]
 dyn_blouse_fr_main_01_l [{shapeType=Capsule, length=10, radius=5, shapeName=dyn_blouse_fr_main_01_l_capsule, center=<5 0 5>, rotation=<rot 0 0 90>}]
 dyn_blouse_fr_main_01_r [{shapeType=Capsule, length=10, radius=5, shapeName=dyn_blouse_fr_main_01_r_capsule, center=<5 -2.1267 5>, rotation=<rot 0 0 90>}]
 dyn_blouse_side_01_l [{shapeType=Capsule, length=10, radius=6, shapeName=dyn_blouse_side_01_l_capsule, center=<5 0 5>, rotation=<rot 0 0 90>}]
 dyn_blouse_side_01_r [{shapeType=Capsule, length=10, radius=6, shapeName=dyn_blouse_side_01_r_capsule, center=<5 0 5>, rotation=<rot 0 0 90>}]
 dyn_blouse_side_02_l [{shapeType=Capsule, length=15, radius=7, shapeName=dyn_blouse_side_02_l_capsule, center=<5 0 5>, rotation=<rot 0 0 90>}]
 dyn_blouse_side_02_r [{shapeType=Capsule, length=15, radius=7, shapeName=dyn_blouse_side_02_r_capsule, center=<5 0 5>, rotation=<rot 0 0 90>}]
 dyn_blouse_side_03_l [{shapeType=Capsule, length=15, radius=7, shapeName=dyn_blouse_side_03_l_capsule, center=<6 0 4>, rotation=<rot 0 0 90>}]
 dyn_blouse_side_03_r [{shapeType=Capsule, length=15, radius=7, shapeName=dyn_blouse_side_03_r_capsule, center=<6 0 4>, rotation=<rot 0 0 90>}]
 foot_l [{shapeType=Capsule, length=15, radius=8, shapeName=foot_l_capsule, center=<-3.1536 6.8653 0>, rotation=<rot 90 -65.3279 0>}]
 foot_r [{shapeType=Capsule, length=15, radius=8, shapeName=foot_r_capsule, center=<3.1536 -6.8653 0>, rotation=<rot 90 114.6721 0>}]
 pelvis [{shapeType=Capsule, length=10.5951, radius=11.1647, shapeName=pelvis_capsule, center=<-1.4429 -0.4526 0>}]
 spine_01 [{shapeType=Capsule, length=8.7957, radius=8.995, shapeName=spine_01_capsule, center=<0.7103 -0.023 0>}]
 spine_02 [{shapeType=Capsule, length=9.9996, radius=7.9634, shapeName=spine_02_capsule, center=<3.2663 0 0>}]
 spine_04 [{shapeType=Capsule, length=12.9065, radius=9.0759, center=<-0.2 2.1557 0.0057>, rotation=<rot -0.0464 -90 -0.0288>}]
 thigh_l [{
   shapeType=Capsule
   length=29.2565
   radius=9
   shapeName=thigh_l_capsule
   center=<-19.929 -1.0447 0.2963>
   rotation=<rot 0.1824 87.2778 -88.9656>
  }]
 thigh_r [{
   shapeType=Capsule
   length=29.2565
   radius=9
   shapeName=thigh_r_capsule
   center=<19.929 1.0447 -0.2963>
   rotation=<rot 0.1824 87.2778 91.0344>
  }]
}
constraints [
 {bone1Name=spine_02, bone2Name=pelvis, swing1LimitDegrees=45, swing2LimitDegrees=45, twistLimitDegrees=45}
 {bone1Name=spine_04, bone2Name=spine_02, swing1LimitDegrees=45, swing2LimitDegrees=45, twistLimitDegrees=45}
 {
  bone1Name=thigh_l
  bone2Name=pelvis
  swing1LimitDegrees=45
  swing1Motion=Free
  swing2LimitDegrees=45
  swing2Motion=Free
  twistLimitDegrees=45
  twistMotion=Free
 }
 {
  bone1Name=thigh_r
  bone2Name=pelvis
  swing1LimitDegrees=45
  swing1Motion=Free
  swing2LimitDegrees=45
  swing2Motion=Free
  twistLimitDegrees=45
  twistMotion=Free
 }
 {
  bone1Name=spine_01
  bone2Name=pelvis
  swing1LimitDegrees=45
  swing1Motion=Free
  swing2LimitDegrees=45
  swing2Motion=Free
  twistLimitDegrees=45
  twistMotion=Free
 }
 {
  bone1Name=spine_02
  bone2Name=spine_01
  swing1LimitDegrees=45
  swing1Motion=Free
  swing2LimitDegrees=45
  swing2Motion=Free
  twistLimitDegrees=45
  twistMotion=Free
 }
 {
  bone1Name=base_blouse_01_c
  bone2Name=pelvis
  swing1LimitDegrees=45
  swing1Motion=Free
  swing2LimitDegrees=45
  swing2Motion=Free
  twistLimitDegrees=45
  twistMotion=Free
 }
 {bone1Name=dyn_blouse_01_c, bone2Name=base_blouse_01_c, swing1LimitDegrees=8, swing2LimitDegrees=30, twistLimitDegrees=1, twistMotion=Locked}
 {bone1Name=dyn_blouse_02_c, bone2Name=dyn_blouse_01_c, swing1LimitDegrees=8, swing2LimitDegrees=20, twistLimitDegrees=1, twistMotion=Locked}
 {bone1Name=dyn_blouse_03_c, bone2Name=dyn_blouse_02_c, swing1LimitDegrees=15, swing2LimitDegrees=15, twistLimitDegrees=1, twistMotion=Locked}
 {
  bone1Name=base_blouse_bk_01_l
  bone2Name=pelvis
  swing1LimitDegrees=45
  swing1Motion=Free
  swing2LimitDegrees=45
  swing2Motion=Free
  twistLimitDegrees=45
  twistMotion=Free
 }
 {
  bone1Name=dyn_blouse_bk_01_l
  bone2Name=base_blouse_bk_01_l
  swing1LimitDegrees=8
  swing2LimitDegrees=30
  twistLimitDegrees=1
  twistMotion=Locked
 }
 {
  bone1Name=dyn_blouse_bk_02_l
  bone2Name=dyn_blouse_bk_01_l
  swing1LimitDegrees=8
  swing2LimitDegrees=20
  twistLimitDegrees=1
  twistMotion=Locked
 }
 {
  bone1Name=dyn_blouse_bk_03_l
  bone2Name=dyn_blouse_bk_02_l
  swing1LimitDegrees=15
  swing2LimitDegrees=15
  twistLimitDegrees=1
  twistMotion=Locked
 }
 {
  bone1Name=base_blouse_bk_01_r
  bone2Name=pelvis
  swing1LimitDegrees=45
  swing1Motion=Free
  swing2LimitDegrees=45
  swing2Motion=Free
  twistLimitDegrees=45
  twistMotion=Free
 }
 {
  bone1Name=dyn_blouse_bk_01_r
  bone2Name=base_blouse_bk_01_r
  swing1LimitDegrees=8
  swing2LimitDegrees=30
  twistLimitDegrees=1
  twistMotion=Locked
 }
 {
  bone1Name=dyn_blouse_bk_02_r
  bone2Name=dyn_blouse_bk_01_r
  swing1LimitDegrees=8
  swing2LimitDegrees=20
  twistLimitDegrees=1
  twistMotion=Locked
 }
 {
  bone1Name=dyn_blouse_bk_03_r
  bone2Name=dyn_blouse_bk_02_r
  swing1LimitDegrees=15
  swing2LimitDegrees=15
  twistLimitDegrees=1
  twistMotion=Locked
 }
 {
  bone1Name=base_blouse_fr_main_01_l
  bone2Name=pelvis
  swing1LimitDegrees=45
  swing1Motion=Free
  swing2LimitDegrees=45
  swing2Motion=Free
  twistLimitDegrees=45
  twistMotion=Free
 }
 {
  bone1Name=dyn_blouse_fr_main_01_l
  bone2Name=base_blouse_fr_main_01_l
  swing1LimitDegrees=10
  swing2LimitDegrees=50
  twistLimitDegrees=1
  twistMotion=Locked
 }
 {
  bone1Name=dyn_blouse_fr_01_l
  bone2Name=dyn_blouse_fr_main_01_l
  swing1LimitDegrees=8
  swing2LimitDegrees=30
  twistLimitDegrees=1
  twistMotion=Locked
 }
 {
  bone1Name=dyn_blouse_fr_02_l
  bone2Name=dyn_blouse_fr_01_l
  swing1LimitDegrees=8
  swing2LimitDegrees=15
  twistLimitDegrees=1
  twistMotion=Locked
 }
 {
  bone1Name=dyn_blouse_fr_03_l
  bone2Name=dyn_blouse_fr_02_l
  swing1LimitDegrees=15
  swing2LimitDegrees=15
  twistLimitDegrees=1
  twistMotion=Locked
 }
 {
  bone1Name=base_blouse_fr_main_01_r
  bone2Name=pelvis
  swing1LimitDegrees=45
  swing1Motion=Free
  swing2LimitDegrees=45
  swing2Motion=Free
  twistLimitDegrees=45
  twistMotion=Free
 }
 {
  bone1Name=dyn_blouse_fr_main_01_r
  bone2Name=base_blouse_fr_main_01_r
  swing1LimitDegrees=10
  swing2LimitDegrees=50
  twistLimitDegrees=1
  twistMotion=Locked
 }
 {
  bone1Name=dyn_blouse_fr_01_r
  bone2Name=dyn_blouse_fr_main_01_r
  swing1LimitDegrees=8
  swing2LimitDegrees=30
  twistLimitDegrees=1
  twistMotion=Locked
 }
 {
  bone1Name=dyn_blouse_fr_02_r
  bone2Name=dyn_blouse_fr_01_r
  swing1LimitDegrees=8
  swing2LimitDegrees=10
  twistLimitDegrees=1
  twistMotion=Locked
 }
 {
  bone1Name=dyn_blouse_fr_03_r
  bone2Name=dyn_blouse_fr_02_r
  swing1LimitDegrees=15
  swing2LimitDegrees=15
  twistLimitDegrees=1
  twistMotion=Locked
 }
 {
  bone1Name=base_blouse_side_01_l
  bone2Name=pelvis
  swing1LimitDegrees=45
  swing1Motion=Free
  swing2LimitDegrees=45
  swing2Motion=Free
  twistLimitDegrees=45
  twistMotion=Free
 }
 {
  bone1Name=dyn_blouse_side_01_l
  bone2Name=base_blouse_side_01_l
  swing1LimitDegrees=8
  swing2LimitDegrees=30
  twistLimitDegrees=1
  twistMotion=Locked
 }
 {
  bone1Name=dyn_blouse_side_02_l
  bone2Name=dyn_blouse_side_01_l
  swing1LimitDegrees=8
  swing2LimitDegrees=20
  twistLimitDegrees=1
  twistMotion=Locked
 }
 {
  bone1Name=dyn_blouse_side_03_l
  bone2Name=dyn_blouse_side_02_l
  swing1LimitDegrees=15
  swing2LimitDegrees=15
  twistLimitDegrees=1
  twistMotion=Locked
 }
 {
  bone1Name=base_blouse_side_01_r
  bone2Name=pelvis
  swing1LimitDegrees=45
  swing1Motion=Free
  swing2LimitDegrees=45
  swing2Motion=Free
  twistLimitDegrees=45
  twistMotion=Free
 }
 {
  bone1Name=dyn_blouse_side_01_r
  bone2Name=base_blouse_side_01_r
  swing1LimitDegrees=8
  swing2LimitDegrees=30
  twistLimitDegrees=1
  twistMotion=Locked
 }
 {
  bone1Name=dyn_blouse_side_02_r
  bone2Name=dyn_blouse_side_01_r
  swing1LimitDegrees=8
  swing2LimitDegrees=20
  twistLimitDegrees=1
  twistMotion=Locked
 }
 {
  bone1Name=dyn_blouse_side_03_r
  bone2Name=dyn_blouse_side_02_r
  swing1LimitDegrees=15
  swing2LimitDegrees=15
  twistLimitDegrees=1
  twistMotion=Locked
 }
 {
  bone1Name=calf_l
  bone2Name=thigh_l
  swing1LimitDegrees=45
  swing1Motion=Free
  swing2LimitDegrees=45
  swing2Motion=Free
  twistLimitDegrees=45
  twistMotion=Free
 }
 {
  bone1Name=foot_l
  bone2Name=calf_l
  swing1LimitDegrees=45
  swing1Motion=Free
  swing2LimitDegrees=45
  swing2Motion=Free
  twistLimitDegrees=45
  twistMotion=Free
 }
 {
  bone1Name=foot_r
  bone2Name=calf_r
  swing1LimitDegrees=45
  swing1Motion=Free
  swing2LimitDegrees=45
  swing2Motion=Free
  twistLimitDegrees=45
  twistMotion=Free
 }
 {
  bone1Name=calf_r
  bone2Name=thigh_r
  swing1LimitDegrees=45
  swing1Motion=Free
  swing2LimitDegrees=45
  swing2Motion=Free
  twistLimitDegrees=45
  twistMotion=Free
 }
]
