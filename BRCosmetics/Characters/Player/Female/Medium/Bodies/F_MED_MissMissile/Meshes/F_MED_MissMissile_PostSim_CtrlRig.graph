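# Post-sim Control Rig for the F_MED_MissMissile body that applies corrective offsets to the dynamic blouse and holster bones during specific movement states.
# A sequence runs weighted ModifyTransforms for kneeslide, baseball slide and stationary crouch blouse poses, then crouch-walk and pickaxe-swing holster offsets.
# AlphaInterp smooths the state weights, except CrouchWalk, which is multiplied by a curve driven by remapped pelvis height.
# Three further ModifyTransforms nodes are disconnected from execution.
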ControlRigBlueprint /BRCosmetics/Characters/Player/Female/Medium/Bodies/F_MED_MissMissile/Meshes/F_MED_MissMissile_PostSim_CtrlRig
graphs [
 RigUnit_SequenceAggregate_1_ContainedGraph {nodes [
   n1 = RigVMFunctionEntryNode()
   RigUnit_SequenceAggregate = &RigVMFunction_Sequence(ExecuteContext=@n1.ExecuteContext)
   RigUnit_SequenceAggregate_1 = &RigVMFunction_Sequence(ExecuteContext=@RigUnit_SequenceAggregate.B)
   RigUnit_SequenceAggregate_2 = &RigVMFunction_Sequence(ExecuteContext=@RigUnit_SequenceAggregate_1.B)
   RigVMFunctionReturnNode(A=@RigUnit_SequenceAggregate.A, B=@RigUnit_SequenceAggregate_1.A, C=@RigUnit_SequenceAggregate_2.A, D=@RigUnit_SequenceAggregate_2.B)
  ]}
 RigVMFunctionLibrary {links [], nodes []}
 RigVMModel {nodes [
   RigUnit_BeginExecution = &RigUnit_BeginExecution()
   RigUnit_SequenceAggregate = RigVMAggregateNode(ExecuteContext=@RigUnit_BeginExecution.ExecutePin)
   n1 = &RigVMFunction_AlphaInterp(InterpSpeedDecreasing=10, InterpSpeedIncreasing=3, Scale=1, bClampResult=False, Value=RigVMVariableNode() {variable=Kneeslide}.Value)
   &RigUnit_ModifyTransforms(
    ItemToModify="((Item=(Type=Bone,Name=\"dyn_blouse_bk_01_r\"),Transform=(Rotation=(X=0.2965,Y=-0.3074,Z=0.0448,W=0.9031),Translation=(X=0,Y=0,Z=0),Scale3D=(X=1,Y=1,Z=1))),(Item=(Type=Bone,Name=\"dyn_blouse_01_c\"),Transform=(Rotation=(X=-0.0008,Y=0.6125,Z=-0.1022,W=-0.7838), ...(truncated)"
    WeightMinimum=0
    ExecutePin=@RigUnit_SequenceAggregate.A
    Weight=@n1.Result
   )
   > &RigUnit_ModifyTransforms(
    ItemToModify="((Item=(Type=Bone,Name=\"dyn_blouse_fr_main_01_r\"),Transform=(Rotation=(X=-0.2595,Y=-0.1189,Z=-0.0752,W=-0.9554),Translation=(X=9.8008,Y=3.5345,Z=0.8487),Scale3D=(X=1,Y=1,Z=1))),(Item=(Type=Bone,Name=\"dyn_blouse_side_01_r\"),Transform=(Rotation=(X=-0.1331,Y= ...(truncated)"
    WeightMinimum=0
    Weight=@n1.Result
   )
   n2 = &RigVMFunction_AlphaInterp(InterpSpeedDecreasing=20, InterpSpeedIncreasing=2, Scale=1, bClampResult=False, Value=RigVMVariableNode() {variable=BaseballSlide}.Value)
   &RigUnit_ModifyTransforms(
    ItemToModify="((Item=(Type=Bone,Name=\"dyn_blouse_fr_main_01_r\"),Transform=(Rotation=(X=0.6006,Y=-0.2395,Z=0.1362,W=0.7505),Translation=(X=0,Y=0,Z=0),Scale3D=(X=1,Y=1,Z=1))),(Item=(Type=Bone,Name=\"dyn_blouse_side_01_r\"),Transform=(Rotation=(X=-0.3107,Y=0.6182,Z=0.26,W=-0 ...(truncated)"
    WeightMinimum=0
    ExecutePin=@RigUnit_SequenceAggregate.B
    Weight=@n2.Result
   )
   > &RigUnit_ModifyTransforms(
    ItemToModify="((Item=(Type=Bone,Name=\"dyn_blouse_fr_main_01_l\"),Transform=(Rotation=(X=0.1067,Y=0.0584,Z=-0.1357,W=0.9832),Translation=(X=10.6753,Y=-0.7478,Z=-10.8864),Scale3D=(X=1,Y=1,Z=1))),(Item=(Type=Bone,Name=\"dyn_blouse_side_01_l\"),Transform=(Rotation=(X=-0.0112,Y ...(truncated)"
    WeightMinimum=0
    Weight=@n2.Result
   )
   > &RigUnit_ModifyTransforms(
    ItemToModify="((Item=(Type=Bone,Name=\"dyn_blouse_bk_01_r\"),Transform=(Rotation=(X=-0.0876,Y=0.7319,Z=0.0326,W=-0.675),Translation=(X=0,Y=0,Z=0),Scale3D=(X=1,Y=1,Z=1))),(Item=(Type=Bone,Name=\"dyn_blouse_01_c\"),Transform=(Rotation=(X=0.0766,Y=0.6402,Z=-0.0668,W=-0.7614),T ...(truncated)"
    WeightMinimum=0
    Weight=@n2.Result
   )
   AlphaInterp = &RigVMFunction_AlphaInterp(InterpSpeedDecreasing=2, InterpSpeedIncreasing=10, Scale=1, bClampResult=False, Value=RigVMVariableNode() {variable=CrouchingNotMove}.Value)
   &RigUnit_ModifyTransforms(
    ItemToModify="((Item=(Type=Bone,Name=\"dyn_blouse_side_02_l\"),Transform=(Rotation=(X=0.0038,Y=0.0217,Z=-0.0587,W=0.998),Translation=(X=-3.4078,Y=-0.9022,Z=0.094),Scale3D=(X=1,Y=1,Z=1))),(Item=(Type=Bone,Name=\"dyn_blouse_side_03_r\"),Transform=(Rotation=(X=0,Y=0,Z=0.0159,W ...(truncated)"
    WeightMinimum=0
    ExecutePin=@RigUnit_SequenceAggregate.C
    Weight=@AlphaInterp.Result
   )
   > &RigUnit_ModifyTransforms(
    ItemToModify="((Item=(Type=Bone,Name=\"dyn_blouse_fr_01_l\"),Transform=(Rotation=(X=-0.1214,Y=-0.2655,Z=0.1491,W=-0.9447),Translation=(X=-2.1775,Y=1.4432,Z=4.4704),Scale3D=(X=1,Y=1,Z=1))),(Item=(Type=Bone,Name=\"dyn_blouse_fr_02_l\"),Transform=(Rotation=(X=0.0031,Y=-0.126,Z ...(truncated)"
    WeightMinimum=0
    Weight=@AlphaInterp.Result
   )
   n3 = &RigVMFunction_AnimEvalRichCurve(
    Curve=(EditorCurveData=(Keys=((),(Time=0.5,Value=1),(Time=0.8))))
    SourceMaximum=2
    SourceMinimum=0
    TargetMinimum=0
    Value=&RigVMFunction_MathDoubleRemap(
     SourceMaximum=75
     SourceMinimum=45
     TargetMaximum=1
     TargetMinimum=0
     bClamp=false
     Value=&RigUnit_GetRelativeTransformForItem(Child=(Type=Bone,Name="pelvis"), Parent=(Type=Bone,Name="root"), bChildInitial=False, bParentInitial=False).RelativeTransform.Translation.Z
    ).Result
   )
   Multiply = &RigVMFunction_MathDoubleMul(B=@n3.Result, A=RigVMVariableNode() {variable=CrouchWalk}.Value)
   > &RigUnit_ModifyTransforms(
    ItemToModify=((Item=(Type=Bone,Name="dyn_holster"),Transform=(Rotation=(X=-0.2268,Y=0.0305,Z=-0.1694,W=-0.9586),Translation=(X=0.3237,Y=0.7987,Z=0.3478),Scale3D=(X=1,Y=1,Z=1))))
    WeightMinimum=0
    Weight=@Multiply.Result
   )
   n4 = &RigVMFunction_AlphaInterp(InterpSpeedDecreasing=2, InterpSpeedIncreasing=10, Scale=1, bClampResult=False, Value=RigVMVariableNode() {variable=PickaxeSwing}.Value)
   > &RigUnit_ModifyTransforms(
    ItemToModify=((Item=(Type=Bone,Name="dyn_holster"),Transform=(Rotation=(X=0,Y=0,Z=0.0855,W=0.9963),Translation=(X=0,Y=0,Z=0),Scale3D=(X=1,Y=1,Z=1))))
    WeightMinimum=0
    Weight=@n4.Result
   )
   &RigUnit_ModifyTransforms(
    ExecutePin=()
    ItemToModify="((Item=(Type=Bone,Name=\"dyn_blouse_side_01_l\"),Transform=(Rotation=(X=0,Y=0,Z=0.3326,W=-0.9431),Translation=(X=0,Y=0,Z=0),Scale3D=(X=1,Y=1,Z=1))),(Item=(Type=Bone,Name=\"dyn_blouse_side_02_l\"),Transform=(Rotation=(X=0.1032,Y=0.0782,Z=-0.1337,W=-0.9825),Tran ...(truncated)"
    Weight=1
    WeightMinimum=0
   )
   &RigUnit_ModifyTransforms(
    ExecutePin=()
    ItemToModify=((Item=(Type=Bone,Name="dyn_holster"),Transform=(Rotation=(X=-0.1706,Y=0,Z=0,W=0.9853),Translation=(X=2.1208,Y=-2.2,Z=-0.2729),Scale3D=(X=1,Y=1,Z=1))))
    WeightMinimum=0
    Weight=@n1.Result
   )
   &RigUnit_ModifyTransforms(
    ExecutePin=()
    ItemToModify="((Item=(Type=Bone,Name=\"dyn_blouse_fr_01_l\"),Transform=(Rotation=(X=0.016,Y=-0.0809,Z=0.252,W=-0.9642),Translation=(X=-1.9627,Y=1.2504,Z=1.9827),Scale3D=(X=1,Y=1,Z=1))),(Item=(Type=Bone,Name=\"dyn_blouse_fr_02_l\"),Transform=(Rotation=(X=-0.0479,Y=-0.0111,Z= ...(truncated)"
    Weight=1
    WeightMaximum=0
    WeightMinimum=0
   )
  ]}
]
hierarchy [
 root {loc=<0 0 0>, rot=<0 0 0 1>}
 pelvis {parent=root, loc=<0 -0.5301 98.0571>, rot=<0.0127 0.707 -0.0127 -0.707>}
 spine_01 {parent=pelvis, loc=<6.1208 -0.9328 0>, rot=<0 0 0.034 -0.9994>}
 spine_02 {parent=spine_01, loc=<6.5997 -0.2095 0>, rot=<0 0 0.0238 -0.9997>}
 spine_03 {parent=spine_02, loc=<6.4914 -0.1237 0>, rot=<0 0 0.0149 -0.9999>}
 spine_04 {parent=spine_03, loc=<6.9598 0.0794 0>, rot=<0 0 -0.0132 -0.9999>}
 spine_05 {parent=spine_04, loc=<6.9987 0 0>, rot=<0 0 -0.0015 -1>}
 neck_01 {parent=spine_05, loc=<11.3138 0.1529 0>, rot=<0 0 -0.2003 -0.9797>}
 neck_02 {parent=neck_01, loc=<4.5627 0 0>, rot=<0 0 0.0504 -0.9987>}
 head {parent=neck_02, loc=<5.3914 0 0>, rot=<0 0 0.115 -0.9934>}
 dyn_hair_side_bk_r_base {parent=head, loc=<7.1717 -4.9596 4.6295>, rot=<0.1238 -0.9923 0 0>}
 dyn_hair_side_bk_01_r {parent=dyn_hair_side_bk_r_base, loc=<0.3451 0.064 0>, rot=<0 0 0 1>}
 dyn_hair_side_bk_02_r {parent=dyn_hair_side_bk_01_r, loc=<7.727 -0.3038 -0.6969>, rot=<-0.0588 -0.0147 0.0601 -0.9964>}
 dyn_hair_side_bk_03_r {parent=dyn_hair_side_bk_02_r, loc=<8.1626 0.2253 -0.0629>, rot=<0.0202 0.0129 0.0049 -0.9997>}
 dyn_hair_side_bk_04_r {parent=dyn_hair_side_bk_03_r, loc=<6.8218 0.0224 -0.0281>, rot=<-0.0057 0.0208 -0.085 -0.9961>}
 dyn_hair_side_bk_05_r {parent=dyn_hair_side_bk_04_r, loc=<6.4092 0.0823 -0.0318>, rot=<-0.013 0.0006 -0.0443 -0.9989>}
 dyn_hair_side_bk_06_r {parent=dyn_hair_side_bk_05_r, loc=<6.4569 0 0>, rot=<0 0 0 1>}
 dyn_hair_center_bk_base {parent=head, loc=<7.2617 -6.2621 0.3833>, rot=<0.172 -0.985 0.0084 0.014>}
 dyn_hair_center_bk_01 {parent=dyn_hair_center_bk_base, loc=<0.4251 0.1227 0.0111>, rot=<0 0 0 1>}
 dyn_hair_center_bk_02 {parent=dyn_hair_center_bk_01, loc=<7.9201 0.5099 0.2304>, rot=<0.0067 0.0149 -0.0053 -0.9999>}
 dyn_hair_center_bk_03 {parent=dyn_hair_center_bk_02, loc=<8.2378 0 0>, rot=<0 0 0.0023 -1>}
 dyn_hair_center_bk_04 {parent=dyn_hair_center_bk_03, loc=<8.2799 0 0>, rot=<0 0 -0.1277 -0.9918>}
 dyn_hair_center_bk_05 {parent=dyn_hair_center_bk_04, loc=<9.6093 0 0>, rot=<0 0 0 1>}
 dyn_hair_side_bk_l_base {parent=head, loc=<7.1714 -4.9596 -4.6295>, rot=<0.1238 -0.9923 0 0>}
 dyn_hair_side_bk_01_l {parent=dyn_hair_side_bk_l_base, loc=<0.3444 0.0639 0>, rot=<0 0 0 1>}
 dyn_hair_side_bk_02_l {parent=dyn_hair_side_bk_01_l, loc=<7.7274 -0.3037 0.6968>, rot=<0.0588 0.0147 0.0601 -0.9964>}
 dyn_hair_side_bk_03_l {parent=dyn_hair_side_bk_02_l, loc=<8.1625 0.2252 0.0629>, rot=<-0.0202 -0.0129 0.0049 -0.9997>}
 dyn_hair_side_bk_04_l {parent=dyn_hair_side_bk_03_l, loc=<6.8219 0.0225 0.0281>, rot=<0.0057 -0.0208 -0.085 -0.9961>}
 dyn_hair_side_bk_05_l {parent=dyn_hair_side_bk_04_l, loc=<6.4095 0.0823 0.0318>, rot=<0.013 -0.0006 -0.0443 -0.9989>}
 dyn_hair_side_bk_06_l {parent=dyn_hair_side_bk_05_l, loc=<6.0346 0 0>, rot=<0 0 0 1>}
 dyn_hair_side_r_base {parent=head, loc=<6.9525 -1.2567 6.9293>, rot=<0.0459 -0.1378 0.985 -0.0931>}
 dyn_hair_side_01_r {parent=dyn_hair_side_r_base, loc=<0.4808 -0.0217 0.0023>, rot=<0 0 0 1>}
 dyn_hair_side_02_r {parent=dyn_hair_side_01_r, loc=<5.2396 0 0>, rot=<-0.0117 0.003 -0.0521 -0.9986>}
 dyn_hair_side_03_r {parent=dyn_hair_side_02_r, loc=<4.7513 -0.1387 -0.0106>, rot=<0.0008 -0.0165 -0.0577 -0.9982>}
 dyn_hair_side_l_base {parent=head, loc=<6.9528 -1.2567 -6.9293>, rot=<0.0931 -0.985 0.1378 -0.0459>}
 dyn_hair_side_01_l {parent=dyn_hair_side_l_base, loc=<0.481 0.0217 0.0023>, rot=<0 0 0 1>}
 dyn_hair_side_02_l {parent=dyn_hair_side_01_l, loc=<5.2398 0 0>, rot=<0.0117 0.003 0.0521 -0.9986>}
 dyn_hair_side_03_l {parent=dyn_hair_side_02_l, loc=<4.7514 0.1387 -0.0106>, rot=<-0.0008 -0.0165 0.0577 -0.9982>}
 dyn_hair_side_fr_l_base {parent=head, loc=<6.6065 1.365 -7.4686>, rot=<-0.0195 -0.9982 -0.0542 -0.0146>}
 dyn_hair_side_fr_01_l {parent=dyn_hair_side_fr_l_base, loc=<0.481 0.0217 0.0023>, rot=<0 0 0 1>}
 dyn_hair_side_fr_02_l {parent=dyn_hair_side_fr_01_l, loc=<7 0 0>, rot=<0.0086 0.002 -0.0918 -0.9957>}
 dyn_hair_side_fr_03_l {parent=dyn_hair_side_fr_02_l, loc=<7 0 0>, rot=<-0.0029 -0.0162 -0.0714 -0.9973>}
 dyn_hair_side_fr_04_l {parent=dyn_hair_side_fr_03_l, loc=<7 0 0>, rot=<0 0 -0.0694 -0.9976>}
 dyn_hair_side_fr_05_l {parent=dyn_hair_side_fr_04_l, loc=<7 0 0>, rot=<0 0 -0.0694 -0.9976>}
 base_R_hair_main {parent=head, loc=<6.5671 -6.0791 5.6207>, rot=<0.1159 -0.7002 0.7012 -0.0682>}
 dyn_R_hair_main_1 {parent=base_R_hair_main, loc=<0 0 0>, rot=<0 0 0 1>}
 dyn_R_hair_main_2 {parent=dyn_R_hair_main_1, loc=<9.469 0 0>, rot=<0 0 0 1>}
 dyn_R_hair_main_3 {parent=dyn_R_hair_main_2, loc=<6.637 0 0>, rot=<0 0 0 1>}
 dyn_R_hair_main_4 {parent=dyn_R_hair_main_3, loc=<9.2744 0 0>, rot=<0 -0.1254 0 -0.9921>}
 dyn_R_hair_main_5 {parent=dyn_R_hair_main_4, loc=<10.0047 0 0>, rot=<0 0 0 1>}
 base_L_hair_main {parent=head, loc=<6.5554 -6.082 -6.08>, rot=<0.0921 -0.7011 0.7011 -0.0921>}
 dyn_L_hair_main_1 {parent=base_L_hair_main, loc=<0 0 0>, rot=<0 0 0 1>}
 dyn_L_hair_main_2 {parent=dyn_L_hair_main_1, loc=<9.4214 0 0>, rot=<0 0 0 1>}
 dyn_L_hair_main_3 {parent=dyn_L_hair_main_2, loc=<6.6371 0 0>, rot=<0 0 0 1>}
 dyn_L_hair_main_4 {parent=dyn_L_hair_main_3, loc=<9.2744 0 0>, rot=<0 -0.1254 0 -0.9921>}
 dyn_L_hair_main_5 {parent=dyn_L_hair_main_4, loc=<10.0047 0 0>, rot=<0 0 0 1>}
 clavicle_l {parent=spine_05, loc=<5.4831 -0.1053 -3.2946>, rot=<0.0465 -0.7289 -0.009 -0.6829>}
 upperarm_l {parent=clavicle_l, loc=<9.8658 -1.2136 -0.4105>, rot=<0.0164 -0.3245 -0.0269 -0.9454>}
 lowerarm_l {parent=upperarm_l, loc=<26.465 -0.8345 0.259>, rot=<0.0036 0.0161 -0.2298 -0.9731>}
 hand_l {parent=lowerarm_l, loc=<22.7355 -0.3203 -0.6105>, rot=<-0.6784 -0.0223 0.0735 -0.7307>}
 hand_attach_l {parent=hand_l, loc=<8.5 -3 0>, rot=<0.7071 0.7071 0 0>}
 weapon_l {parent=hand_l, loc=<-0.1974 -3.4111 0.3811>, rot=<0 0 0 1>}
 thumb_01_l {parent=hand_l, loc=<1.3118 -0.3191 -2.459>, rot=<0.6546 -0.4042 0.0664 -0.6354>}
 thumb_02_l {parent=thumb_01_l, loc=<5.9336 -0.0448 0.0401>, rot=<-0.0037 0.0213 0.298 -0.9543>}
 thumb_03_l {parent=thumb_02_l, loc=<2.2223 0.3682 -0.0907>, rot=<0.0118 -0.013 0.0066 -0.9998>}
 index_metacarpal_l {parent=hand_l, loc=<1.6535 0.0183 -2.96>, rot=<0.0158 -0.0124 -0.0173 -0.9996>}
 index_01_l {parent=index_metacarpal_l, loc=<8.3008 0.0357 -0.0524>, rot=<-0.0044 0.0485 0.0924 -0.9945>}
 index_02_l {parent=index_01_l, loc=<3.3587 -0.3004 0.0452>, rot=<-0.002 0.0027 0.3225 -0.9466>}
 index_03_l {parent=index_02_l, loc=<2.5578 -0.123 -0.0252>, rot=<-0.002 0.0067 0.2558 -0.9667>}
 middle_metacarpal_l {parent=hand_l, loc=<2.1856 -0.1725 -0.5984>, rot=<0.006 0.0168 -0.0329 -0.9993>}
 middle_01_l {parent=middle_metacarpal_l, loc=<7.3275 -0.2447 0.0129>, rot=<-0.0208 0.0042 0.0858 -0.9961>}
 middle_02_l {parent=middle_01_l, loc=<4.0123 0.1401 0.1175>, rot=<-0.0218 0.0258 0.3512 -0.9357>}
 middle_03_l {parent=middle_02_l, loc=<2.6077 0.0534 0.1373>, rot=<0.0082 0.0188 0.1708 -0.9851>}
 ring_metacarpal_l {parent=hand_l, loc=<2.2074 -0.3625 1.3638>, rot=<-0.0284 0.0722 -0.0458 -0.9959>}
 ring_01_l {parent=ring_metacarpal_l, loc=<7.0295 -0.6197 -0.1446>, rot=<-0.0489 -0.0252 0.1892 -0.9804>}
 ring_02_l {parent=ring_01_l, loc=<3.6174 0.4647 -0.1027>, rot=<0 0 0.2806 -0.9598>}
 ring_03_l {parent=ring_02_l, loc=<2.4152 0.114 0.0199>, rot=<0 0 0.1189 -0.9929>}
 pinky_metacarpal_l {parent=hand_l, loc=<1.7616 -0.9753 3.3814>, rot=<-0.0397 0.0814 -0.069 -0.9935>}
 pinky_01_l {parent=pinky_metacarpal_l, loc=<7.1709 -0.4831 -0.1094>, rot=<-0.0654 -0.0035 0.2242 -0.9723>}
 pinky_02_l {parent=pinky_01_l, loc=<2.8274 0.0893 -0.1043>, rot=<0.0037 0.0011 0.2854 -0.9584>}
 pinky_03_l {parent=pinky_02_l, loc=<2.2046 0.0307 -0.0078>, rot=<-0.0002 0.0098 0.1445 -0.9895>}
 deform_wrist_dn_base_l {parent=hand_l, loc=<0 -1.5326 0>, rot=<0 0 0 -1>}
 deform_wrist_dn_l {parent=deform_wrist_dn_base_l, loc=<0 0 0>, rot=<0 0 0 1>}
 deform_wrist_up_base_l {parent=hand_l, loc=<0 2.3009 0>, rot=<0 0 0 -1>}
 deform_wrist_up_l {parent=deform_wrist_up_base_l, loc=<0 0 0>, rot=<0 0 0 1>}
 lowerarm_twist_02_l {parent=lowerarm_l, loc=<10.6707 0 0>, rot=<0 0 0 1>}
 lowerarm_twist_01_l {parent=lowerarm_l, loc=<18.316 0 0>, rot=<0 0 0 1>}
 deform_elbow_in_base_l {parent=lowerarm_l, loc=<0.1834 3.6062 -0.2572>, rot=<0 0 0 -1>}
 deform_elbow_in_l {parent=deform_elbow_in_base_l, loc=<0 0 0>, rot=<0 0 0 1>}
 deform_elbow_out_base_l {parent=lowerarm_l, loc=<0 -4 0>, rot=<0 0 0 -1>}
 deform_elbow_out_l {parent=deform_elbow_out_base_l, loc=<0 0 0>, rot=<0 0 0 1>}
 upperarm_twist_01_l {parent=upperarm_l, loc=<7.2135 -0.1134 0>, rot=<0 0 0 1>}
 deform_upperarm_bk_base_l {parent=upperarm_twist_01_l, loc=<-6.4053 -4.3746 0.314>, rot=<0 0 0 -1>}
 deform_upperarm_bk_l {parent=deform_upperarm_bk_base_l, loc=<0 0 0>, rot=<0 0 0 1>}
 deform_upperarm_fr_base_l {parent=upperarm_twist_01_l, loc=<-5.6312 5.092 0.3354>, rot=<0 0 0 -1>}
 deform_upperarm_fr_l {parent=deform_upperarm_fr_base_l, loc=<0 0 0>, rot=<0 0 0 1>}
 deform_upperarm_up_base_l {parent=upperarm_twist_01_l, loc=<-8.891 0.4825 3.9517>, rot=<0 0 0 -1>}
 deform_upperarm_up_l {parent=deform_upperarm_up_base_l, loc=<0 0 0>, rot=<0 0 0 1>}
 upperarm_twist_02_l {parent=upperarm_l, loc=<19.7998 -0.1134 0>, rot=<0 0 0 1>}
 clavicle_r {parent=spine_05, loc=<5.4831 -0.1053 3.2946>, rot=<0.7289 0.0465 -0.6829 0.009>}
 upperarm_r {parent=clavicle_r, loc=<-9.8658 1.2136 0.4105>, rot=<0.0164 -0.3245 -0.0269 -0.9454>}
 lowerarm_r {parent=upperarm_r, loc=<-26.465 0.8345 -0.259>, rot=<0.0036 0.0161 -0.2298 -0.9731>}
 hand_r {parent=lowerarm_r, loc=<-22.7355 0.3203 0.6106>, rot=<-0.6784 -0.0223 0.0735 -0.7307>}
 thumb_01_r {parent=hand_r, loc=<-1.3118 0.3191 2.459>, rot=<0.6546 -0.4042 0.0664 -0.6354>}
 thumb_02_r {parent=thumb_01_r, loc=<-5.9336 0.0448 -0.0401>, rot=<-0.0037 0.0213 0.298 -0.9543>}
 thumb_03_r {parent=thumb_02_r, loc=<-2.2223 -0.3682 0.0907>, rot=<0.0118 -0.013 0.0066 -0.9998>}
 index_metacarpal_r {parent=hand_r, loc=<-1.6535 -0.0183 2.96>, rot=<0.0158 -0.0124 -0.0173 -0.9996>}
 index_01_r {parent=index_metacarpal_r, loc=<-8.3008 -0.0357 0.0524>, rot=<-0.0044 0.0485 0.0924 -0.9945>}
 index_02_r {parent=index_01_r, loc=<-3.3587 0.3004 -0.0452>, rot=<-0.002 0.0027 0.3225 -0.9466>}
 index_03_r {parent=index_02_r, loc=<-2.5578 0.123 0.0252>, rot=<-0.002 0.0067 0.2558 -0.9667>}
 middle_metacarpal_r {parent=hand_r, loc=<-2.1856 0.1725 0.5984>, rot=<0.006 0.0168 -0.0329 -0.9993>}
 middle_01_r {parent=middle_metacarpal_r, loc=<-7.3275 0.2447 -0.0129>, rot=<-0.0208 0.0042 0.0858 -0.9961>}
 middle_02_r {parent=middle_01_r, loc=<-4.0123 -0.1401 -0.1175>, rot=<-0.0218 0.0258 0.3512 -0.9357>}
 middle_03_r {parent=middle_02_r, loc=<-2.6077 -0.0534 -0.1373>, rot=<0.0082 0.0188 0.1708 -0.9851>}
 ring_metacarpal_r {parent=hand_r, loc=<-2.2074 0.3625 -1.3638>, rot=<-0.0284 0.0722 -0.0458 -0.9959>}
 ring_01_r {parent=ring_metacarpal_r, loc=<-7.0295 0.6197 0.1446>, rot=<-0.0489 -0.0252 0.1892 -0.9804>}
 ring_02_r {parent=ring_01_r, loc=<-3.6174 -0.4647 0.1027>, rot=<0 0 0.2806 -0.9598>}
 ring_03_r {parent=ring_02_r, loc=<-2.4152 -0.114 -0.0199>, rot=<0 0 0.1189 -0.9929>}
 pinky_metacarpal_r {parent=hand_r, loc=<-1.7616 0.9753 -3.3814>, rot=<-0.0397 0.0814 -0.069 -0.9935>}
 pinky_01_r {parent=pinky_metacarpal_r, loc=<-7.1709 0.4831 0.1094>, rot=<-0.0654 -0.0035 0.2242 -0.9723>}
 pinky_02_r {parent=pinky_01_r, loc=<-2.8274 -0.0893 0.1043>, rot=<0.0037 0.0011 0.2854 -0.9584>}
 pinky_03_r {parent=pinky_02_r, loc=<-2.2046 -0.0307 0.0078>, rot=<-0.0002 0.0098 0.1445 -0.9895>}
 weapon_r {parent=hand_r, loc=<0.1974 3.4111 -0.3811>, rot=<0 0 0 1>}
 hand_attach_r {parent=hand_r, loc=<-8.5001 2.9999 0>, rot=<0 0 -0.7071 -0.7071>}
 deform_wrist_dn_base_r {parent=hand_r, loc=<-0.0001 1.5328 0>, rot=<0 0 0 -1>}
 deform_wrist_dn_r {parent=deform_wrist_dn_base_r, loc=<0 0 0>, rot=<0 0 0 1>}
 deform_wrist_up_base_r {parent=hand_r, loc=<-0.0002 -2.3007 -0.0001>, rot=<0 0 0 -1>}
 deform_wrist_up_r {parent=deform_wrist_up_base_r, loc=<0 0 0>, rot=<0 0 0 1>}
 lowerarm_twist_02_r {parent=lowerarm_r, loc=<-10.6707 0 0>, rot=<0 0 0 -1>}
 lowerarm_twist_01_r {parent=lowerarm_r, loc=<-18.316 0 0>, rot=<0 0 0 -1>}
 deform_elbow_in_base_r {parent=lowerarm_r, loc=<-0.1834 -3.6062 0.2572>, rot=<0 0 0 -1>}
 deform_elbow_in_r {parent=deform_elbow_in_base_r, loc=<0 0 0>, rot=<0 0 0 1>}
 deform_elbow_out_base_r {parent=lowerarm_r, loc=<0 4 -0.0001>, rot=<0 0 0 -1>}
 deform_elbow_out_r {parent=deform_elbow_out_base_r, loc=<0 0 0>, rot=<0 0 0 1>}
 upperarm_twist_01_r {parent=upperarm_r, loc=<-7.2135 0.1134 0>, rot=<0 0 0 -1>}
 deform_upperarm_bk_base_r {parent=upperarm_twist_01_r, loc=<6.4053 4.3746 -0.314>, rot=<0 0 0 -1>}
 deform_upperarm_bk_r {parent=deform_upperarm_bk_base_r, loc=<0 0 0>, rot=<0 0 0 1>}
 deform_upperarm_fr_base_r {parent=upperarm_twist_01_r, loc=<5.6312 -5.092 -0.3354>, rot=<0 0 0 -1>}
 deform_upperarm_fr_r {parent=deform_upperarm_fr_base_r, loc=<0 0 0>, rot=<0 0 0 1>}
 deform_upperarm_up_base_r {parent=upperarm_twist_01_r, loc=<8.891 -0.4825 -3.9517>, rot=<0 0 0 -1>}
 deform_upperarm_up_r {parent=deform_upperarm_up_base_r, loc=<0 0 0>, rot=<0 0 0 1>}
 upperarm_twist_02_r {parent=upperarm_r, loc=<-19.7998 0.1134 0>, rot=<0 0 0 -1>}
 attach_backpack {parent=spine_05, loc=<-0.123 -13.473 0>, rot=<-0.0093 -0.707 0.0093 -0.707>}
 Backpack_BR {parent=attach_backpack, type=Null, loc=<0 0 0>, rot=<0 0 0 1>}
 attach_cape {parent=spine_05, loc=<9.017 -11.714 0>, rot=<0.0616 -0.7044 -0.0616 -0.7044>}
 cape {parent=attach_cape, type=Null, loc=<0 0 0>, rot=<0 0 0 1>}
 base_necklace {parent=spine_05, loc=<11.3138 0.1529 0>, rot=<-0.2693 -0.6538 -0.6538 -0.2693>}
 dyn_necklace_01 {parent=base_necklace, loc=<0 0 0>, rot=<0 0 0 1>}
 dyn_necklace_02 {parent=dyn_necklace_01, loc=<4.7 0 0>, rot=<0 0 0 1>}
 dyn_necklace_03 {parent=dyn_necklace_02, loc=<4.7 0 0>, rot=<0 0 0 1>}
 deform_pec_aim_base_l {parent=spine_04, loc=<3.2456 9.3309 -1.2237>, rot=<0.0603 -0.4033 0.3318 -0.8507>}
 deform_pec_aim_l {parent=deform_pec_aim_base_l, loc=<0 0 0>, rot=<0 0 0 1>}
 deform_pec_base_l {parent=deform_pec_aim_l, loc=<9.3665 0 0>, rot=<0 0 0 1>}
 deform_pec_l {parent=deform_pec_base_l, loc=<0 0 0>, rot=<0 0 0 1>}
 deform_pec_aim_base_r {parent=spine_04, loc=<3.2452 9.3309 1.2237>, rot=<-0.4033 -0.0603 0.8507 0.3318>}
 deform_pec_aim_r {parent=deform_pec_aim_base_r, loc=<0 0 0>, rot=<0 0 0 1>}
 deform_pec_base_r {parent=deform_pec_aim_r, loc=<-9.3667 -0.0003 -0.0004>, rot=<0 0 0 1>}
 deform_pec_r {parent=deform_pec_base_r, loc=<0 0 0>, rot=<0 0 0 1>}
 thigh_l {parent=pelvis, loc=<-1.0855 -0.2109 -8.7053>, rot=<0.0009 0.0355 0.0155 -0.9993>}
 calf_l {parent=thigh_l, loc=<-42.5284 -0.0019 -0.7297>, rot=<0.0006 -0.0161 -0.0391 -0.9991>}
 foot_l {parent=calf_l, loc=<-46.7401 -2.5014 -0.592>, rot=<0.0427 -0.0238 0.0353 -0.9982>}
 ball_l {parent=foot_l, loc=<-3.5658 10.8066 -0.1544>, rot=<0.0001 -0.0001 -0.7063 -0.7079>}
 calf_twist_02_l {parent=calf_l, loc=<-19.1546 -1.7943 -0.2638>, rot=<0.0237 0.0095 -0.0035 -0.9997>}
 calf_twist_01_l {parent=calf_l, loc=<-35.8483 -1.9197 -0.5794>, rot=<0.0237 0.0095 -0.0035 -0.9997>}
 deform_knee_in_base_l {parent=calf_l, loc=<0.8031 -5.0231 0.2456>, rot=<0 0 0 -1>}
 deform_knee_in_l {parent=deform_knee_in_base_l, loc=<0 0 0>, rot=<0 0 0 1>}
 deform_knee_out_base_l {parent=calf_l, loc=<0 4 0>, rot=<0 0 0 -1>}
 deform_knee_out_l {parent=deform_knee_out_base_l, loc=<0 0 0>, rot=<0 0 0 1>}
 thigh_twist_01_l {parent=thigh_l, loc=<-27.952 0 0>, rot=<0.0013 0 0 -1>}
 deform_groin_base_l {parent=thigh_l, loc=<0.4801 5.6776 -0.2486>, rot=<0 0 0 -1>}
 deform_groin_l {parent=deform_groin_base_l, loc=<0 0 0>, rot=<0 0 0 1>}
 thigh_r {parent=pelvis, loc=<-1.0855 -0.2109 8.7053>, rot=<0.0355 -0.0009 0.9993 0.0155>}
 calf_r {parent=thigh_r, loc=<42.5284 0.0019 0.7297>, rot=<0.0006 -0.0161 -0.0391 -0.9991>}
 foot_r {parent=calf_r, loc=<46.7401 2.5014 0.592>, rot=<0.0427 -0.0238 0.0353 -0.9982>}
 ball_r {parent=foot_r, loc=<3.5658 -10.8066 0.1544>, rot=<0.0001 -0.0001 -0.7063 -0.7079>}
 calf_twist_02_r {parent=calf_r, loc=<19.1546 1.7943 0.2638>, rot=<0.0237 0.0095 -0.0035 -0.9997>}
 calf_twist_01_r {parent=calf_r, loc=<35.8483 1.9197 0.5794>, rot=<0.0237 0.0095 -0.0035 -0.9997>}
 deform_knee_in_base_r {parent=calf_r, loc=<-0.8031 5.0231 -0.2457>, rot=<0 0 0 -1>}
 deform_knee_in_r {parent=deform_knee_in_base_r, loc=<0 0 0>, rot=<0 0 0 1>}
 deform_knee_out_base_r {parent=calf_r, loc=<0 -4 0>, rot=<0 0 0 -1>}
 deform_knee_out_r {parent=deform_knee_out_base_r, loc=<0 0 0>, rot=<0 0 0 1>}
 thigh_twist_01_r {parent=thigh_r, loc=<27.952 0 0>, rot=<0.0013 0 0 -1>}
 deform_groin_base_r {parent=thigh_r, loc=<-0.4801 -5.6776 0.2486>, rot=<0 0 0 -1>}
 deform_groin_r {parent=deform_groin_base_r, loc=<0 0 0>, rot=<0 0 0 1>}
 deform_glute_base_l {parent=pelvis, loc=<-7.3466 -9.0213 -9.1245>, rot=<0.0172 0.031 -0.4477 -0.8935>}
 deform_glute_l {parent=deform_glute_base_l, loc=<0 0 0>, rot=<0 0 0 1>}
 deform_glute_base_r {parent=pelvis, loc=<-7.3464 -9.0213 9.1245>, rot=<0.031 -0.0172 0.8935 -0.4477>}
 deform_glute_r {parent=deform_glute_base_r, loc=<0 0 0>, rot=<0 0 0 1>}
 base_holster {parent=pelvis, loc=<3.7734 8.6356 7.8724>, rot=<0.2439 0.5387 0.7619 -0.2644>}
 dyn_holster {parent=base_holster, loc=<0 0 0>, rot=<0 0 0 1>}
 thigh_l_null {parent=pelvis, type=Null, loc=<-1.0855 -0.2109 -8.7053>, rot=<0.0009 0.0355 0.0155 -0.9993>}
 thigh_r_null {parent=pelvis, type=Null, loc=<-1.0855 -0.2109 8.7053>, rot=<0.0355 -0.0009 0.9993 0.0155>}
 base_blouse_side_01_l {parent=pelvis, loc=<9.2706 -2.1521 -14.0111>, rot=<0 -0.9969 0 -0.0785>}
 dyn_blouse_side_01_l {parent=base_blouse_side_01_l, loc=<0 0 0>, rot=<0 0 0 1>}
 dyn_blouse_side_02_l {parent=dyn_blouse_side_01_l, loc=<11 0 0>, rot=<0 0 0 1>}
 dyn_blouse_side_03_l {parent=dyn_blouse_side_02_l, loc=<10.5 0 0>, rot=<0 0 0 1>}
 base_blouse_bk_01_l {parent=pelvis, loc=<9.2478 -9.6737 -10.0212>, rot=<0.0128 -0.8656 0.4983 -0.0476>}
 dyn_blouse_bk_01_l {parent=base_blouse_bk_01_l, loc=<0 0 0>, rot=<0 0 0 1>}
 dyn_blouse_bk_02_l {parent=dyn_blouse_bk_01_l, loc=<11 0 0>, rot=<0 0 0 1>}
 dyn_blouse_bk_03_l {parent=dyn_blouse_bk_02_l, loc=<10.5 0 0>, rot=<0 0 0 1>}
 base_blouse_01_c {parent=pelvis, loc=<8.7452 -12.2453 0.0084>, rot=<0.0278 -0.7066 0.7066 -0.0278>}
 dyn_blouse_01_c {parent=base_blouse_01_c, loc=<0 0 0>, rot=<0 0 0 1>}
 dyn_blouse_02_c {parent=dyn_blouse_01_c, loc=<11 0 0>, rot=<0 0 0 1>}
 dyn_blouse_03_c {parent=dyn_blouse_02_c, loc=<10.5 0 0>, rot=<0 0 0 1>}
 base_blouse_bk_01_r {parent=pelvis, loc=<9.2478 -9.6736 10.0212>, rot=<0.0476 -0.4983 0.8656 -0.0128>}
 dyn_blouse_bk_01_r {parent=base_blouse_bk_01_r, loc=<0 0 0>, rot=<0 0 0 1>}
 dyn_blouse_bk_02_r {parent=dyn_blouse_bk_01_r, loc=<11 0 0>, rot=<0 0 0 1>}
 dyn_blouse_bk_03_r {parent=dyn_blouse_bk_02_r, loc=<10.5 0 0>, rot=<0 0 0 1>}
 base_blouse_side_01_r {parent=pelvis, loc=<9.2706 -2.1521 14.0111>, rot=<0.0785 0 0.9969 0>}
 dyn_blouse_side_01_r {parent=base_blouse_side_01_r, loc=<0 0 0>, rot=<0 0 0 1>}
 dyn_blouse_side_02_r {parent=dyn_blouse_side_01_r, loc=<11 0 0>, rot=<0 0 0 1>}
 dyn_blouse_side_03_r {parent=dyn_blouse_side_02_r, loc=<10.5 0 0>, rot=<0 0 0 1>}
 base_blouse_fr_main_01_l {parent=pelvis, loc=<20.0715 6.3839 -10.0111>, rot=<-0.0206 -0.9039 -0.4163 -0.096>}
 dyn_blouse_fr_main_01_l {parent=base_blouse_fr_main_01_l, loc=<0 0 0>, rot=<0 0 0 1>}
 base_blouse_fr_01_l {parent=dyn_blouse_fr_main_01_l, loc=<10.8361 -0.1226 -0.3134>, rot=<0.0059 -0.0358 0.0187 -0.9992>}
 dyn_blouse_fr_01_l {parent=base_blouse_fr_01_l, loc=<0 0 0>, rot=<0 0 0 1>}
 dyn_blouse_fr_02_l {parent=dyn_blouse_fr_01_l, loc=<11 0 0>, rot=<0 0 0 1>}
 dyn_blouse_fr_03_l {parent=dyn_blouse_fr_02_l, loc=<11.5 0 0>, rot=<0 0 0 1>}
 base_blouse_fr_main_01_r {parent=pelvis, loc=<20.1409 6.3756 10>, rot=<0.0882 0.4171 0.9043 0.0244>}
 dyn_blouse_fr_main_01_r {parent=base_blouse_fr_main_01_r, loc=<0 0 0>, rot=<0 0 0 1>}
 base_blouse_fr_01_r {parent=dyn_blouse_fr_main_01_r, loc=<10.9071 0.2684 -0.191>, rot=<-0.0055 -0.0302 -0.0254 -0.9992>}
 dyn_blouse_fr_01_r {parent=base_blouse_fr_01_r, loc=<0 0 0>, rot=<0 0 0 1>}
 dyn_blouse_fr_02_r {parent=dyn_blouse_fr_01_r, loc=<11 0 0>, rot=<0 0 0 1>}
 dyn_blouse_fr_03_r {parent=dyn_blouse_fr_02_r, loc=<11.5 0 0>, rot=<0 0 0 1>}
 attach {parent=root, loc=<0 0 0>, rot=<0 0 0 1>}
 ik_hand_root {parent=root, loc=<0 0 0>, rot=<0 0 0 1>}
 ik_hand_gun {parent=ik_hand_root, loc=<-48.1743 3.3545 104.4381>, rot=<0.6276 0.1309 0.4183 -0.6434>}
 ik_hand_r {parent=ik_hand_gun, loc=<0 0 0>, rot=<0 0 0 1>}
 ik_hand_l {parent=ik_hand_gun, loc=<59.3271 67.6997 34.3524>, rot=<-0.6158 -0.7027 -0.3565 0>}
 ik_foot_root {parent=root, loc=<0 0 0>, rot=<0 0 0 1>}
 ik_foot_l {parent=ik_foot_root, loc=<14.8262 -7.3545 8.296>, rot=<0.0357 0.7045 0.0284 -0.7082>}
 ik_foot_r {parent=ik_foot_root, loc=<-14.8262 -7.3545 8.296>, rot=<0.7082 0.0284 -0.7045 0.0357>}
 root_fp {parent=root, loc=<0 0 0>, rot=<0 0 0 1>}
 attach_fp {parent=root_fp, loc=<0 0 0>, rot=<0 0 0 1>}
 camera_root {parent=root_fp, loc=<0 0 0>, rot=<0 0 0 1>}
 camera {parent=camera_root, loc=<0 0 0>, rot=<0 0 0 1>}
 attach_emote_01 {parent=root, loc=<0 0 0>, rot=<0 0 0 1>}
 attach_emote_02 {parent=root, loc=<0 0 0>, rot=<0 0 0 1>}
 attach_emote_03 {parent=root, loc=<0 0 0>, rot=<0 0 0 1>}
 warp_point_01 {parent=root, loc=<0 0 0>, rot=<0 0 0 1>}
 warp_point_02 {parent=root, loc=<0 0 0>, rot=<0 0 0 1>}
 warp_point_03 {parent=root, loc=<0 0 0>, rot=<0 0 0 1>}
]
variables [
 BaseballSlide {type=double, default=0}
 CrouchWalk {type=double, default=0}
 CrouchingNotMove {type=double, default=0}
 Kneeslide {type=double, default=0}
 PickaxeSwing {type=double, default=0}
]
