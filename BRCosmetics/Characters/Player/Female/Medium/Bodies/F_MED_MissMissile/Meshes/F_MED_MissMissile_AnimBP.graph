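# Animation Blueprint for the F_MED_MissMissile female cosmetic body, adding secondary motion to a pose copied from its attached parent mesh.
# A Control Rig adjusts for crouch and emotes, then a pelvis-space RigidBody node runs, weighted by Dyn_Alpha.
# Chained per-bone AnimDynamics springs move blouse and necklace, before a post-sim Control Rig handles slide and crouch states.
# The graph repeats this chain, with a mime-emote input on the copy.
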AnimBlueprint /BRCosmetics/Characters/Player/Female/Medium/Bodies/F_MED_MissMissile/Meshes/F_MED_MissMissile_AnimBP
target_skeleton=&/F_MED_MissMissile_Skeleton
graphs [
 AnimGraph {nodes [
   n1 = Control Rig!(
    Source=Copy Pose From Mesh() {bUseAttachedParent=true}
    Alpha
    Crouching:double=Get IsCrouching() -> (IsCrouching:double)
    Playing Emote:double=Get Playing Emote() -> (Playing Emote:double)
   ) -> (Pose) {controlRigAssetReference {blueprintRigClass=&/F_MED_MissMissile_CtrlRig.F_MED_MissMissile_CtrlRig_C}}
   n2 = RigidBody!(
    ComponentPose=Local To Component(LocalPose=@n1)
    Alpha=Get Dyn_Alpha() -> (Dyn_Alpha:double)
    OverrideWorldGravity:Vector="0,0,0"
    ExternalForce
   ) -> (Pose) {
    bOverrideWorldGravity=true
    overridePhysicsAsset=&/F_MED_MissMissile_Physics
    simulationSpace=BaseBoneSpace
    baseBoneRef {boneName=pelvis}
    componentAppliedLinearAccClamp=<2000 2000 2000>
    simSpaceSettings {maxAngularAcceleration=100, maxAngularVelocity=2, maxLinearAcceleration=100, maxLinearVelocity=160, worldAlpha=0.7}
   }
   n3 = Anim Dynamics - Bone: dyn_blouse_fr_01_l!(ComponentPose=@n2, Alpha=1, GravityOverride:Vector="0,0,-980", ExternalForce="0,0,0") -> (Pose) {
    angularDampingOverride=1
    angularSpringConstant=2500
    bAngularSpring=true
    bOverrideAngularDamping=true
    bUseGravityOverride=true
    gravityScale=0.35
    simulationSpace=RootRelative
    boundBone {boneName=dyn_blouse_fr_01_l}
    componentAppliedLinearAccClamp=<4000 4000 4000>
    componentLinearAccScale=<0.6 0.6 0.6>
    componentLinearVelScale=<0.6 0.6 0.6>
    gravityOverride {z=-980}
    physicsBodyDefinitions [{
      boundBone {boneName=dyn_blouse_fr_01_l}
      boxExtents=<20 30 30>
      constraintSetup {angularLimitsMax=<0 0 2>, angularLimitsMin=<0 -20 -2>, angularTarget=<1 0 0>}
      localJointOffset=<4 0 0>
     }]
   }
   n4 = Anim Dynamics - Bone: dyn_blouse_fr_02_l!(ComponentPose=@n3, Alpha=1, GravityOverride:Vector="0,0,-980", ExternalForce="0,0,0") -> (Pose) {
    angularDampingOverride=1
    angularSpringConstant=2500
    bAngularSpring=true
    bOverrideAngularDamping=true
    bUseGravityOverride=true
    gravityScale=0.35
    simulationSpace=RootRelative
    boundBone {boneName=dyn_blouse_fr_02_l}
    componentAppliedLinearAccClamp=<4000 4000 4000>
    componentLinearAccScale=<0.6 0.6 0.6>
    componentLinearVelScale=<0.6 0.6 0.6>
    gravityOverride {z=-980}
    physicsBodyDefinitions [{
      boundBone {boneName=dyn_blouse_fr_02_l}
      boxExtents=<20 30 30>
      constraintSetup {angularLimitsMax=<0 0 2>, angularLimitsMin=<0 -20 -2>, angularTarget=<1 0 0>}
      localJointOffset=<4 0 0>
     }]
   }
   n5 = Anim Dynamics - Bone: dyn_blouse_fr_03_l!(ComponentPose=@n4, Alpha=1, GravityOverride:Vector="0,0,-980", ExternalForce="0,0,0") -> (Pose) {
    angularDampingOverride=1
    angularSpringConstant=2500
    bAngularSpring=true
    bOverrideAngularDamping=true
    bUseGravityOverride=true
    gravityScale=0.35
    simulationSpace=RootRelative
    boundBone {boneName=dyn_blouse_fr_03_l}
    componentAppliedLinearAccClamp=<4000 4000 4000>
    componentLinearAccScale=<0.6 0.6 0.6>
    componentLinearVelScale=<0.6 0.6 0.6>
    gravityOverride {z=-980}
    physicsBodyDefinitions [{
      boundBone {boneName=dyn_blouse_fr_03_l}
      boxExtents=<20 30 30>
      constraintSetup {angularLimitsMax=<0 0 2>, angularLimitsMin=<0 -20 -2>, angularTarget=<1 0 0>}
      localJointOffset=<4 0 0>
     }]
   }
   n6 = Anim Dynamics - Bone: dyn_blouse_side_01_l!(ComponentPose=@n5, Alpha=1, GravityOverride:Vector="0,0,-980", ExternalForce="0,0,0") -> (Pose) {
    angularDampingOverride=1
    angularSpringConstant=2500
    bAngularSpring=true
    bOverrideAngularDamping=true
    bUseGravityOverride=true
    gravityScale=0.35
    simulationSpace=RootRelative
    boundBone {boneName=dyn_blouse_side_01_l}
    componentAppliedLinearAccClamp=<4000 4000 4000>
    componentLinearAccScale=<0.6 0.6 0.6>
    componentLinearVelScale=<0.6 0.6 0.6>
    gravityOverride {z=-980}
    physicsBodyDefinitions [{
      boundBone {boneName=dyn_blouse_side_01_l}
      boxExtents=<20 30 30>
      constraintSetup {angularLimitsMax=<0 0 2>, angularLimitsMin=<0 -20 -2>, angularTarget=<1 0 0>}
      localJointOffset=<4 0 0>
     }]
   }
   n7 = Anim Dynamics - Bone: dyn_blouse_side_02_l!(ComponentPose=@n6, Alpha=1, GravityOverride:Vector="0,0,-980", ExternalForce="0,0,0") -> (Pose) {
    angularDampingOverride=1
    angularSpringConstant=2500
    bAngularSpring=true
    bOverrideAngularDamping=true
    bUseGravityOverride=true
    gravityScale=0.35
    simulationSpace=RootRelative
    boundBone {boneName=dyn_blouse_side_02_l}
    componentAppliedLinearAccClamp=<4000 4000 4000>
    componentLinearAccScale=<0.6 0.6 0.6>
    componentLinearVelScale=<0.6 0.6 0.6>
    gravityOverride {z=-980}
    physicsBodyDefinitions [{
      boundBone {boneName=dyn_blouse_side_02_l}
      boxExtents=<20 30 30>
      constraintSetup {angularLimitsMax=<0 0 2>, angularLimitsMin=<0 -20 -2>, angularTarget=<1 0 0>}
      localJointOffset=<4 0 0>
     }]
   }
   n8 = Anim Dynamics - Bone: dyn_blouse_side_03_l!(ComponentPose=@n7, Alpha=1, GravityOverride:Vector="0,0,-980", ExternalForce="0,0,0") -> (Pose) {
    angularDampingOverride=1
    angularSpringConstant=2500
    bAngularSpring=true
    bOverrideAngularDamping=true
    bUseGravityOverride=true
    gravityScale=0.35
    simulationSpace=RootRelative
    boundBone {boneName=dyn_blouse_side_03_l}
    componentAppliedLinearAccClamp=<4000 4000 4000>
    componentLinearAccScale=<0.6 0.6 0.6>
    componentLinearVelScale=<0.6 0.6 0.6>
    gravityOverride {z=-980}
    physicsBodyDefinitions [{
      boundBone {boneName=dyn_blouse_side_03_l}
      boxExtents=<20 30 30>
      constraintSetup {angularLimitsMax=<0 0 2>, angularLimitsMin=<0 -20 -2>, angularTarget=<1 0 0>}
      localJointOffset=<4 0 0>
     }]
   }
   n9 = Anim Dynamics - Bone: dyn_blouse_bk_01_l!(ComponentPose=@n8, Alpha=1, GravityOverride:Vector="0,0,-980", ExternalForce="0,0,0") -> (Pose) {
    angularDampingOverride=1
    angularSpringConstant=2500
    bAngularSpring=true
    bOverrideAngularDamping=true
    bUseGravityOverride=true
    gravityScale=0.35
    simulationSpace=RootRelative
    boundBone {boneName=dyn_blouse_bk_01_l}
    componentAppliedLinearAccClamp=<4000 4000 4000>
    componentLinearAccScale=<0.6 0.6 0.6>
    componentLinearVelScale=<0.6 0.6 0.6>
    gravityOverride {z=-980}
    physicsBodyDefinitions [{
      boundBone {boneName=dyn_blouse_bk_01_l}
      boxExtents=<20 30 30>
      constraintSetup {angularLimitsMax=<0 0 2>, angularLimitsMin=<0 -20 -2>, angularTarget=<1 0 0>}
      localJointOffset=<4 0 0>
     }]
   }
   n10 = Anim Dynamics - Bone: dyn_blouse_bk_02_l!(ComponentPose=@n9, Alpha=1, GravityOverride:Vector="0,0,-980", ExternalForce="0,0,0") -> (Pose) {
    angularDampingOverride=1
    angularSpringConstant=2500
    bAngularSpring=true
    bOverrideAngularDamping=true
    bUseGravityOverride=true
    gravityScale=0.35
    simulationSpace=RootRelative
    boundBone {boneName=dyn_blouse_bk_02_l}
    componentAppliedLinearAccClamp=<4000 4000 4000>
    componentLinearAccScale=<0.6 0.6 0.6>
    componentLinearVelScale=<0.6 0.6 0.6>
    gravityOverride {z=-980}
    physicsBodyDefinitions [{
      boundBone {boneName=dyn_blouse_bk_02_l}
      boxExtents=<20 30 30>
      constraintSetup {angularLimitsMax=<0 0 2>, angularLimitsMin=<0 -20 -2>, angularTarget=<1 0 0>}
      localJointOffset=<4 0 0>
     }]
   }
   n11 = Anim Dynamics - Bone: dyn_blouse_bk_03_l!(ComponentPose=@n10, Alpha=1, GravityOverride:Vector="0,0,-980", ExternalForce="0,0,0") -> (Pose) {
    angularDampingOverride=1
    angularSpringConstant=2500
    bAngularSpring=true
    bOverrideAngularDamping=true
    bUseGravityOverride=true
    gravityScale=0.35
    simulationSpace=RootRelative
    boundBone {boneName=dyn_blouse_bk_03_l}
    componentAppliedLinearAccClamp=<4000 4000 4000>
    componentLinearAccScale=<0.6 0.6 0.6>
    componentLinearVelScale=<0.6 0.6 0.6>
    gravityOverride {z=-980}
    physicsBodyDefinitions [{
      boundBone {boneName=dyn_blouse_bk_03_l}
      boxExtents=<20 30 30>
      constraintSetup {angularLimitsMax=<0 0 2>, angularLimitsMin=<0 -20 -2>, angularTarget=<1 0 0>}
      localJointOffset=<4 0 0>
     }]
   }
   n12 = Anim Dynamics - Bone: dyn_blouse_01_c!(ComponentPose=@n11, Alpha=1, GravityOverride:Vector="0,0,-980", ExternalForce="0,0,0") -> (Pose) {
    angularDampingOverride=1
    angularSpringConstant=2500
    bAngularSpring=true
    bOverrideAngularDamping=true
    bUseGravityOverride=true
    gravityScale=0.35
    simulationSpace=RootRelative
    boundBone {boneName=dyn_blouse_01_c}
    componentAppliedLinearAccClamp=<4000 4000 4000>
    componentLinearAccScale=<0.6 0.6 0.6>
    componentLinearVelScale=<0.6 0.6 0.6>
    gravityOverride {z=-980}
    physicsBodyDefinitions [{
      boundBone {boneName=dyn_blouse_01_c}
      boxExtents=<20 30 30>
      constraintSetup {angularLimitsMax=<0 0 2>, angularLimitsMin=<0 -20 -2>, angularTarget=<1 0 0>}
      localJointOffset=<4 0 0>
     }]
   }
   n13 = Anim Dynamics - Bone: dyn_blouse_02_c!(ComponentPose=@n12, Alpha=1, GravityOverride:Vector="0,0,-980", ExternalForce="0,0,0") -> (Pose) {
    angularDampingOverride=1
    angularSpringConstant=2500
    bAngularSpring=true
    bOverrideAngularDamping=true
    bUseGravityOverride=true
    gravityScale=0.35
    simulationSpace=RootRelative
    boundBone {boneName=dyn_blouse_02_c}
    componentAppliedLinearAccClamp=<4000 4000 4000>
    componentLinearAccScale=<0.6 0.6 0.6>
    componentLinearVelScale=<0.6 0.6 0.6>
    gravityOverride {z=-980}
    physicsBodyDefinitions [{
      boundBone {boneName=dyn_blouse_02_c}
      boxExtents=<20 30 30>
      constraintSetup {angularLimitsMax=<0 0 2>, angularLimitsMin=<0 -20 -2>, angularTarget=<1 0 0>}
      localJointOffset=<4 0 0>
     }]
   }
   n14 = Anim Dynamics - Bone: dyn_blouse_03_c!(ComponentPose=@n13, Alpha=1, GravityOverride:Vector="0,0,-980", ExternalForce="0,0,0") -> (Pose) {
    angularDampingOverride=1
    angularSpringConstant=2500
    bAngularSpring=true
    bOverrideAngularDamping=true
    bUseGravityOverride=true
    gravityScale=0.35
    simulationSpace=RootRelative
    boundBone {boneName=dyn_blouse_03_c}
    componentAppliedLinearAccClamp=<4000 4000 4000>
    componentLinearAccScale=<0.6 0.6 0.6>
    componentLinearVelScale=<0.6 0.6 0.6>
    gravityOverride {z=-980}
    physicsBodyDefinitions [{
      boundBone {boneName=dyn_blouse_03_c}
      boxExtents=<20 30 30>
      constraintSetup {angularLimitsMax=<0 0 2>, angularLimitsMin=<0 -20 -2>, angularTarget=<1 0 0>}
      localJointOffset=<4 0 0>
     }]
   }
   n15 = Anim Dynamics - Bone: dyn_blouse_bk_01_r!(ComponentPose=@n14, Alpha=1, GravityOverride:Vector="0,0,-980", ExternalForce="0,0,0") -> (Pose) {
    angularDampingOverride=1
    angularSpringConstant=2500
    bAngularSpring=true
    bOverrideAngularDamping=true
    bUseGravityOverride=true
    gravityScale=0.35
    simulationSpace=RootRelative
    boundBone {boneName=dyn_blouse_bk_01_r}
    componentAppliedLinearAccClamp=<4000 4000 4000>
    componentLinearAccScale=<0.6 0.6 0.6>
    componentLinearVelScale=<0.6 0.6 0.6>
    gravityOverride {z=-980}
    physicsBodyDefinitions [{
      boundBone {boneName=dyn_blouse_bk_01_r}
      boxExtents=<20 30 30>
      constraintSetup {angularLimitsMax=<0 0 2>, angularLimitsMin=<0 -20 -2>, angularTarget=<1 0 0>}
      localJointOffset=<4 0 0>
     }]
   }
   n16 = Anim Dynamics - Bone: dyn_blouse_bk_02_r!(ComponentPose=@n15, Alpha=1, GravityOverride:Vector="0,0,-980", ExternalForce="0,0,0") -> (Pose) {
    angularDampingOverride=1
    angularSpringConstant=2500
    bAngularSpring=true
    bOverrideAngularDamping=true
    bUseGravityOverride=true
    gravityScale=0.35
    simulationSpace=RootRelative
    boundBone {boneName=dyn_blouse_bk_02_r}
    componentAppliedLinearAccClamp=<4000 4000 4000>
    componentLinearAccScale=<0.6 0.6 0.6>
    componentLinearVelScale=<0.6 0.6 0.6>
    gravityOverride {z=-980}
    physicsBodyDefinitions [{
      boundBone {boneName=dyn_blouse_bk_02_r}
      boxExtents=<20 30 30>
      constraintSetup {angularLimitsMax=<0 0 2>, angularLimitsMin=<0 -20 -2>, angularTarget=<1 0 0>}
      localJointOffset=<4 0 0>
     }]
   }
   n17 = Anim Dynamics - Bone: dyn_blouse_bk_03_r!(ComponentPose=@n16, Alpha=1, GravityOverride:Vector="0,0,-980", ExternalForce="0,0,0") -> (Pose) {
    angularDampingOverride=1
    angularSpringConstant=2500
    bAngularSpring=true
    bOverrideAngularDamping=true
    bUseGravityOverride=true
    gravityScale=0.35
    simulationSpace=RootRelative
    boundBone {boneName=dyn_blouse_bk_03_r}
    componentAppliedLinearAccClamp=<4000 4000 4000>
    componentLinearAccScale=<0.6 0.6 0.6>
    componentLinearVelScale=<0.6 0.6 0.6>
    gravityOverride {z=-980}
    physicsBodyDefinitions [{
      boundBone {boneName=dyn_blouse_bk_03_r}
      boxExtents=<20 30 30>
      constraintSetup {angularLimitsMax=<0 0 2>, angularLimitsMin=<0 -20 -2>, angularTarget=<1 0 0>}
      localJointOffset=<4 0 0>
     }]
   }
   n18 = Anim Dynamics - Bone: dyn_blouse_side_01_r!(ComponentPose=@n17, Alpha=1, GravityOverride:Vector="0,0,-980", ExternalForce="0,0,0") -> (Pose) {
    angularDampingOverride=1
    angularSpringConstant=2500
    bAngularSpring=true
    bOverrideAngularDamping=true
    bUseGravityOverride=true
    gravityScale=0.35
    simulationSpace=RootRelative
    boundBone {boneName=dyn_blouse_side_01_r}
    componentAppliedLinearAccClamp=<4000 4000 4000>
    componentLinearAccScale=<0.6 0.6 0.6>
    componentLinearVelScale=<0.6 0.6 0.6>
    gravityOverride {z=-980}
    physicsBodyDefinitions [{
      boundBone {boneName=dyn_blouse_side_01_r}
      boxExtents=<20 30 30>
      constraintSetup {angularLimitsMax=<0 0 2>, angularLimitsMin=<0 -20 -2>, angularTarget=<1 0 0>}
      localJointOffset=<4 0 0>
     }]
   }
   n19 = Anim Dynamics - Bone: dyn_blouse_side_02_r!(ComponentPose=@n18, Alpha=1, GravityOverride:Vector="0,0,-980", ExternalForce="0,0,0") -> (Pose) {
    angularDampingOverride=1
    angularSpringConstant=2500
    bAngularSpring=true
    bOverrideAngularDamping=true
    bUseGravityOverride=true
    gravityScale=0.35
    simulationSpace=RootRelative
    boundBone {boneName=dyn_blouse_side_02_r}
    componentAppliedLinearAccClamp=<4000 4000 4000>
    componentLinearAccScale=<0.6 0.6 0.6>
    componentLinearVelScale=<0.6 0.6 0.6>
    gravityOverride {z=-980}
    physicsBodyDefinitions [{
      boundBone {boneName=dyn_blouse_side_02_r}
      boxExtents=<20 30 30>
      constraintSetup {angularLimitsMax=<0 0 2>, angularLimitsMin=<0 -20 -2>, angularTarget=<1 0 0>}
      localJointOffset=<4 0 0>
     }]
   }
   n20 = Anim Dynamics - Bone: dyn_blouse_side_03_r!(ComponentPose=@n19, Alpha=1, GravityOverride:Vector="0,0,-980", ExternalForce="0,0,0") -> (Pose) {
    angularDampingOverride=1
    angularSpringConstant=2500
    bAngularSpring=true
    bOverrideAngularDamping=true
    bUseGravityOverride=true
    gravityScale=0.35
    simulationSpace=RootRelative
    boundBone {boneName=dyn_blouse_side_03_r}
    componentAppliedLinearAccClamp=<4000 4000 4000>
    componentLinearAccScale=<0.6 0.6 0.6>
    componentLinearVelScale=<0.6 0.6 0.6>
    gravityOverride {z=-980}
    physicsBodyDefinitions [{
      boundBone {boneName=dyn_blouse_side_03_r}
      boxExtents=<20 30 30>
      constraintSetup {angularLimitsMax=<0 0 2>, angularLimitsMin=<0 -20 -2>, angularTarget=<1 0 0>}
      localJointOffset=<4 0 0>
     }]
   }
   n21 = Anim Dynamics - Bone: dyn_blouse_fr_01_r!(ComponentPose=@n20, Alpha=1, GravityOverride:Vector="0,0,-980", ExternalForce="0,0,0") -> (Pose) {
    angularDampingOverride=1
    angularSpringConstant=2500
    bAngularSpring=true
    bOverrideAngularDamping=true
    bUseGravityOverride=true
    gravityScale=0.35
    simulationSpace=RootRelative
    boundBone {boneName=dyn_blouse_fr_01_r}
    componentAppliedLinearAccClamp=<4000 4000 4000>
    componentLinearAccScale=<0.6 0.6 0.6>
    componentLinearVelScale=<0.6 0.6 0.6>
    gravityOverride {z=-980}
    physicsBodyDefinitions [{
      boundBone {boneName=dyn_blouse_fr_01_r}
      boxExtents=<20 30 30>
      constraintSetup {angularLimitsMax=<0 0 2>, angularLimitsMin=<0 -20 -2>, angularTarget=<1 0 0>}
      localJointOffset=<4 0 0>
     }]
   }
   n22 = Anim Dynamics - Bone: dyn_blouse_fr_02_r!(ComponentPose=@n21, Alpha=1, GravityOverride:Vector="0,0,-980", ExternalForce="0,0,0") -> (Pose) {
    angularDampingOverride=1
    angularSpringConstant=2500
    bAngularSpring=true
    bOverrideAngularDamping=true
    bUseGravityOverride=true
    gravityScale=0.35
    simulationSpace=RootRelative
    boundBone {boneName=dyn_blouse_fr_02_r}
    componentAppliedLinearAccClamp=<4000 4000 4000>
    componentLinearAccScale=<0.6 0.6 0.6>
    componentLinearVelScale=<0.6 0.6 0.6>
    gravityOverride {z=-980}
    physicsBodyDefinitions [{
      boundBone {boneName=dyn_blouse_fr_02_r}
      boxExtents=<20 30 30>
      constraintSetup {angularLimitsMax=<0 0 2>, angularLimitsMin=<0 -20 -2>, angularTarget=<1 0 0>}
      localJointOffset=<4 0 0>
     }]
   }
   n23 = Anim Dynamics - Bone: dyn_blouse_fr_03_r!(ComponentPose=@n22, Alpha=1, GravityOverride:Vector="0,0,-980", ExternalForce="0,0,0") -> (Pose) {
    angularDampingOverride=1
    angularSpringConstant=2500
    bAngularSpring=true
    bOverrideAngularDamping=true
    bUseGravityOverride=true
    gravityScale=0.35
    simulationSpace=RootRelative
    boundBone {boneName=dyn_blouse_fr_03_r}
    componentAppliedLinearAccClamp=<4000 4000 4000>
    componentLinearAccScale=<0.6 0.6 0.6>
    componentLinearVelScale=<0.6 0.6 0.6>
    gravityOverride {z=-980}
    physicsBodyDefinitions [{
      boundBone {boneName=dyn_blouse_fr_03_r}
      boxExtents=<20 30 30>
      constraintSetup {angularLimitsMax=<0 0 2>, angularLimitsMin=<0 -20 -2>, angularTarget=<1 0 0>}
      localJointOffset=<4 0 0>
     }]
   }
   n24 = "Transform (Modify) Bone - Bone: dyn_blouse_bk_01_l"(ComponentPose=@n23, Alpha=1, Translation="0,0,0", Rotation="0,0,0", Scale="1,1,1") {rotationMode=BMM_Additive, boneToModify {boneName=dyn_blouse_bk_01_l}}
   n25 = "Transform (Modify) Bone - Bone: dyn_blouse_01_c"(
    ComponentPose="Transform (Modify) Bone - Bone: dyn_blouse_bk_03_l"(
     ComponentPose="Transform (Modify) Bone - Bone: dyn_blouse_bk_02_l"(ComponentPose=@n24, Alpha=1, Translation="0,0,0", Rotation="0,0,0", Scale="1,1,1") {rotationMode=BMM_Additive, boneToModify {boneName=dyn_blouse_bk_02_l}}
     Alpha=1
     Translation="0,0,0"
     Rotation="0,0,0"
     Scale="1,1,1"
    ) {rotationMode=BMM_Additive, boneToModify {boneName=dyn_blouse_bk_03_l}}
    Alpha=1
    Translation="0,0,0"
    Rotation="0,0,0"
    Scale="1,1,1"
   ) {rotationMode=BMM_Additive, boneToModify {boneName=dyn_blouse_01_c}}
   n26 = "Transform (Modify) Bone - Bone: dyn_blouse_bk_01_r"(
    ComponentPose="Transform (Modify) Bone - Bone: dyn_blouse_03_c"(
     ComponentPose="Transform (Modify) Bone - Bone: dyn_blouse_02_c"(ComponentPose=@n25, Alpha=1, Translation="0,0,0", Rotation="0,0,0", Scale="1,1,1") {rotationMode=BMM_Additive, boneToModify {boneName=dyn_blouse_02_c}}
     Alpha=1
     Translation="0,0,0"
     Rotation="0,0,0"
     Scale="1,1,1"
    ) {rotationMode=BMM_Additive, boneToModify {boneName=dyn_blouse_03_c}}
    Alpha=1
    Translation="0,0,0"
    Rotation="0,0,0"
    Scale="1,1,1"
   ) {rotationMode=BMM_Additive, boneToModify {boneName=dyn_blouse_bk_01_r}}
   n27 = "Transform (Modify) Bone - Bone: dyn_blouse_side_01_r"(
    ComponentPose="Transform (Modify) Bone - Bone: dyn_blouse_bk_03_r"(
     ComponentPose="Transform (Modify) Bone - Bone: dyn_blouse_bk_02_r"(ComponentPose=@n26, Alpha=1, Translation="0,0,0", Rotation="0,0,0", Scale="1,1,1") {rotationMode=BMM_Additive, boneToModify {boneName=dyn_blouse_bk_02_r}}
     Alpha=1
     Translation="0,0,0"
     Rotation="0,0,0"
     Scale="1,1,1"
    ) {rotationMode=BMM_Additive, boneToModify {boneName=dyn_blouse_bk_03_r}}
    Alpha=1
    Translation="0,0,0"
    Rotation="0,0,0"
    Scale="1,1,1"
   ) {rotationMode=BMM_Additive, boneToModify {boneName=dyn_blouse_side_01_r}}
   n28 = "Transform (Modify) Bone - Bone: dyn_blouse_side_01_l"(
    ComponentPose="Transform (Modify) Bone - Bone: dyn_blouse_side_03_r"(
     ComponentPose="Transform (Modify) Bone - Bone: dyn_blouse_side_02_r"(ComponentPose=@n27, Alpha=1, Translation="0,0,0", Rotation="0,0,0", Scale="1,1,1") {rotationMode=BMM_Additive, boneToModify {boneName=dyn_blouse_side_02_r}}
     Alpha=1
     Translation="0,0,0"
     Rotation="0,0,0"
     Scale="1,1,1"
    ) {rotationMode=BMM_Additive, boneToModify {boneName=dyn_blouse_side_03_r}}
    Alpha=1
    Translation="0,0,0"
    Rotation="0,0,0"
    Scale="1,1,1"
   ) {rotationMode=BMM_Additive, boneToModify {boneName=dyn_blouse_side_01_l}}
   n29 = Anim Dynamics - Bone: dyn_necklace_01!(
    ComponentPose="Transform (Modify) Bone - Bone: dyn_blouse_side_03_l"(
     ComponentPose="Transform (Modify) Bone - Bone: dyn_blouse_side_02_l"(ComponentPose=@n28, Alpha=1, Translation="0,0,0", Rotation="0,0,0", Scale="1,1,1") {rotationMode=BMM_Additive, boneToModify {boneName=dyn_blouse_side_02_l}}
     Alpha=1
     Translation="0,0,0"
     Rotation="0,0,0"
     Scale="1,1,1"
    ) {rotationMode=BMM_Additive, boneToModify {boneName=dyn_blouse_side_03_l}}
    Alpha=1
    GravityOverride:Vector="0,0,-980"
    ExternalForce="0,0,0"
   ) -> (Pose) {
    angularDampingOverride=0.9999
    angularSpringConstant=10000
    bAngularSpring=true
    bOverrideAngularDamping=true
    bUseGravityOverride=true
    gravityScale=0.35
    simulationSpace=RootRelative
    boundBone {boneName=dyn_necklace_01}
    componentAppliedLinearAccClamp=<2000 2000 2000>
    componentLinearAccScale=<0.4 0.4 0.4>
    componentLinearVelScale=<0.4 0.4 0.4>
    gravityOverride {z=-980}
    physicsBodyDefinitions [{
      boundBone {boneName=dyn_necklace_01}
      boxExtents=<30 20 20>
      constraintSetup {angularLimitsMax=<0 -2 10>, angularLimitsMin=<0 -5 -10>, angularTarget=<1 0 0>}
      localJointOffset=<2 0 0>
     }]
   }
   n30 = Anim Dynamics - Bone: dyn_necklace_02!(ComponentPose=@n29, Alpha=1, GravityOverride:Vector="0,0,-980", ExternalForce="0,0,0") -> (Pose) {
    angularDampingOverride=0.9999
    angularSpringConstant=10000
    bAngularSpring=true
    bOverrideAngularDamping=true
    bUseGravityOverride=true
    gravityScale=0.35
    simulationSpace=RootRelative
    boundBone {boneName=dyn_necklace_02}
    componentAppliedLinearAccClamp=<2000 2000 2000>
    componentLinearAccScale=<0.4 0.4 0.4>
    componentLinearVelScale=<0.4 0.4 0.4>
    gravityOverride {z=-980}
    physicsBodyDefinitions [{
      boundBone {boneName=dyn_necklace_02}
      boxExtents=<30 20 20>
      constraintSetup {angularLimitsMax=<0 0 15>, angularLimitsMin=<0 -15 -15>, angularTarget=<1 0 0>}
      localJointOffset=<2 0 0>
     }]
   }
   n31 = Anim Dynamics - Bone: dyn_necklace_03!(ComponentPose=@n30, Alpha=1, GravityOverride:Vector="0,0,-980", ExternalForce="0,0,0") -> (Pose) {
    angularDampingOverride=0.9999
    angularSpringConstant=16000
    bAngularSpring=true
    bOverrideAngularDamping=true
    bUseGravityOverride=true
    gravityScale=0.35
    simulationSpace=RootRelative
    boundBone {boneName=dyn_necklace_03}
    componentAppliedLinearAccClamp=<2000 2000 2000>
    componentLinearAccScale=<0.4 0.4 0.4>
    componentLinearVelScale=<0.4 0.4 0.4>
    gravityOverride {z=-980}
    physicsBodyDefinitions [{
      boundBone {boneName=dyn_necklace_03}
      boxExtents=<30 30 30>
      constraintSetup {angularLimitsMax=<0 0 20>, angularLimitsMin=<0 -15 -20>, angularTarget=<1 0 0>}
      localJointOffset=<3 0 0>
     }]
   }
   Control Rig!(
    Source=Component To Local(ComponentPose=@n31)
    Alpha
    BaseballSlide:double=Get isBaseballSlidling float() -> (isBaseballSlidling float:double)
    Kneeslide:double=Get isKneeSlidling float() -> (isKneeSlidling float:double)
    CrouchingNotMove:double=Get CrouchNotMove() -> (CrouchNotMove:double)
    CrouchWalk:double=Get CrouchWalk() -> (CrouchWalk:double)
   ) -> (Pose) {
    alphaBoolBlend {blendOption=Linear}
    controlRigAssetReference {blueprintRigClass=&/F_MED_MissMissile_PostSim_CtrlRig.F_MED_MissMissile_PostSim_CtrlRig_C}
   }
   n32 = Control Rig!(
    Source=Copy Pose From Mesh() {bUseAttachedParent=true}
    Alpha
    Crouching:double=Get IsCrouching()
    Playing Emote:double=Get Playing Emote()
    MimeEmote:bool=Get isMimeEmote() -> (isMimeEmote:bool)
   ) -> (Pose) {controlRigAssetReference {blueprintRigClass=&F_MED_MissMissile_CtrlRig_C}}
   n33 = RigidBody(ComponentPose=Local To Component(LocalPose=@n32), Alpha=Get Dyn_Alpha(), OverrideWorldGravity="0,0,0") {
    bOverrideWorldGravity=true
    overridePhysicsAsset=&F_MED_MissMissile_Physics
    simulationSpace=BaseBoneSpace
    baseBoneRef {boneName=pelvis}
    componentAppliedLinearAccClamp=<2000 2000 2000>
    simSpaceSettings {maxAngularAcceleration=100, maxAngularVelocity=2, maxLinearAcceleration=100, maxLinearVelocity=160, worldAlpha=0.7}
   }
   n34 = Anim Dynamics - Bone: dyn_blouse_fr_01_l(ComponentPose=@n33, Alpha=1, GravityOverride="0,0,-980", ExternalForce="0,0,0") {
    angularDampingOverride=1
    angularSpringConstant=2500
    bAngularSpring=true
    bOverrideAngularDamping=true
    bUseGravityOverride=true
    gravityScale=0.35
    simulationSpace=RootRelative
    boundBone {boneName=dyn_blouse_fr_01_l}
    componentAppliedLinearAccClamp=<4000 4000 4000>
    componentLinearAccScale=<0.6 0.6 0.6>
    componentLinearVelScale=<0.6 0.6 0.6>
    gravityOverride {z=-980}
    physicsBodyDefinitions [{
      boundBone {boneName=dyn_blouse_fr_01_l}
      boxExtents=<20 30 30>
      constraintSetup {angularLimitsMax=<0 0 2>, angularLimitsMin=<0 -20 -2>, angularTarget=<1 0 0>}
      localJointOffset=<4 0 0>
     }]
   }
   n35 = Anim Dynamics - Bone: dyn_blouse_fr_02_l(ComponentPose=@n34, Alpha=1, GravityOverride="0,0,-980", ExternalForce="0,0,0") {
    angularDampingOverride=1
    angularSpringConstant=2500
    bAngularSpring=true
    bOverrideAngularDamping=true
    bUseGravityOverride=true
    gravityScale=0.35
    simulationSpace=RootRelative
    boundBone {boneName=dyn_blouse_fr_02_l}
    componentAppliedLinearAccClamp=<4000 4000 4000>
    componentLinearAccScale=<0.6 0.6 0.6>
    componentLinearVelScale=<0.6 0.6 0.6>
    gravityOverride {z=-980}
    physicsBodyDefinitions [{
      boundBone {boneName=dyn_blouse_fr_02_l}
      boxExtents=<20 30 30>
      constraintSetup {angularLimitsMax=<0 0 2>, angularLimitsMin=<0 -20 -2>, angularTarget=<1 0 0>}
      localJointOffset=<4 0 0>
     }]
   }
   n36 = Anim Dynamics - Bone: dyn_blouse_fr_03_l(ComponentPose=@n35, Alpha=1, GravityOverride="0,0,-980", ExternalForce="0,0,0") {
    angularDampingOverride=1
    angularSpringConstant=2500
    bAngularSpring=true
    bOverrideAngularDamping=true
    bUseGravityOverride=true
    gravityScale=0.35
    simulationSpace=RootRelative
    boundBone {boneName=dyn_blouse_fr_03_l}
    componentAppliedLinearAccClamp=<4000 4000 4000>
    componentLinearAccScale=<0.6 0.6 0.6>
    componentLinearVelScale=<0.6 0.6 0.6>
    gravityOverride {z=-980}
    physicsBodyDefinitions [{
      boundBone {boneName=dyn_blouse_fr_03_l}
      boxExtents=<20 30 30>
      constraintSetup {angularLimitsMax=<0 0 2>, angularLimitsMin=<0 -20 -2>, angularTarget=<1 0 0>}
      localJointOffset=<4 0 0>
     }]
   }
   n37 = Anim Dynamics - Bone: dyn_blouse_side_01_l(ComponentPose=@n36, Alpha=1, GravityOverride="0,0,-980", ExternalForce="0,0,0") {
    angularDampingOverride=1
    angularSpringConstant=2500
    bAngularSpring=true
    bOverrideAngularDamping=true
    bUseGravityOverride=true
    gravityScale=0.35
    simulationSpace=RootRelative
    boundBone {boneName=dyn_blouse_side_01_l}
    componentAppliedLinearAccClamp=<4000 4000 4000>
    componentLinearAccScale=<0.6 0.6 0.6>
    componentLinearVelScale=<0.6 0.6 0.6>
    gravityOverride {z=-980}
    physicsBodyDefinitions [{
      boundBone {boneName=dyn_blouse_side_01_l}
      boxExtents=<20 30 30>
      constraintSetup {angularLimitsMax=<0 0 2>, angularLimitsMin=<0 -20 -2>, angularTarget=<1 0 0>}
      localJointOffset=<4 0 0>
     }]
   }
   n38 = Anim Dynamics - Bone: dyn_blouse_side_02_l(ComponentPose=@n37, Alpha=1, GravityOverride="0,0,-980", ExternalForce="0,0,0") {
    angularDampingOverride=1
    angularSpringConstant=2500
    bAngularSpring=true
    bOverrideAngularDamping=true
    bUseGravityOverride=true
    gravityScale=0.35
    simulationSpace=RootRelative
    boundBone {boneName=dyn_blouse_side_02_l}
    componentAppliedLinearAccClamp=<4000 4000 4000>
    componentLinearAccScale=<0.6 0.6 0.6>
    componentLinearVelScale=<0.6 0.6 0.6>
    gravityOverride {z=-980}
    physicsBodyDefinitions [{
      boundBone {boneName=dyn_blouse_side_02_l}
      boxExtents=<20 30 30>
      constraintSetup {angularLimitsMax=<0 0 2>, angularLimitsMin=<0 -20 -2>, angularTarget=<1 0 0>}
      localJointOffset=<4 0 0>
     }]
   }
   n39 = Anim Dynamics - Bone: dyn_blouse_side_03_l(ComponentPose=@n38, Alpha=1, GravityOverride="0,0,-980", ExternalForce="0,0,0") {
    angularDampingOverride=1
    angularSpringConstant=2500
    bAngularSpring=true
    bOverrideAngularDamping=true
    bUseGravityOverride=true
    gravityScale=0.35
    simulationSpace=RootRelative
    boundBone {boneName=dyn_blouse_side_03_l}
    componentAppliedLinearAccClamp=<4000 4000 4000>
    componentLinearAccScale=<0.6 0.6 0.6>
    componentLinearVelScale=<0.6 0.6 0.6>
    gravityOverride {z=-980}
    physicsBodyDefinitions [{
      boundBone {boneName=dyn_blouse_side_03_l}
      boxExtents=<20 30 30>
      constraintSetup {angularLimitsMax=<0 0 2>, angularLimitsMin=<0 -20 -2>, angularTarget=<1 0 0>}
      localJointOffset=<4 0 0>
     }]
   }
   n40 = Anim Dynamics - Bone: dyn_blouse_bk_01_l(ComponentPose=@n39, Alpha=1, GravityOverride="0,0,-980", ExternalForce="0,0,0") {
    angularDampingOverride=1
    angularSpringConstant=2500
    bAngularSpring=true
    bOverrideAngularDamping=true
    bUseGravityOverride=true
    gravityScale=0.35
    simulationSpace=RootRelative
    boundBone {boneName=dyn_blouse_bk_01_l}
    componentAppliedLinearAccClamp=<4000 4000 4000>
    componentLinearAccScale=<0.6 0.6 0.6>
    componentLinearVelScale=<0.6 0.6 0.6>
    gravityOverride {z=-980}
    physicsBodyDefinitions [{
      boundBone {boneName=dyn_blouse_bk_01_l}
      boxExtents=<20 30 30>
      constraintSetup {angularLimitsMax=<0 0 2>, angularLimitsMin=<0 -20 -2>, angularTarget=<1 0 0>}
      localJointOffset=<4 0 0>
     }]
   }
   n41 = Anim Dynamics - Bone: dyn_blouse_bk_02_l(ComponentPose=@n40, Alpha=1, GravityOverride="0,0,-980", ExternalForce="0,0,0") {
    angularDampingOverride=1
    angularSpringConstant=2500
    bAngularSpring=true
    bOverrideAngularDamping=true
    bUseGravityOverride=true
    gravityScale=0.35
    simulationSpace=RootRelative
    boundBone {boneName=dyn_blouse_bk_02_l}
    componentAppliedLinearAccClamp=<4000 4000 4000>
    componentLinearAccScale=<0.6 0.6 0.6>
    componentLinearVelScale=<0.6 0.6 0.6>
    gravityOverride {z=-980}
    physicsBodyDefinitions [{
      boundBone {boneName=dyn_blouse_bk_02_l}
      boxExtents=<20 30 30>
      constraintSetup {angularLimitsMax=<0 0 2>, angularLimitsMin=<0 -20 -2>, angularTarget=<1 0 0>}
      localJointOffset=<4 0 0>
     }]
   }
   n42 = Anim Dynamics - Bone: dyn_blouse_bk_03_l(ComponentPose=@n41, Alpha=1, GravityOverride="0,0,-980", ExternalForce="0,0,0") {
    angularDampingOverride=1
    angularSpringConstant=2500
    bAngularSpring=true
    bOverrideAngularDamping=true
    bUseGravityOverride=true
    gravityScale=0.35
    simulationSpace=RootRelative
    boundBone {boneName=dyn_blouse_bk_03_l}
    componentAppliedLinearAccClamp=<4000 4000 4000>
    componentLinearAccScale=<0.6 0.6 0.6>
    componentLinearVelScale=<0.6 0.6 0.6>
    gravityOverride {z=-980}
    physicsBodyDefinitions [{
      boundBone {boneName=dyn_blouse_bk_03_l}
      boxExtents=<20 30 30>
      constraintSetup {angularLimitsMax=<0 0 2>, angularLimitsMin=<0 -20 -2>, angularTarget=<1 0 0>}
      localJointOffset=<4 0 0>
     }]
   }
   n43 = Anim Dynamics - Bone: dyn_blouse_01_c(ComponentPose=@n42, Alpha=1, GravityOverride="0,0,-980", ExternalForce="0,0,0") {
    angularDampingOverride=1
    angularSpringConstant=2500
    bAngularSpring=true
    bOverrideAngularDamping=true
    bUseGravityOverride=true
    gravityScale=0.35
    simulationSpace=RootRelative
    boundBone {boneName=dyn_blouse_01_c}
    componentAppliedLinearAccClamp=<4000 4000 4000>
    componentLinearAccScale=<0.6 0.6 0.6>
    componentLinearVelScale=<0.6 0.6 0.6>
    gravityOverride {z=-980}
    physicsBodyDefinitions [{
      boundBone {boneName=dyn_blouse_01_c}
      boxExtents=<20 30 30>
      constraintSetup {angularLimitsMax=<0 0 2>, angularLimitsMin=<0 -20 -2>, angularTarget=<1 0 0>}
      localJointOffset=<4 0 0>
     }]
   }
   n44 = Anim Dynamics - Bone: dyn_blouse_02_c(ComponentPose=@n43, Alpha=1, GravityOverride="0,0,-980", ExternalForce="0,0,0") {
    angularDampingOverride=1
    angularSpringConstant=2500
    bAngularSpring=true
    bOverrideAngularDamping=true
    bUseGravityOverride=true
    gravityScale=0.35
    simulationSpace=RootRelative
    boundBone {boneName=dyn_blouse_02_c}
    componentAppliedLinearAccClamp=<4000 4000 4000>
    componentLinearAccScale=<0.6 0.6 0.6>
    componentLinearVelScale=<0.6 0.6 0.6>
    gravityOverride {z=-980}
    physicsBodyDefinitions [{
      boundBone {boneName=dyn_blouse_02_c}
      boxExtents=<20 30 30>
      constraintSetup {angularLimitsMax=<0 0 2>, angularLimitsMin=<0 -20 -2>, angularTarget=<1 0 0>}
      localJointOffset=<4 0 0>
     }]
   }
   n45 = Anim Dynamics - Bone: dyn_blouse_03_c(ComponentPose=@n44, Alpha=1, GravityOverride="0,0,-980", ExternalForce="0,0,0") {
    angularDampingOverride=1
    angularSpringConstant=2500
    bAngularSpring=true
    bOverrideAngularDamping=true
    bUseGravityOverride=true
    gravityScale=0.35
    simulationSpace=RootRelative
    boundBone {boneName=dyn_blouse_03_c}
    componentAppliedLinearAccClamp=<4000 4000 4000>
    componentLinearAccScale=<0.6 0.6 0.6>
    componentLinearVelScale=<0.6 0.6 0.6>
    gravityOverride {z=-980}
    physicsBodyDefinitions [{
      boundBone {boneName=dyn_blouse_03_c}
      boxExtents=<20 30 30>
      constraintSetup {angularLimitsMax=<0 0 2>, angularLimitsMin=<0 -20 -2>, angularTarget=<1 0 0>}
      localJointOffset=<4 0 0>
     }]
   }
   n46 = Anim Dynamics - Bone: dyn_blouse_bk_01_r(ComponentPose=@n45, Alpha=1, GravityOverride="0,0,-980", ExternalForce="0,0,0") {
    angularDampingOverride=1
    angularSpringConstant=2500
    bAngularSpring=true
    bOverrideAngularDamping=true
    bUseGravityOverride=true
    gravityScale=0.35
    simulationSpace=RootRelative
    boundBone {boneName=dyn_blouse_bk_01_r}
    componentAppliedLinearAccClamp=<4000 4000 4000>
    componentLinearAccScale=<0.6 0.6 0.6>
    componentLinearVelScale=<0.6 0.6 0.6>
    gravityOverride {z=-980}
    physicsBodyDefinitions [{
      boundBone {boneName=dyn_blouse_bk_01_r}
      boxExtents=<20 30 30>
      constraintSetup {angularLimitsMax=<0 0 2>, angularLimitsMin=<0 -20 -2>, angularTarget=<1 0 0>}
      localJointOffset=<4 0 0>
     }]
   }
   n47 = Anim Dynamics - Bone: dyn_blouse_bk_02_r(ComponentPose=@n46, Alpha=1, GravityOverride="0,0,-980", ExternalForce="0,0,0") {
    angularDampingOverride=1
    angularSpringConstant=2500
    bAngularSpring=true
    bOverrideAngularDamping=true
    bUseGravityOverride=true
    gravityScale=0.35
    simulationSpace=RootRelative
    boundBone {boneName=dyn_blouse_bk_02_r}
    componentAppliedLinearAccClamp=<4000 4000 4000>
    componentLinearAccScale=<0.6 0.6 0.6>
    componentLinearVelScale=<0.6 0.6 0.6>
    gravityOverride {z=-980}
    physicsBodyDefinitions [{
      boundBone {boneName=dyn_blouse_bk_02_r}
      boxExtents=<20 30 30>
      constraintSetup {angularLimitsMax=<0 0 2>, angularLimitsMin=<0 -20 -2>, angularTarget=<1 0 0>}
      localJointOffset=<4 0 0>
     }]
   }
   n48 = Anim Dynamics - Bone: dyn_blouse_bk_03_r(ComponentPose=@n47, Alpha=1, GravityOverride="0,0,-980", ExternalForce="0,0,0") {
    angularDampingOverride=1
    angularSpringConstant=2500
    bAngularSpring=true
    bOverrideAngularDamping=true
    bUseGravityOverride=true
    gravityScale=0.35
    simulationSpace=RootRelative
    boundBone {boneName=dyn_blouse_bk_03_r}
    componentAppliedLinearAccClamp=<4000 4000 4000>
    componentLinearAccScale=<0.6 0.6 0.6>
    componentLinearVelScale=<0.6 0.6 0.6>
    gravityOverride {z=-980}
    physicsBodyDefinitions [{
      boundBone {boneName=dyn_blouse_bk_03_r}
      boxExtents=<20 30 30>
      constraintSetup {angularLimitsMax=<0 0 2>, angularLimitsMin=<0 -20 -2>, angularTarget=<1 0 0>}
      localJointOffset=<4 0 0>
     }]
   }
   n49 = Anim Dynamics - Bone: dyn_blouse_side_01_r(ComponentPose=@n48, Alpha=1, GravityOverride="0,0,-980", ExternalForce="0,0,0") {
    angularDampingOverride=1
    angularSpringConstant=2500
    bAngularSpring=true
    bOverrideAngularDamping=true
    bUseGravityOverride=true
    gravityScale=0.35
    simulationSpace=RootRelative
    boundBone {boneName=dyn_blouse_side_01_r}
    componentAppliedLinearAccClamp=<4000 4000 4000>
    componentLinearAccScale=<0.6 0.6 0.6>
    componentLinearVelScale=<0.6 0.6 0.6>
    gravityOverride {z=-980}
    physicsBodyDefinitions [{
      boundBone {boneName=dyn_blouse_side_01_r}
      boxExtents=<20 30 30>
      constraintSetup {angularLimitsMax=<0 0 2>, angularLimitsMin=<0 -20 -2>, angularTarget=<1 0 0>}
      localJointOffset=<4 0 0>
     }]
   }
   n50 = Anim Dynamics - Bone: dyn_blouse_side_02_r(ComponentPose=@n49, Alpha=1, GravityOverride="0,0,-980", ExternalForce="0,0,0") {
    angularDampingOverride=1
    angularSpringConstant=2500
    bAngularSpring=true
    bOverrideAngularDamping=true
    bUseGravityOverride=true
    gravityScale=0.35
    simulationSpace=RootRelative
    boundBone {boneName=dyn_blouse_side_02_r}
    componentAppliedLinearAccClamp=<4000 4000 4000>
    componentLinearAccScale=<0.6 0.6 0.6>
    componentLinearVelScale=<0.6 0.6 0.6>
    gravityOverride {z=-980}
    physicsBodyDefinitions [{
      boundBone {boneName=dyn_blouse_side_02_r}
      boxExtents=<20 30 30>
      constraintSetup {angularLimitsMax=<0 0 2>, angularLimitsMin=<0 -20 -2>, angularTarget=<1 0 0>}
      localJointOffset=<4 0 0>
     }]
   }
   n51 = Anim Dynamics - Bone: dyn_blouse_side_03_r(ComponentPose=@n50, Alpha=1, GravityOverride="0,0,-980", ExternalForce="0,0,0") {
    angularDampingOverride=1
    angularSpringConstant=2500
    bAngularSpring=true
    bOverrideAngularDamping=true
    bUseGravityOverride=true
    gravityScale=0.35
    simulationSpace=RootRelative
    boundBone {boneName=dyn_blouse_side_03_r}
    componentAppliedLinearAccClamp=<4000 4000 4000>
    componentLinearAccScale=<0.6 0.6 0.6>
    componentLinearVelScale=<0.6 0.6 0.6>
    gravityOverride {z=-980}
    physicsBodyDefinitions [{
      boundBone {boneName=dyn_blouse_side_03_r}
      boxExtents=<20 30 30>
      constraintSetup {angularLimitsMax=<0 0 2>, angularLimitsMin=<0 -20 -2>, angularTarget=<1 0 0>}
      localJointOffset=<4 0 0>
     }]
   }
   n52 = Anim Dynamics - Bone: dyn_blouse_fr_01_r(ComponentPose=@n51, Alpha=1, GravityOverride="0,0,-980", ExternalForce="0,0,0") {
    angularDampingOverride=1
    angularSpringConstant=2500
    bAngularSpring=true
    bOverrideAngularDamping=true
    bUseGravityOverride=true
    gravityScale=0.35
    simulationSpace=RootRelative
    boundBone {boneName=dyn_blouse_fr_01_r}
    componentAppliedLinearAccClamp=<4000 4000 4000>
    componentLinearAccScale=<0.6 0.6 0.6>
    componentLinearVelScale=<0.6 0.6 0.6>
    gravityOverride {z=-980}
    physicsBodyDefinitions [{
      boundBone {boneName=dyn_blouse_fr_01_r}
      boxExtents=<20 30 30>
      constraintSetup {angularLimitsMax=<0 0 2>, angularLimitsMin=<0 -20 -2>, angularTarget=<1 0 0>}
      localJointOffset=<4 0 0>
     }]
   }
   n53 = Anim Dynamics - Bone: dyn_blouse_fr_02_r(ComponentPose=@n52, Alpha=1, GravityOverride="0,0,-980", ExternalForce="0,0,0") {
    angularDampingOverride=1
    angularSpringConstant=2500
    bAngularSpring=true
    bOverrideAngularDamping=true
    bUseGravityOverride=true
    gravityScale=0.35
    simulationSpace=RootRelative
    boundBone {boneName=dyn_blouse_fr_02_r}
    componentAppliedLinearAccClamp=<4000 4000 4000>
    componentLinearAccScale=<0.6 0.6 0.6>
    componentLinearVelScale=<0.6 0.6 0.6>
    gravityOverride {z=-980}
    physicsBodyDefinitions [{
      boundBone {boneName=dyn_blouse_fr_02_r}
      boxExtents=<20 30 30>
      constraintSetup {angularLimitsMax=<0 0 2>, angularLimitsMin=<0 -20 -2>, angularTarget=<1 0 0>}
      localJointOffset=<4 0 0>
     }]
   }
   n54 = Anim Dynamics - Bone: dyn_blouse_fr_03_r(ComponentPose=@n53, Alpha=1, GravityOverride="0,0,-980", ExternalForce="0,0,0") {
    angularDampingOverride=1
    angularSpringConstant=2500
    bAngularSpring=true
    bOverrideAngularDamping=true
    bUseGravityOverride=true
    gravityScale=0.35
    simulationSpace=RootRelative
    boundBone {boneName=dyn_blouse_fr_03_r}
    componentAppliedLinearAccClamp=<4000 4000 4000>
    componentLinearAccScale=<0.6 0.6 0.6>
    componentLinearVelScale=<0.6 0.6 0.6>
    gravityOverride {z=-980}
    physicsBodyDefinitions [{
      boundBone {boneName=dyn_blouse_fr_03_r}
      boxExtents=<20 30 30>
      constraintSetup {angularLimitsMax=<0 0 2>, angularLimitsMin=<0 -20 -2>, angularTarget=<1 0 0>}
      localJointOffset=<4 0 0>
     }]
   }
   n55 = "Transform (Modify) Bone - Bone: dyn_blouse_bk_01_l"(ComponentPose=@n54, Alpha=1, Translation="0,0,0", Rotation="0,0,0", Scale="1,1,1") {rotationMode=BMM_Additive, boneToModify {boneName=dyn_blouse_bk_01_l}}
   n56 = "Transform (Modify) Bone - Bone: dyn_blouse_01_c"(
    ComponentPose="Transform (Modify) Bone - Bone: dyn_blouse_bk_03_l"(
     ComponentPose="Transform (Modify) Bone - Bone: dyn_blouse_bk_02_l"(ComponentPose=@n55, Alpha=1, Translation="0,0,0", Rotation="0,0,0", Scale="1,1,1") {rotationMode=BMM_Additive, boneToModify {boneName=dyn_blouse_bk_02_l}}
     Alpha=1
     Translation="0,0,0"
     Rotation="0,0,0"
     Scale="1,1,1"
    ) {rotationMode=BMM_Additive, boneToModify {boneName=dyn_blouse_bk_03_l}}
    Alpha=1
    Translation="0,0,0"
    Rotation="0,0,0"
    Scale="1,1,1"
   ) {rotationMode=BMM_Additive, boneToModify {boneName=dyn_blouse_01_c}}
   n57 = "Transform (Modify) Bone - Bone: dyn_blouse_bk_01_r"(
    ComponentPose="Transform (Modify) Bone - Bone: dyn_blouse_03_c"(
     ComponentPose="Transform (Modify) Bone - Bone: dyn_blouse_02_c"(ComponentPose=@n56, Alpha=1, Translation="0,0,0", Rotation="0,0,0", Scale="1,1,1") {rotationMode=BMM_Additive, boneToModify {boneName=dyn_blouse_02_c}}
     Alpha=1
     Translation="0,0,0"
     Rotation="0,0,0"
     Scale="1,1,1"
    ) {rotationMode=BMM_Additive, boneToModify {boneName=dyn_blouse_03_c}}
    Alpha=1
    Translation="0,0,0"
    Rotation="0,0,0"
    Scale="1,1,1"
   ) {rotationMode=BMM_Additive, boneToModify {boneName=dyn_blouse_bk_01_r}}
   n58 = "Transform (Modify) Bone - Bone: dyn_blouse_side_01_r"(
    ComponentPose="Transform (Modify) Bone - Bone: dyn_blouse_bk_03_r"(
     ComponentPose="Transform (Modify) Bone - Bone: dyn_blouse_bk_02_r"(ComponentPose=@n57, Alpha=1, Translation="0,0,0", Rotation="0,0,0", Scale="1,1,1") {rotationMode=BMM_Additive, boneToModify {boneName=dyn_blouse_bk_02_r}}
     Alpha=1
     Translation="0,0,0"
     Rotation="0,0,0"
     Scale="1,1,1"
    ) {rotationMode=BMM_Additive, boneToModify {boneName=dyn_blouse_bk_03_r}}
    Alpha=1
    Translation="0,0,0"
    Rotation="0,0,0"
    Scale="1,1,1"
   ) {rotationMode=BMM_Additive, boneToModify {boneName=dyn_blouse_side_01_r}}
   n59 = "Transform (Modify) Bone - Bone: dyn_blouse_side_01_l"(
    ComponentPose="Transform (Modify) Bone - Bone: dyn_blouse_side_03_r"(
     ComponentPose="Transform (Modify) Bone - Bone: dyn_blouse_side_02_r"(ComponentPose=@n58, Alpha=1, Translation="0,0,0", Rotation="0,0,0", Scale="1,1,1") {rotationMode=BMM_Additive, boneToModify {boneName=dyn_blouse_side_02_r}}
     Alpha=1
     Translation="0,0,0"
     Rotation="0,0,0"
     Scale="1,1,1"
    ) {rotationMode=BMM_Additive, boneToModify {boneName=dyn_blouse_side_03_r}}
    Alpha=1
    Translation="0,0,0"
    Rotation="0,0,0"
    Scale="1,1,1"
   ) {rotationMode=BMM_Additive, boneToModify {boneName=dyn_blouse_side_01_l}}
   n60 = Anim Dynamics - Bone: dyn_necklace_01(
    ComponentPose="Transform (Modify) Bone - Bone: dyn_blouse_side_03_l"(
     ComponentPose="Transform (Modify) Bone - Bone: dyn_blouse_side_02_l"(ComponentPose=@n59, Alpha=1, Translation="0,0,0", Rotation="0,0,0", Scale="1,1,1") {rotationMode=BMM_Additive, boneToModify {boneName=dyn_blouse_side_02_l}}
     Alpha=1
     Translation="0,0,0"
     Rotation="0,0,0"
     Scale="1,1,1"
    ) {rotationMode=BMM_Additive, boneToModify {boneName=dyn_blouse_side_03_l}}
    Alpha=1
    GravityOverride="0,0,-980"
    ExternalForce="0,0,0"
   ) {
    angularDampingOverride=0.9999
    angularSpringConstant=10000
    bAngularSpring=true
    bOverrideAngularDamping=true
    bUseGravityOverride=true
    gravityScale=0.35
    simulationSpace=RootRelative
    boundBone {boneName=dyn_necklace_01}
    componentAppliedLinearAccClamp=<2000 2000 2000>
    componentLinearAccScale=<0.4 0.4 0.4>
    componentLinearVelScale=<0.4 0.4 0.4>
    gravityOverride {z=-980}
    physicsBodyDefinitions [{
      boundBone {boneName=dyn_necklace_01}
      boxExtents=<30 20 20>
      constraintSetup {angularLimitsMax=<0 -2 10>, angularLimitsMin=<0 -5 -10>, angularTarget=<1 0 0>}
      localJointOffset=<2 0 0>
     }]
   }
   n61 = Anim Dynamics - Bone: dyn_necklace_02(ComponentPose=@n60, Alpha=1, GravityOverride="0,0,-980", ExternalForce="0,0,0") {
    angularDampingOverride=0.9999
    angularSpringConstant=10000
    bAngularSpring=true
    bOverrideAngularDamping=true
    bUseGravityOverride=true
    gravityScale=0.35
    simulationSpace=RootRelative
    boundBone {boneName=dyn_necklace_02}
    componentAppliedLinearAccClamp=<2000 2000 2000>
    componentLinearAccScale=<0.4 0.4 0.4>
    componentLinearVelScale=<0.4 0.4 0.4>
    gravityOverride {z=-980}
    physicsBodyDefinitions [{
      boundBone {boneName=dyn_necklace_02}
      boxExtents=<30 20 20>
      constraintSetup {angularLimitsMax=<0 0 15>, angularLimitsMin=<0 -15 -15>, angularTarget=<1 0 0>}
      localJointOffset=<2 0 0>
     }]
   }
   n62 = Anim Dynamics - Bone: dyn_necklace_03(ComponentPose=@n61, Alpha=1, GravityOverride="0,0,-980", ExternalForce="0,0,0") {
    angularDampingOverride=0.9999
    angularSpringConstant=16000
    bAngularSpring=true
    bOverrideAngularDamping=true
    bUseGravityOverride=true
    gravityScale=0.35
    simulationSpace=RootRelative
    boundBone {boneName=dyn_necklace_03}
    componentAppliedLinearAccClamp=<2000 2000 2000>
    componentLinearAccScale=<0.4 0.4 0.4>
    componentLinearVelScale=<0.4 0.4 0.4>
    gravityOverride {z=-980}
    physicsBodyDefinitions [{
      boundBone {boneName=dyn_necklace_03}
      boxExtents=<30 30 30>
      constraintSetup {angularLimitsMax=<0 0 20>, angularLimitsMin=<0 -15 -20>, angularTarget=<1 0 0>}
      localJointOffset=<3 0 0>
     }]
   }
   n63 = Control Rig!(
    Source=Component To Local(ComponentPose=@n62)
    Alpha
    BaseballSlide:double=Get isBaseballSlidling float()
    Kneeslide:double=Get isKneeSlidling float()
    CrouchingNotMove:double=Get CrouchNotMove()
    CrouchWalk:double=Get CrouchWalk()
    PickaxeSwing:double=Get PickaxeSwing() -> (PickaxeSwing:double)
   ) -> (Pose) {
    alphaBoolBlend {blendOption=Linear}
    controlRigAssetReference {blueprintRigClass=&F_MED_MissMissile_PostSim_CtrlRig_C}
   }
   Output Pose(Result=@n63) {layerGroup=None}
  ]}
 EventGraph {nodes [
   n1 = Event BlueprintUpdateAnimation()
   > Set Delta Time X(Delta Time X=@n1.DeltaTimeX)
   > n2 = Sequence!(execute) -> (then_0, then_1, then_2:Exec, then_3:Exec, then_4:Exec, then_5:Exec, then_6:Exec, then_7:Exec, then_8:Exec, then_9:Exec)
   > n3 = "Gravity Override State Function (Parameter Asset) (Nativized)"(
    CharacterPartAnimInstance=Self-Reference()
    GameTimeInSeconds=GetGameTimeInSeconds()
    Parameters=&/F_MED_MissMissile_Parameters
   )
   > Set Gravity Override(Gravity Override=@n3)
   > Set RBAN Settings Out(RBAN Settings Out=@n3.OutRigidBodyAnimNodeParameters)
   > n4 = Branch(Condition=Get bIsPlayingMeleeAnim())
   > Set PickaxeSwing(PickaxeSwing:double=1) -> (Output_Get:double)
   Set PickaxeSwing(execute=@n4.else, PickaxeSwing=0)
   n5 = GetOwnerSkeletalMeshComponent()
   Set Spine_3 rotation(
    execute=@n2.then_1
    Spine_3 rotation:double=MapRangeClamped(
     Value=GetDeltaTransformFromRefPose!(self=@n5, BoneName=spine_04, BaseName=spine_02) -> (
      ReturnValue_Location:Vector
      ReturnValue_Rotation_Roll:float
      ReturnValue_Rotation_Pitch:float
      ReturnValue_Rotation_Yaw:float
      ReturnValue_Scale:Vector
     ).ReturnValue_Rotation_Yaw
     InRangeA=-1
     InRangeB=0
     OutRangeA=-0.2
    )
   ) -> (Output_Get:double)
   > Set is skydive(is skydive:double=SelectFloat(A=0.1, B=0.5, bPickA=Get bIsSkyDiving())) -> (Output_Get:double)
   Branch(execute=@n2.then_3, Condition=Get bIsInFrontEnd())
   > Set FrontEnd Idle Bypass(FrontEnd Idle Bypass:bool=NOT Boolean(A=Get bIsPlayingEmote())) -> (Output_Get:bool)
   n6 = Branch(execute=@n2.then_4, Condition=OR Boolean!(A=Get bIsSprinting(), B=Get bIsTacticalSprinting(), C:bool=Get bIsSkyDiving()) -> (ReturnValue))
   > Set SkydiveDyn(SkydiveDyn:double=0.2) -> (Output_Get:double)
   Set SkydiveDyn(execute=@n6.else, SkydiveDyn=1)
   > n7 = Branch(Condition=Get bIsPlayingEmote())
   > Set SkydiveDyn(SkydiveDyn=0.6)
   Set SkydiveDyn(execute=@n7.else, SkydiveDyn=1)
   n8 = Get bIsCrouching()
   Set IsCrouching(
    execute=@n2.then_8
    IsCrouching:double=FInterpTo(Current=Get IsCrouching(), Target="To Float (Boolean)"(InBool=@n8), DeltaTime=Get Delta Time X(), InterpSpeed=15)
   ) -> (Output_Get:double)
   n9 = AND Boolean(A=Get bIsCrouching(), B=NOT Boolean(A=Get bIsMoving2D()))
   > Set CrouchNotMove(CrouchNotMove:double=FInterpTo(Current=Get CrouchNotMove(), Target="To Float (Boolean)"(InBool=@n9), DeltaTime=Get Delta Time X(), InterpSpeed=15)) -> (Output_Get:double)
   > n10 = Branch(Condition=OR Boolean!(A=Get bIsPlayingEmote(), B=Get bIsSkyDiving(), C=Get bIsBaseballSliding(), D:bool=Get bIsKneeSliding()) -> (ReturnValue))
   > Set Dyn_Alpha(Dyn_Alpha:double=0) -> (Output_Get:double)
   Set Dyn_Alpha(execute=@n10.else, Dyn_Alpha=1)
   > n11 = Branch(Condition=Get bIsCrouching())
   > Set Dyn_Alpha(Dyn_Alpha=1)
   Set Dyn_Alpha(execute=@n11.else, Dyn_Alpha=0.4)
   n12 = CheckForBackbling()
   n13 = Set IsWearingBackbling(execute=@n2.then_2, IsWearingBackbling:bool=@n12) -> (Output_Get:bool)
   > Set Is Backbling A Cape(Is Backbling A Cape:bool=@n12.bIsBackblingACape) -> (Output_Get:bool)
   > Set NotBackbling(NotBackbling:double="To Float (Boolean)"(InBool=NOT Boolean(A=@n12))) -> (Output_Get:double)
   > Set is Wear a Cape(is Wear a Cape:bool=AND Boolean(A=@n12, B=@n12.bIsBackblingLong)) -> (Output_Get:bool)
   > n14 = Branch(Condition=@n13)
   > n15 = Branch(Condition=@n12.bIsBackblingLong)
   Set IsShortBackblingOnAlpha(execute=[@n15.then, @n14.else], IsShortBackblingOnAlpha:double=0) -> (Output_Get:double)
   Set IsShortBackblingOnAlpha(execute=@n15.else, IsShortBackblingOnAlpha=1)
   n16 = Sequence!(execute=@n2.then_7) -> (then_0, then_1, then_2)
   > n17 = Branch(Condition=OR Boolean!(A=Get bIsSkydiveDiveMode(), B=Get bIsSkydiveFloating()) -> (ReturnValue))
   > Set enableSin(enableSin:double=1) -> (Output_Get:double)
   Set enableSin(execute=@n17.else, enableSin=0)
   n18 = Branch(execute=@n16.then_1, Condition=Get bIsSkydiveDiveMode())
   > Set freqVar1(freqVar1:double=35) -> (Output_Get:double)
   > Set freqVar2(freqVar2:double=32) -> (Output_Get:double)
   > Set freqVar3(freqVar3:double=28) -> (Output_Get:double)
   Set freqVar1(execute=@n18.else, freqVar1=22)
   > Set freqVar2(freqVar2=25)
   > Set freqVar3(freqVar3=20)
   Set jacketRotVar1(
    execute=@n16.then_2
    jacketRotVar1:Rotator=Rotator Rumble!(frequency:double=Get freqVar1() -> (freqVar1:double), amplitude:double=0.8, time offset:Vector="0,1.2,0", coefficient:Vector="0,1,0") -> (rotatorOutput:Rotator=~) {is=title K2Node_MacroInstance}
   ) -> (Output_Get:Rotator)
   > Set jacketRotVar2(jacketRotVar2:Rotator=Rotator Rumble(frequency=Get freqVar2() -> (freqVar2:double), amplitude=0.4, time offset="0,1,0", coefficient="0,1,0") -> (rotatorOutput=~)) -> (Output_Get:Rotator)
   > Set JacketRotVar3(JacketRotVar3:Rotator=Rotator Rumble(frequency=Get freqVar3() -> (freqVar3:double), amplitude=0.7, time offset="0,1,0", coefficient="0,1,0") -> (rotatorOutput=~)) -> (Output_Get:Rotator)
   n19 = GetAnimInstance(self=GetOwnerSkeletalMeshComponent())
   Is Valid(exec=@n2.then_9, InputObject=@n19)
   > Set isMimeEmote(isMimeEmote:bool=Montage_IsPlaying(self=@n19, Montage=/BRCosmetics/Animation/Game/MainPlayer/Emotes/Mime/Emote_Mime_CMF_M)) -> (Output_Get:bool)
   n20 = Get bIsCrouching()
   n21 = Get bIsMoving2D()
   Set CrouchWalk(execute=@n2.then_6, CrouchWalk:double="To Float (Boolean)"(InBool=AND Boolean(A=@n21, B=@n20))) -> (Output_Get:double)
   > Set Playing Emote(Playing Emote:double="To Float (Boolean)"(InBool=Get bIsPlayingEmote())) -> (Output_Get:double)
   Event OnBeginKneeSliding() {is=title K2Node_Event}
   > Set isKneeSlidling float(isKneeSlidling float:double=1) -> (Output_Get:double)
   Event OnEndKneeSliding() {is=title K2Node_Event}
   > Set isKneeSlidling float(isKneeSlidling float=0)
   Event OnBeginBaseballSliding() {is=title K2Node_Event}
   > Set isBaseballSlidling float(isBaseballSlidling float:double=1) -> (Output_Get:double)
   Event OnEndBaseballSliding() {is=title K2Node_Event}
   > Set isBaseballSlidling float(isBaseballSlidling float=0)
   Event OnBeginSkydiving()
   > Set IsSkydive(IsSkydive:double=1) -> (Output_Get:double)
   Event OnEndSkydiving()
   > Set IsSkydive(IsSkydive=0)
   Branch()
   Reroute Node(InputPin:Wildcard) -> (OutputPin:Wildcard)
  ]}
 floatSinFunc {nodes [
   n1 = Inputs!() -> (time:double=~, time offset:double=~, frequency:double=~, amplitude:double=~, coefficient:double=~)
   n2 = float + float(A=@n1, B=@n1.time offset)
   n3 = MultiplyByPi!(Value:double="Sin (Radians)"(A=float * float(A=@n2, B=@n1.frequency))) -> (ReturnValue:double)
   Outputs!(floatOutput:double=float * float(A=float * float(A=@n3, B=@n1.amplitude), B=@n1.coefficient))
  ]}
 rotatorRumble {nodes [
   n1 = Inputs!() -> (frequency, amplitude, time offset:Vector=~, coefficient:Vector=~)
   n2 = BreakVector(InVec=@n1.coefficient)
   n3 = BreakVector(InVec=@n1.time offset)
   n4 = GetGameTimeInSeconds()
   Outputs!(rotatorOutput:Rotator=MakeRotator(
     Roll=Float Sin Func!(time:double=@n4, time offset:double=@n3, frequency:double=@n1, amplitude:double=@n1.amplitude, coefficient:double=@n2) -> (floatOutput:double=~) {is=title K2Node_MacroInstance}
     Pitch=Float Sin Func(time=@n4, time offset=@n3.Y, frequency=@n1, amplitude=@n1.amplitude, coefficient=@n2.Y) -> (floatOutput=~)
     Yaw=Float Sin Func(time=@n4, time offset=@n3.Z, frequency=@n1, amplitude=@n1.amplitude, coefficient=@n2.Z) -> (floatOutput=~)
    ))
  ]}
]
variables [
 CrouchNotMove {type=number}
 CrouchWalk {type=number}
 Delta Time X {type=number}
 Dyn_Alpha {type=number}
 enableSin {type=number}
 freqVar1 {type=number}
 freqVar2 {type=number}
 freqVar3 {type=number}
 FrontEnd Idle Bypass {type=boolean}
 Gravity Override {type=Vector}
 Is Backbling A Cape {type=boolean}
 is skydive {type=number}
 is Wear a Cape {type=boolean}
 isBaseballSlidling float {type=number}
 IsCrouching {type=number}
 isKneeSlidling float {type=number}
 isMimeEmote {type=boolean}
 IsShortBackblingOnAlpha {type=number}
 IsSkydive {type=number}
 IsWearingBackbling {type=boolean}
 jacketRotVar1 {type=Rotator}
 jacketRotVar2 {type=Rotator}
 JacketRotVar3 {type=Rotator}
 "Move Left " {type=number}
 Move Right {type=number}
 moveBackward {type=number}
 moveForward {type=number}
 NotBackbling {type=number}
 PickaxeSwing {type=number}
 Playing Emote {type=number}
 Pub Pawn Delta Rotation {type=Rotator}
 RBAN Settings Out {type=FortRigidBodyAnimNodeParameters}
 SkydiveDyn {type=number}
 Spine_3 rotation {type=number}
]
