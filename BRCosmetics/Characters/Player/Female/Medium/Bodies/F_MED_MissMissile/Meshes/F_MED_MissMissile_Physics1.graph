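# Physics asset for the MissMissile female medium body, giving capsule bodies and joints to the blouse bones around pelvis, lower spine and thigh capsules.
# Seven blouse chains (center, plus left and right front, side and back) each run from a small base body locked to the pelvis through three larger dynamic segments.
# Center and right-back chains are fully angle-locked, other segments mostly have tight angular limits, and free-rotation constraints tie lower left-side segments to left-front ones.
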PhysicsAsset /BRCosmetics/Characters/Player/Female/Medium/Bodies/F_MED_MissMissile/Meshes/F_MED_MissMissile_Physics1
bodies {
 base_blouse_01_c [{shapeType=Capsule, length=0.505, radius=0.505, shapeName=base_blouse_01_c_capsule, rotation=<rot 0 -90 0>}]
 base_blouse_bk_01_l [{shapeType=Capsule, length=0.505, radius=0.505, shapeName=base_blouse_bk_01_l_capsule, rotation=<rot 0 -90 0>}]
 base_blouse_bk_01_r [{shapeType=Capsule, length=0.505, radius=0.505, shapeName=base_blouse_bk_01_r_capsule, rotation=<rot 0 -90 0>}]
 base_blouse_fr_01_l [{shapeType=Capsule, length=0.505, radius=0.505, shapeName=base_blouse_fr_01_l_capsule, rotation=<rot 0 -90 0>}]
 base_blouse_fr_01_r [{shapeType=Capsule, length=0.505, radius=0.505, shapeName=base_blouse_fr_01_r_capsule, rotation=<rot 0 -90 0>}]
 base_blouse_side_01_l [{shapeType=Capsule, length=0.505, radius=0.505, shapeName=base_blouse_side_01_l_capsule, rotation=<rot 0 -90 0>}]
 base_blouse_side_01_r [{shapeType=Capsule, length=0.505, radius=0.505, shapeName=base_blouse_side_01_r_capsule, rotation=<rot 0 -90 0>}]
 dyn_blouse_01_c [{shapeType=Capsule, length=12, radius=6, shapeName=dyn_blouse_01_c_capsule, center=<8 0 4>, rotation=<rot 0 0 90>}]
 dyn_blouse_02_c [{shapeType=Capsule, length=12, radius=6, shapeName=dyn_blouse_02_c_capsule, center=<8 0 0.5>, rotation=<rot 0 0 90>}]
 dyn_blouse_03_c [{shapeType=Capsule, length=12, radius=6, shapeName=dyn_blouse_03_c_capsule, center=<8 0 1>, rotation=<rot 0 0 90>}]
 dyn_blouse_bk_01_l [{shapeType=Capsule, length=12, radius=6, shapeName=dyn_blouse_bk_01_l_capsule, center=<8 0 5>, rotation=<rot 0 0 90>}]
 dyn_blouse_bk_01_r [{shapeType=Capsule, length=12, radius=6, shapeName=dyn_blouse_bk_01_r_capsule, center=<8 0 5>, rotation=<rot 0 0 90>}]
 dyn_blouse_bk_02_l [{shapeType=Capsule, length=12, radius=6, shapeName=dyn_blouse_bk_02_l_capsule, center=<8 0 2>, rotation=<rot 0 0 90>}]
 dyn_blouse_bk_02_r [{shapeType=Capsule, length=12, radius=6, shapeName=dyn_blouse_bk_02_r_capsule, center=<8 0 2>, rotation=<rot 0 0 90>}]
 dyn_blouse_bk_03_l [{shapeType=Capsule, length=12, radius=6, shapeName=dyn_blouse_bk_03_l_capsule, center=<8 3.6442 2>, rotation=<rot 0 0 90>}]
 dyn_blouse_bk_03_r [{shapeType=Capsule, length=12, radius=6, shapeName=dyn_blouse_bk_03_r_capsule, center=<8 -3.5635 2>, rotation=<rot 0 0 90>}]
 dyn_blouse_fr_01_l [
  {shapeType=Capsule, length=12, radius=6, shapeName=dyn_blouse_fr_01_l_capsule, center=<8 7.0082 5.2248>, rotation=<rot 90 0 0>}
  {shapeType=Capsule, length=15, radius=6, shapeName=dyn_blouse_fr_01_l_capsule, center=<18.1244 -3.8673 4.6051>, rotation=<rot 0 0 69.7721>}
 ]
 dyn_blouse_fr_01_r [{shapeType=Capsule, length=12, radius=6, shapeName=dyn_blouse_fr_01_r_capsule, center=<8 0 5>, rotation=<rot 0 0 90>}]
 dyn_blouse_fr_02_l [{shapeType=Capsule, length=12, radius=6, shapeName=dyn_blouse_fr_02_l_capsule, center=<8 5.8356 3.5>, rotation=<rot 90 0 0>}]
 dyn_blouse_fr_02_r [{shapeType=Capsule, length=12, radius=6, shapeName=dyn_blouse_fr_02_r_capsule, center=<8 0 5>, rotation=<rot 0 0 90>}]
 dyn_blouse_fr_03_l [{shapeType=Capsule, length=12, radius=6, shapeName=dyn_blouse_fr_03_l_capsule, center=<8 5.8356 3.5>, rotation=<rot 0 0 90>}]
 dyn_blouse_fr_03_r [{shapeType=Capsule, length=12, radius=6, shapeName=dyn_blouse_fr_03_r_capsule, center=<8 0 5>, rotation=<rot 0 0 90>}]
 dyn_blouse_side_01_l [{shapeType=Capsule, length=12, radius=6, shapeName=dyn_blouse_side_01_l_capsule, center=<8 4.4256 5>, rotation=<rot 0 0 90>}]
 dyn_blouse_side_01_r [{shapeType=Capsule, length=12, radius=6, shapeName=dyn_blouse_side_01_r_capsule, center=<8 0 5>, rotation=<rot 0 0 90>}]
 dyn_blouse_side_02_l [{shapeType=Capsule, length=12, radius=6, shapeName=dyn_blouse_side_02_l_capsule, center=<8 4.7434 5>, rotation=<rot 0 0 90>}]
 dyn_blouse_side_02_r [{shapeType=Capsule, length=12, radius=6, shapeName=dyn_blouse_side_02_r_capsule, center=<8 0 5>, rotation=<rot 0 0 90>}]
 dyn_blouse_side_03_l [{shapeType=Capsule, length=12, radius=6, shapeName=dyn_blouse_side_03_l_capsule, center=<8 4.0112 5>, rotation=<rot 0 0 90>}]
 dyn_blouse_side_03_r [{shapeType=Capsule, length=12, radius=6, shapeName=dyn_blouse_side_03_r_capsule, center=<8 0 5>, rotation=<rot 0 0 90>}]
 pelvis [{shapeType=Capsule, length=11.0229, radius=11}]
 spine_01 [{shapeType=Capsule, length=10, radius=10, center=<2.1256 0 0>}]
 thigh_l [{shapeType=Capsule, length=25, radius=9, center=<-20 0 0>, rotation=<rot 90 0 0>}]
 thigh_r [{shapeType=Capsule, length=25, radius=9, center=<20 0 0>, rotation=<rot 90 180 0>}]
}
constraints [
 {
  bone1Name=spine_01
  bone2Name=pelvis
  swing1LimitDegrees=45
  swing1Motion=Locked
  swing2LimitDegrees=45
  swing2Motion=Locked
  twistLimitDegrees=45
  twistMotion=Locked
 }
 {
  bone1Name=thigh_l
  bone2Name=pelvis
  swing1LimitDegrees=45
  swing1Motion=Locked
  swing2LimitDegrees=45
  swing2Motion=Locked
  twistLimitDegrees=45
  twistMotion=Locked
 }
 {
  bone1Name=thigh_r
  bone2Name=pelvis
  swing1LimitDegrees=45
  swing1Motion=Locked
  swing2LimitDegrees=45
  swing2Motion=Locked
  twistLimitDegrees=45
  twistMotion=Locked
 }
 {
  bone1Name=base_blouse_01_c
  bone2Name=pelvis
  swing1LimitDegrees=45
  swing1Motion=Locked
  swing2LimitDegrees=45
  swing2Motion=Locked
  twistLimitDegrees=45
  twistMotion=Locked
 }
 {
  bone1Name=dyn_blouse_01_c
  bone2Name=base_blouse_01_c
  swing1LimitDegrees=10
  swing1Motion=Locked
  swing2LimitDegrees=20
  swing2Motion=Locked
  twistLimitDegrees=2
  twistMotion=Locked
 }
 {
  bone1Name=dyn_blouse_02_c
  bone2Name=dyn_blouse_01_c
  swing1LimitDegrees=10
  swing1Motion=Locked
  swing2LimitDegrees=20
  swing2Motion=Locked
  twistLimitDegrees=2
  twistMotion=Locked
 }
 {
  bone1Name=dyn_blouse_03_c
  bone2Name=dyn_blouse_02_c
  swing1LimitDegrees=10
  swing1Motion=Locked
  swing2LimitDegrees=20
  swing2Motion=Locked
  twistLimitDegrees=2
  twistMotion=Locked
 }
 {
  bone1Name=base_blouse_bk_01_l
  bone2Name=pelvis
  swing1LimitDegrees=45
  swing1Motion=Locked
  swing2LimitDegrees=45
  swing2Motion=Locked
  twistLimitDegrees=45
  twistMotion=Locked
 }
 {
  bone1Name=dyn_blouse_bk_01_l
  bone2Name=base_blouse_bk_01_l
  swing1LimitDegrees=10
  swing1Motion=Locked
  swing2LimitDegrees=20
  swing2Motion=Locked
  twistLimitDegrees=2
  twistMotion=Locked
 }
 {bone1Name=dyn_blouse_bk_02_l, bone2Name=dyn_blouse_bk_01_l, swing1LimitDegrees=10, swing2LimitDegrees=10, twistLimitDegrees=2}
 {bone1Name=dyn_blouse_bk_03_l, bone2Name=dyn_blouse_bk_02_l, swing1LimitDegrees=10, swing2LimitDegrees=10, twistLimitDegrees=2}
 {
  bone1Name=base_blouse_bk_01_r
  bone2Name=pelvis
  swing1LimitDegrees=45
  swing1Motion=Locked
  swing2LimitDegrees=45
  swing2Motion=Locked
  twistLimitDegrees=45
  twistMotion=Locked
 }
 {
  bone1Name=dyn_blouse_bk_01_r
  bone2Name=base_blouse_bk_01_r
  swing1LimitDegrees=10
  swing1Motion=Locked
  swing2LimitDegrees=20
  swing2Motion=Locked
  twistLimitDegrees=2
  twistMotion=Locked
 }
 {
  bone1Name=dyn_blouse_bk_02_r
  bone2Name=dyn_blouse_bk_01_r
  swing1LimitDegrees=10
  swing1Motion=Locked
  swing2LimitDegrees=20
  swing2Motion=Locked
  twistLimitDegrees=2
  twistMotion=Locked
 }
 {
  bone1Name=dyn_blouse_bk_03_r
  bone2Name=dyn_blouse_bk_02_r
  swing1LimitDegrees=10
  swing1Motion=Locked
  swing2LimitDegrees=20
  swing2Motion=Locked
  twistLimitDegrees=2
  twistMotion=Locked
 }
 {
  bone1Name=base_blouse_fr_01_l
  bone2Name=pelvis
  swing1LimitDegrees=45
  swing1Motion=Locked
  swing2LimitDegrees=45
  swing2Motion=Locked
  twistLimitDegrees=45
  twistMotion=Locked
 }
 {bone1Name=dyn_blouse_fr_01_l, bone2Name=base_blouse_fr_01_l, swing1LimitDegrees=10, swing2LimitDegrees=10, twistLimitDegrees=2}
 {bone1Name=dyn_blouse_fr_02_l, bone2Name=dyn_blouse_fr_01_l, swing1LimitDegrees=10, swing2LimitDegrees=10, twistLimitDegrees=2}
 {bone1Name=dyn_blouse_fr_03_l, bone2Name=dyn_blouse_fr_02_l, swing1LimitDegrees=10, swing2LimitDegrees=10, twistLimitDegrees=2}
 {
  bone1Name=base_blouse_fr_01_r
  bone2Name=pelvis
  swing1LimitDegrees=45
  swing1Motion=Locked
  swing2LimitDegrees=45
  swing2Motion=Locked
  twistLimitDegrees=45
  twistMotion=Locked
 }
 {bone1Name=dyn_blouse_fr_01_r, bone2Name=base_blouse_fr_01_r, swing1LimitDegrees=10, swing2LimitDegrees=10, twistLimitDegrees=2}
 {bone1Name=dyn_blouse_fr_02_r, bone2Name=dyn_blouse_fr_01_r, swing1LimitDegrees=10, swing2LimitDegrees=10, twistLimitDegrees=2}
 {bone1Name=dyn_blouse_fr_03_r, bone2Name=dyn_blouse_fr_02_r, swing1LimitDegrees=10, swing2LimitDegrees=10, twistLimitDegrees=2}
 {
  bone1Name=base_blouse_side_01_l
  bone2Name=pelvis
  swing1LimitDegrees=45
  swing1Motion=Locked
  swing2LimitDegrees=45
  swing2Motion=Locked
  twistLimitDegrees=45
  twistMotion=Locked
 }
 {bone1Name=dyn_blouse_side_01_l, bone2Name=base_blouse_side_01_l, swing1LimitDegrees=10, swing2LimitDegrees=10, twistLimitDegrees=2}
 {bone1Name=dyn_blouse_side_02_l, bone2Name=dyn_blouse_side_01_l, swing1LimitDegrees=10, swing2LimitDegrees=10, twistLimitDegrees=2}
 {bone1Name=dyn_blouse_side_03_l, bone2Name=dyn_blouse_side_02_l, swing1LimitDegrees=10, swing2LimitDegrees=10, twistLimitDegrees=2}
 {
  bone1Name=base_blouse_side_01_r
  bone2Name=pelvis
  swing1LimitDegrees=45
  swing1Motion=Locked
  swing2LimitDegrees=45
  swing2Motion=Locked
  twistLimitDegrees=45
  twistMotion=Locked
 }
 {bone1Name=dyn_blouse_side_01_r, bone2Name=base_blouse_side_01_r, swing1LimitDegrees=10, swing2LimitDegrees=10, twistLimitDegrees=2}
 {bone1Name=dyn_blouse_side_02_r, bone2Name=dyn_blouse_side_01_r, swing1LimitDegrees=10, swing2LimitDegrees=10, twistLimitDegrees=2}
 {bone1Name=dyn_blouse_side_03_r, bone2Name=dyn_blouse_side_02_r, swing1LimitDegrees=10, swing2LimitDegrees=10, twistLimitDegrees=2}
 {
  bone1Name=dyn_blouse_side_02_l
  bone2Name=dyn_blouse_fr_02_l
  swing1LimitDegrees=45
  swing1Motion=Free
  swing2LimitDegrees=45
  swing2Motion=Free
  twistLimitDegrees=45
  twistMotion=Free
 }
 {
  bone1Name=dyn_blouse_side_03_l
  bone2Name=dyn_blouse_fr_03_l
  swing1LimitDegrees=45
  swing1Motion=Free
  swing2LimitDegrees=45
  swing2Motion=Free
  twistLimitDegrees=45
  twistMotion=Free
 }
]
