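# Physics asset for the back hair of the Miss Missile female character's face accessory mesh.
# Left, right and center four-link capsule chains hang from head-anchored base bodies, with limited swing and zero-degree twist limits.
# Angularly free constraints cross-link both side chains to the center chain at each level.
# Head, neck, spine, clavicle and upper-arm capsules form the body; an unnamed, unconstrained body holds left front side-hair capsules.
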PhysicsAsset /BRCosmetics/Characters/Player/Female/Medium/Bodies/F_MED_MissMissile/Meshes/Parts/F_MED_MissMissile_FaceAcc_BackHairPhysics
bodies {
 None [
  {shapeType=Capsule, length=4, radius=2, shapeName=dyn_hair_side_fr_05_l_capsule, center=<4.3582 0.8301 0>, rotation=<rot 90 0 0>}
  {shapeType=Capsule, length=4, radius=2, shapeName=dyn_hair_side_fr_04_l_capsule, center=<5.3842 1.2528 0>, rotation=<rot 90 0 0>}
  {shapeType=Capsule, length=4, radius=2.2, shapeName=dyn_hair_side_fr_03_l_capsule, center=<5 1 0>, rotation=<rot 90 0 0>}
  {shapeType=Capsule, length=4, radius=1.8, shapeName=dyn_hair_side_fr_02_l_capsule, center=<5 0 1>, rotation=<rot 90 0 0>}
  {shapeType=Capsule, length=4, radius=1, shapeName=dyn_hair_side_fr_01_l_capsule, center=<5 0 1>, rotation=<rot 90 0 0>}
  {shapeType=Capsule, length=0.505, radius=0.505, shapeName=dyn_hair_side_fr_l_base_capsule, rotation=<rot 0 -90 0>}
 ]
 base_Const_Jnt [{shapeType=Capsule, length=0.505, radius=0.505, shapeName=base_Const_Jnt_capsule, rotation=<rot 0 -90 0>}]
 base_L_hair_main [{shapeType=Capsule, length=0.505, radius=0.505, shapeName=base_L_hair_main_capsule, rotation=<rot 0 -90 0>}]
 base_R_hair_main [{shapeType=Capsule, length=0.505, radius=0.505, shapeName=base_R_hair_main_capsule, rotation=<rot 0 -90 0>}]
 clavicle_l [{
   shapeType=Capsule
   length=9.1568
   radius=8.2
   shapeName=clavicle_l_capsule
   center=<6.7709 -0.378 -0.743>
   rotation=<rot 79.0669 -0.0002 -0.0003>
  }]
 clavicle_r [{
   shapeType=Capsule
   length=9.1868
   radius=8.2
   shapeName=clavicle_r_capsule
   center=<-6.6401 0.378 1.4201>
   rotation=<rot 79.0669 -0.0002 179.9997>
  }]
 dyn_Const_Jnt_1 [{shapeType=Capsule, length=8, radius=6, shapeName=dyn_Const_Jnt_1_capsule, center=<10 0 3>, rotation=<rot 78.0075 -180 -180>}]
 dyn_Const_Jnt_2 [{shapeType=Capsule, length=8, radius=6, shapeName=dyn_Const_Jnt_2_capsule, center=<8 0 6>, rotation=<rot 90 0 0>}]
 dyn_Const_Jnt_3 [{shapeType=Capsule, length=8, radius=6, shapeName=dyn_Const_Jnt_3_capsule, center=<8 0 5>, rotation=<rot 90 0 0>}]
 dyn_Const_Jnt_4 [
  {shapeType=Capsule, length=8, radius=6, shapeName=dyn_Const_Jnt_4_capsule, center=<8 0 7>, rotation=<rot 90 0 0>}
  {
   shapeType=Capsule
   length=14.8977
   radius=6.2618
   shapeName=dyn_Const_Jnt_4_capsule
   center=<11.7707 0.1727 7.0147>
   rotation=<rot -0.0001 0 92.0843>
  }
 ]
 dyn_L_hair_main_1 [{shapeType=Capsule, length=8, radius=6, shapeName=dyn_L_hair_main_1_capsule, center=<10 0 3>, rotation=<rot 78.0075 -180 -180>}]
 dyn_L_hair_main_2 [{shapeType=Capsule, length=8, radius=6.5, shapeName=dyn_L_hair_main_2_capsule, center=<8 0 6>, rotation=<rot 90 0 0>}]
 dyn_L_hair_main_3 [{shapeType=Capsule, length=8, radius=6.5, shapeName=dyn_L_hair_main_3_capsule, center=<8 0 5>, rotation=<rot 90 0 0>}]
 dyn_L_hair_main_4 [{shapeType=Capsule, length=8, radius=6.5, shapeName=dyn_L_hair_main_4_capsule, center=<7 0 7>, rotation=<rot 90 0 0>}]
 dyn_R_hair_main_1 [{shapeType=Capsule, length=8, radius=6, shapeName=dyn_R_hair_main_1_capsule, center=<10 0 3>, rotation=<rot 78.0075 180 180>}]
 dyn_R_hair_main_2 [{shapeType=Capsule, length=8, radius=6.5, shapeName=dyn_R_hair_main_2_capsule, center=<8 0 6>, rotation=<rot 90 0 0>}]
 dyn_R_hair_main_3 [{shapeType=Capsule, length=8, radius=6.5, shapeName=dyn_R_hair_main_3_capsule, center=<8 0 5>, rotation=<rot 90 0 0>}]
 dyn_R_hair_main_4 [{shapeType=Capsule, length=8, radius=6.5, shapeName=dyn_R_hair_main_4_capsule, center=<7 0 7>, rotation=<rot 90 0 0>}]
 head [{shapeType=Capsule, length=5, radius=6, shapeName=head_capsule, rotation=<rot 90 0 0>}]
 neck_01 [{shapeType=Capsule, length=0.505, radius=0.505, shapeName=neck_01_capsule, rotation=<rot 0 -90 0>}]
 neck_02 [{shapeType=Capsule, length=5, radius=6, shapeName=neck_02_capsule, rotation=<rot 90 0 0>}]
 spine_03 [{shapeType=Capsule, length=12.9881, radius=11.9562, shapeName=spine_03_capsule, center=<2.3385 -0.2546 0>}]
 spine_04 [{shapeType=Capsule, length=13.7664, radius=13.0417, shapeName=spine_04_capsule, center=<7.5008 1.4637 0>}]
 spine_05 [{shapeType=Capsule, length=8, radius=12, shapeName=spine_05_capsule, center=<6.8341 3.3479 0>}]
 upperarm_l [{shapeType=Capsule, length=18.3304, radius=7, shapeName=upperarm_l_capsule, center=<15.0826 -0.6451 0.4414>, rotation=<rot 90 0 0>}]
 upperarm_r [{shapeType=Capsule, length=17, radius=7.5, shapeName=upperarm_r_capsule, center=<-14.9195 -0.6 0.4>, rotation=<rot 90 0 0>}]
}
constraints [
 {
  bone1Name=spine_04
  bone2Name=spine_03
  swing1LimitDegrees=45
  swing1Motion=Free
  swing2LimitDegrees=45
  swing2Motion=Free
  twistLimitDegrees=45
  twistMotion=Free
 }
 {
  bone1Name=spine_05
  bone2Name=spine_04
  swing1LimitDegrees=45
  swing1Motion=Free
  swing2LimitDegrees=45
  swing2Motion=Free
  twistLimitDegrees=45
  twistMotion=Free
 }
 {
  bone1Name=upperarm_l
  bone2Name=clavicle_l
  swing1LimitDegrees=45
  swing1Motion=Free
  swing2LimitDegrees=45
  swing2Motion=Free
  twistLimitDegrees=45
  twistMotion=Free
 }
 {
  bone1Name=head
  bone2Name=neck_02
  swing1LimitDegrees=45
  swing1Motion=Free
  swing2LimitDegrees=45
  swing2Motion=Free
  twistLimitDegrees=45
  twistMotion=Free
 }
 {
  bone1Name=neck_02
  bone2Name=neck_01
  swing1LimitDegrees=45
  swing1Motion=Free
  swing2LimitDegrees=45
  swing2Motion=Free
  twistLimitDegrees=45
  twistMotion=Free
 }
 {
  bone1Name=neck_01
  bone2Name=spine_05
  swing1LimitDegrees=45
  swing1Motion=Free
  swing2LimitDegrees=45
  swing2Motion=Free
  twistLimitDegrees=45
  twistMotion=Free
 }
 {
  bone1Name=clavicle_l
  bone2Name=spine_05
  swing1LimitDegrees=45
  swing1Motion=Free
  swing2LimitDegrees=45
  swing2Motion=Free
  twistLimitDegrees=45
  twistMotion=Free
 }
 {
  bone1Name=base_L_hair_main
  bone2Name=head
  swing1LimitDegrees=45
  swing1Motion=Free
  swing2LimitDegrees=45
  swing2Motion=Free
  twistLimitDegrees=45
  twistMotion=Free
 }
 {bone1Name=dyn_L_hair_main_1, bone2Name=base_L_hair_main, swing1LimitDegrees=70, swing2LimitDegrees=50, twistLimitDegrees=0}
 {bone1Name=dyn_L_hair_main_2, bone2Name=dyn_L_hair_main_1, swing1LimitDegrees=90, swing2LimitDegrees=50, twistLimitDegrees=0}
 {bone1Name=dyn_L_hair_main_3, bone2Name=dyn_L_hair_main_2, swing1LimitDegrees=30, swing2LimitDegrees=40, twistLimitDegrees=0}
 {bone1Name=dyn_L_hair_main_4, bone2Name=dyn_L_hair_main_3, swing1LimitDegrees=30, swing2LimitDegrees=50, twistLimitDegrees=0}
 {
  bone1Name=base_R_hair_main
  bone2Name=head
  swing1LimitDegrees=45
  swing1Motion=Free
  swing2LimitDegrees=45
  swing2Motion=Free
  twistLimitDegrees=45
  twistMotion=Free
 }
 {bone1Name=dyn_R_hair_main_1, bone2Name=base_R_hair_main, swing1LimitDegrees=70, swing2LimitDegrees=50, twistLimitDegrees=0}
 {bone1Name=dyn_R_hair_main_2, bone2Name=dyn_R_hair_main_1, swing1LimitDegrees=90, swing2LimitDegrees=50, twistLimitDegrees=0}
 {bone1Name=dyn_R_hair_main_3, bone2Name=dyn_R_hair_main_2, swing1LimitDegrees=30, swing2LimitDegrees=40, twistLimitDegrees=0}
 {bone1Name=dyn_R_hair_main_4, bone2Name=dyn_R_hair_main_3, swing1LimitDegrees=30, swing2LimitDegrees=50, twistLimitDegrees=0}
 {
  bone1Name=base_Const_Jnt
  bone2Name=head
  swing1LimitDegrees=45
  swing1Motion=Free
  swing2LimitDegrees=45
  swing2Motion=Free
  twistLimitDegrees=45
  twistMotion=Free
 }
 {bone1Name=dyn_Const_Jnt_1, bone2Name=base_Const_Jnt, swing1LimitDegrees=30, swing2LimitDegrees=40, twistLimitDegrees=0}
 {bone1Name=dyn_Const_Jnt_2, bone2Name=dyn_Const_Jnt_1, swing1LimitDegrees=40, swing2LimitDegrees=40, twistLimitDegrees=0}
 {bone1Name=dyn_Const_Jnt_3, bone2Name=dyn_Const_Jnt_2, swing1LimitDegrees=30, swing2LimitDegrees=40, twistLimitDegrees=0}
 {bone1Name=dyn_Const_Jnt_4, bone2Name=dyn_Const_Jnt_3, swing1LimitDegrees=30, swing2LimitDegrees=50, twistLimitDegrees=0}
 {
  bone1Name=dyn_L_hair_main_4
  bone2Name=dyn_Const_Jnt_4
  swing1LimitDegrees=45
  swing1Motion=Free
  swing2LimitDegrees=45
  swing2Motion=Free
  twistLimitDegrees=45
  twistMotion=Free
 }
 {
  bone1Name=dyn_R_hair_main_4
  bone2Name=dyn_Const_Jnt_4
  swing1LimitDegrees=45
  swing1Motion=Free
  swing2LimitDegrees=45
  swing2Motion=Free
  twistLimitDegrees=45
  twistMotion=Free
 }
 {
  bone1Name=upperarm_r
  bone2Name=clavicle_r
  swing1LimitDegrees=45
  swing1Motion=Free
  swing2LimitDegrees=45
  swing2Motion=Free
  twistLimitDegrees=45
  twistMotion=Free
 }
 {
  bone1Name=clavicle_r
  bone2Name=spine_05
  swing1LimitDegrees=45
  swing1Motion=Free
  swing2LimitDegrees=45
  swing2Motion=Free
  twistLimitDegrees=45
  twistMotion=Free
 }
 {
  bone1Name=dyn_L_hair_main_1
  bone2Name=dyn_Const_Jnt_1
  swing1LimitDegrees=45
  swing1Motion=Free
  swing2LimitDegrees=45
  swing2Motion=Free
  twistLimitDegrees=45
  twistMotion=Free
 }
 {
  bone1Name=dyn_R_hair_main_1
  bone2Name=dyn_Const_Jnt_1
  swing1LimitDegrees=45
  swing1Motion=Free
  swing2LimitDegrees=45
  swing2Motion=Free
  twistLimitDegrees=45
  twistMotion=Free
 }
 {
  bone1Name=dyn_L_hair_main_2
  bone2Name=dyn_Const_Jnt_2
  swing1LimitDegrees=45
  swing1Motion=Free
  swing2LimitDegrees=45
  swing2Motion=Free
  twistLimitDegrees=45
  twistMotion=Free
 }
 {
  bone1Name=dyn_R_hair_main_2
  bone2Name=dyn_Const_Jnt_2
  swing1LimitDegrees=45
  swing1Motion=Free
  swing2LimitDegrees=45
  swing2Motion=Free
  twistLimitDegrees=45
  twistMotion=Free
 }
 {
  bone1Name=dyn_L_hair_main_3
  bone2Name=dyn_Const_Jnt_3
  swing1LimitDegrees=45
  swing1Motion=Free
  swing2LimitDegrees=45
  swing2Motion=Free
  twistLimitDegrees=45
  twistMotion=Free
 }
 {
  bone1Name=dyn_R_hair_main_3
  bone2Name=dyn_Const_Jnt_3
  swing1LimitDegrees=45
  swing1Motion=Free
  swing2LimitDegrees=45
  swing2Motion=Free
  twistLimitDegrees=45
  twistMotion=Free
 }
]
