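# Physics asset for the back hair of the Miss Missile female medium cosmetic, named as the emote variant.
# Three four-bone capsule chains (left, right and center Const_Jnt) hang from head-attached base bones, with limited swing and zero twist.
# Angularly free constraints tie each left and right segment to its matching center segment, cross-linking the chains.
# Capsules on head, neck, upper spine, clavicles and upper arms supply body shapes around the hair.
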PhysicsAsset /BRCosmetics/Characters/Player/Female/Medium/Bodies/F_MED_MissMissile/Meshes/Parts/F_MED_MissMissile_FaceAcc_BackHair_Emotes_Physics
bodies {
 base_Const_Jnt [{shapeType=Capsule, length=0.505, radius=0.505, shapeName=base_Const_Jnt_capsule, rotation=<rot 0 -90 0>}]
 base_L_hair_main [{shapeType=Capsule, length=0.505, radius=0.505, shapeName=base_L_hair_main_capsule, rotation=<rot 0 -90 0>}]
 base_R_hair_main [{shapeType=Capsule, length=0.505, radius=0.505, shapeName=base_R_hair_main_capsule, rotation=<rot 0 -90 0>}]
 clavicle_l [{shapeType=Capsule, length=3.7493, radius=9.2, shapeName=clavicle_l_capsule, center=<6.8391 1.7 -2.6881>, rotation=<rot -83 180 0>}]
 clavicle_r [{shapeType=Capsule, length=3.7493, radius=9.2, shapeName=clavicle_r_capsule, center=<-6.8391 -1.7 2.6881>, rotation=<rot -83 180 -180>}]
 dyn_Const_Jnt_1 [{shapeType=Capsule, length=8, radius=6, shapeName=dyn_Const_Jnt_1_capsule, center=<8 0 3.5>, rotation=<rot 90 0 0>}]
 dyn_Const_Jnt_2 [{shapeType=Capsule, length=8, radius=6, shapeName=dyn_Const_Jnt_2_capsule, center=<8 0 4>, rotation=<rot 90 0 0>}]
 dyn_Const_Jnt_3 [{shapeType=Capsule, length=8, radius=6, shapeName=dyn_Const_Jnt_3_capsule, center=<8 0 4>, rotation=<rot 90 0 0>}]
 dyn_Const_Jnt_4 [
  {shapeType=Capsule, length=8, radius=6, shapeName=dyn_Const_Jnt_4_capsule, center=<8 0 4>, rotation=<rot 90 0 0>}
  {
   shapeType=Capsule
   length=14.8977
   radius=6.2618
   shapeName=dyn_Const_Jnt_4_capsule
   center=<11.7707 0.0994 2.7305>
   rotation=<rot -0.0001 0 92.0843>
  }
 ]
 dyn_L_hair_main_1 [{shapeType=Capsule, length=8, radius=6, shapeName=dyn_L_hair_main_1_capsule, center=<8 0 3.5>, rotation=<rot 90 0 0>}]
 dyn_L_hair_main_2 [{shapeType=Capsule, length=8, radius=6.5, shapeName=dyn_L_hair_main_2_capsule, center=<8 0 3>, rotation=<rot 90 0 0>}]
 dyn_L_hair_main_3 [{shapeType=Capsule, length=8, radius=6.5, shapeName=dyn_L_hair_main_3_capsule, center=<8 0 3>, rotation=<rot 90 0 0>}]
 dyn_L_hair_main_4 [{shapeType=Capsule, length=8, radius=6.5, shapeName=dyn_L_hair_main_4_capsule, center=<7 0 3>, rotation=<rot 90 0 0>}]
 dyn_R_hair_main_1 [{shapeType=Capsule, length=8, radius=6, shapeName=dyn_R_hair_main_1_capsule, center=<8 0 3.5>, rotation=<rot 90 0 0>}]
 dyn_R_hair_main_2 [{shapeType=Capsule, length=8, radius=6.5, shapeName=dyn_R_hair_main_2_capsule, center=<8 0 3.5>, rotation=<rot 90 0 0>}]
 dyn_R_hair_main_3 [{shapeType=Capsule, length=8, radius=6.5, shapeName=dyn_R_hair_main_3_capsule, center=<8 0 4>, rotation=<rot 90 0 0>}]
 dyn_R_hair_main_4 [{shapeType=Capsule, length=8, radius=6.5, shapeName=dyn_R_hair_main_4_capsule, center=<7 0 4>, rotation=<rot 90 0 0>}]
 head [{shapeType=Capsule, length=5, radius=6, shapeName=head_capsule, rotation=<rot 90 0 0>}]
 neck_01 [{shapeType=Capsule, length=10.0031, radius=5, shapeName=neck_01_capsule, center=<0 -4.007 0>, rotation=<rot 0 -90 0>}]
 neck_02 [{shapeType=Capsule, length=5, radius=6, shapeName=neck_02_capsule, rotation=<rot 90 0 0>}]
 spine_03 [{shapeType=Capsule, length=10, radius=8, shapeName=spine_03_capsule, center=<0 1.032 0>}]
 spine_04 [{shapeType=Capsule, length=9.9496, radius=13.6998, shapeName=spine_04_capsule, center=<0 1.1695 0>}]
 spine_05 [{
   shapeType=Capsule
   length=15.4879
   radius=12.4735
   shapeName=spine_05_capsule
   center=<2.9539 0.7735 -0.0541>
   rotation=<rot 85.8721 179.9942 163.3794>
  }]
 upperarm_l [{shapeType=Capsule, length=18.3304, radius=7, shapeName=upperarm_l_capsule, center=<15.0826 -0.6451 0.4414>, rotation=<rot 90 0 0>}]
 upperarm_r [{shapeType=Capsule, length=17, radius=7.5, shapeName=upperarm_r_capsule, center=<-14.9195 -0.6 0.4>, rotation=<rot 90 0 0>}]
}
constraints [
 {
  bone1Name=spine_04
  bone2Name=spine_03
  swing1LimitDegrees=45
  swing1Motion=Free
  swing2LimitDegrees=45
  swing2Motion=Free
  twistLimitDegrees=45
  twistMotion=Free
 }
 {
  bone1Name=spine_05
  bone2Name=spine_04
  swing1LimitDegrees=45
  swing1Motion=Free
  swing2LimitDegrees=45
  swing2Motion=Free
  twistLimitDegrees=45
  twistMotion=Free
 }
 {
  bone1Name=upperarm_l
  bone2Name=clavicle_l
  swing1LimitDegrees=45
  swing1Motion=Free
  swing2LimitDegrees=45
  swing2Motion=Free
  twistLimitDegrees=45
  twistMotion=Free
 }
 {
  bone1Name=head
  bone2Name=neck_02
  swing1LimitDegrees=45
  swing1Motion=Free
  swing2LimitDegrees=45
  swing2Motion=Free
  twistLimitDegrees=45
  twistMotion=Free
 }
 {
  bone1Name=neck_02
  bone2Name=neck_01
  swing1LimitDegrees=45
  swing1Motion=Free
  swing2LimitDegrees=45
  swing2Motion=Free
  twistLimitDegrees=45
  twistMotion=Free
 }
 {
  bone1Name=neck_01
  bone2Name=spine_05
  swing1LimitDegrees=45
  swing1Motion=Free
  swing2LimitDegrees=45
  swing2Motion=Free
  twistLimitDegrees=45
  twistMotion=Free
 }
 {
  bone1Name=clavicle_l
  bone2Name=spine_05
  swing1LimitDegrees=45
  swing1Motion=Free
  swing2LimitDegrees=45
  swing2Motion=Free
  twistLimitDegrees=45
  twistMotion=Free
 }
 {
  bone1Name=base_L_hair_main
  bone2Name=head
  swing1LimitDegrees=45
  swing1Motion=Free
  swing2LimitDegrees=45
  swing2Motion=Free
  twistLimitDegrees=45
  twistMotion=Free
 }
 {bone1Name=dyn_L_hair_main_1, bone2Name=base_L_hair_main, swing1LimitDegrees=70, swing2LimitDegrees=50, twistLimitDegrees=0}
 {bone1Name=dyn_L_hair_main_2, bone2Name=dyn_L_hair_main_1, swing1LimitDegrees=90, swing2LimitDegrees=50, twistLimitDegrees=0}
 {bone1Name=dyn_L_hair_main_3, bone2Name=dyn_L_hair_main_2, swing1LimitDegrees=90, swing2LimitDegrees=50, twistLimitDegrees=0}
 {bone1Name=dyn_L_hair_main_4, bone2Name=dyn_L_hair_main_3, swing1LimitDegrees=90, swing2LimitDegrees=50, twistLimitDegrees=0}
 {
  bone1Name=base_R_hair_main
  bone2Name=head
  swing1LimitDegrees=45
  swing1Motion=Free
  swing2LimitDegrees=45
  swing2Motion=Free
  twistLimitDegrees=45
  twistMotion=Free
 }
 {bone1Name=dyn_R_hair_main_1, bone2Name=base_R_hair_main, swing1LimitDegrees=70, swing2LimitDegrees=50, twistLimitDegrees=0}
 {bone1Name=dyn_R_hair_main_2, bone2Name=dyn_R_hair_main_1, swing1LimitDegrees=90, swing2LimitDegrees=50, twistLimitDegrees=0}
 {bone1Name=dyn_R_hair_main_3, bone2Name=dyn_R_hair_main_2, swing1LimitDegrees=90, swing2LimitDegrees=50, twistLimitDegrees=0}
 {bone1Name=dyn_R_hair_main_4, bone2Name=dyn_R_hair_main_3, swing1LimitDegrees=90, swing2LimitDegrees=50, twistLimitDegrees=0}
 {
  bone1Name=base_Const_Jnt
  bone2Name=head
  swing1LimitDegrees=45
  swing1Motion=Free
  swing2LimitDegrees=45
  swing2Motion=Free
  twistLimitDegrees=45
  twistMotion=Free
 }
 {bone1Name=dyn_Const_Jnt_1, bone2Name=base_Const_Jnt, swing1LimitDegrees=30, swing2LimitDegrees=40, twistLimitDegrees=0}
 {bone1Name=dyn_Const_Jnt_2, bone2Name=dyn_Const_Jnt_1, swing1LimitDegrees=40, swing2LimitDegrees=40, twistLimitDegrees=0}
 {bone1Name=dyn_Const_Jnt_3, bone2Name=dyn_Const_Jnt_2, swing1LimitDegrees=40, swing2LimitDegrees=40, twistLimitDegrees=0}
 {bone1Name=dyn_Const_Jnt_4, bone2Name=dyn_Const_Jnt_3, swing1LimitDegrees=40, swing2LimitDegrees=40, twistLimitDegrees=0}
 {
  bone1Name=dyn_L_hair_main_4
  bone2Name=dyn_Const_Jnt_4
  swing1LimitDegrees=45
  swing1Motion=Free
  swing2LimitDegrees=45
  swing2Motion=Free
  twistLimitDegrees=45
  twistMotion=Free
 }
 {
  bone1Name=dyn_R_hair_main_4
  bone2Name=dyn_Const_Jnt_4
  swing1LimitDegrees=45
  swing1Motion=Free
  swing2LimitDegrees=45
  swing2Motion=Free
  twistLimitDegrees=45
  twistMotion=Free
 }
 {
  bone1Name=upperarm_r
  bone2Name=clavicle_r
  swing1LimitDegrees=45
  swing1Motion=Free
  swing2LimitDegrees=45
  swing2Motion=Free
  twistLimitDegrees=45
  twistMotion=Free
 }
 {
  bone1Name=clavicle_r
  bone2Name=spine_05
  swing1LimitDegrees=45
  swing1Motion=Free
  swing2LimitDegrees=45
  swing2Motion=Free
  twistLimitDegrees=45
  twistMotion=Free
 }
 {
  bone1Name=dyn_L_hair_main_1
  bone2Name=dyn_Const_Jnt_1
  swing1LimitDegrees=45
  swing1Motion=Free
  swing2LimitDegrees=45
  swing2Motion=Free
  twistLimitDegrees=45
  twistMotion=Free
 }
 {
  bone1Name=dyn_R_hair_main_1
  bone2Name=dyn_Const_Jnt_1
  swing1LimitDegrees=45
  swing1Motion=Free
  swing2LimitDegrees=45
  swing2Motion=Free
  twistLimitDegrees=45
  twistMotion=Free
 }
 {
  bone1Name=dyn_L_hair_main_2
  bone2Name=dyn_Const_Jnt_2
  swing1LimitDegrees=45
  swing1Motion=Free
  swing2LimitDegrees=45
  swing2Motion=Free
  twistLimitDegrees=45
  twistMotion=Free
 }
 {
  bone1Name=dyn_R_hair_main_2
  bone2Name=dyn_Const_Jnt_2
  swing1LimitDegrees=45
  swing1Motion=Free
  swing2LimitDegrees=45
  swing2Motion=Free
  twistLimitDegrees=45
  twistMotion=Free
 }
 {
  bone1Name=dyn_L_hair_main_3
  bone2Name=dyn_Const_Jnt_3
  swing1LimitDegrees=45
  swing1Motion=Free
  swing2LimitDegrees=45
  swing2Motion=Free
  twistLimitDegrees=45
  twistMotion=Free
 }
 {
  bone1Name=dyn_R_hair_main_3
  bone2Name=dyn_Const_Jnt_3
  swing1LimitDegrees=45
  swing1Motion=Free
  swing2LimitDegrees=45
  swing2Motion=Free
  twistLimitDegrees=45
  twistMotion=Free
 }
]
