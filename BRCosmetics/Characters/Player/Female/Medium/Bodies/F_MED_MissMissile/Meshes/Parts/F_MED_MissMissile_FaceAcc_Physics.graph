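# Physics asset for the Miss Missile female medium body's FaceAcc mesh part, rigging a left front side hair strand for physical motion.
# A five-link capsule chain hangs from a hair base body on the head, with swing capped at 20 degrees at the root and 30 below, and zero-degree twist limits.
# Capsules on head, neck, upper spine, left clavicle and upper arm form the surrounding body shapes, linked by free-motion constraints.
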PhysicsAsset /BRCosmetics/Characters/Player/Female/Medium/Bodies/F_MED_MissMissile/Meshes/Parts/F_MED_MissMissile_FaceAcc_Physics
bodies {
 clavicle_l [{shapeType=Capsule, length=13.6231, radius=6, shapeName=clavicle_l_capsule, center=<2.9174 -0.7056 1.0658>, rotation=<rot 90 0 0>}]
 dyn_hair_side_fr_01_l [{shapeType=Capsule, length=4, radius=1, shapeName=dyn_hair_side_fr_01_l_capsule, center=<5 0 1>, rotation=<rot 90 0 0>}]
 dyn_hair_side_fr_02_l [{shapeType=Capsule, length=4, radius=1.8, shapeName=dyn_hair_side_fr_02_l_capsule, center=<5 0 1>, rotation=<rot 90 0 0>}]
 dyn_hair_side_fr_03_l [{shapeType=Capsule, length=4, radius=2.2, shapeName=dyn_hair_side_fr_03_l_capsule, center=<5 1 0>, rotation=<rot 90 0 0>}]
 dyn_hair_side_fr_04_l [{shapeType=Capsule, length=4, radius=2, shapeName=dyn_hair_side_fr_04_l_capsule, center=<5.3842 1.2528 0>, rotation=<rot 90 0 0>}]
 dyn_hair_side_fr_05_l [{shapeType=Capsule, length=4, radius=2, shapeName=dyn_hair_side_fr_05_l_capsule, center=<4.3582 0.8301 0>, rotation=<rot 90 0 0>}]
 dyn_hair_side_fr_l_base [{shapeType=Capsule, length=0.505, radius=0.505, shapeName=dyn_hair_side_fr_l_base_capsule, rotation=<rot 0 -90 0>}]
 head [{shapeType=Capsule, length=5, radius=6, shapeName=head_capsule, rotation=<rot 90 0 0>}]
 neck_01 [{shapeType=Capsule, length=0.505, radius=0.505, shapeName=neck_01_capsule, rotation=<rot 0 -90 0>}]
 neck_02 [{shapeType=Capsule, length=5, radius=6, shapeName=neck_02_capsule, rotation=<rot 90 0 0>}]
 spine_04 [{shapeType=Capsule, length=18.8904, radius=13, shapeName=spine_04_capsule, center=<0 1.9595 0>}]
 spine_05 [
  {shapeType=Capsule, length=25, radius=9.5152, shapeName=spine_05_capsule, center=<0 -0.2341 0>}
  {
   shapeType=Capsule
   length=14.5445
   radius=8.6836
   shapeName=spine_05_capsule
   center=<12.2675 -4.5114 -17.2529>
   rotation=<rot 86.4416 179.9999 180>
  }
  {
   shapeType=Capsule
   length=10.0591
   radius=11.1003
   shapeName=spine_05_capsule
   center=<7.7967 2.9046 3.8221>
   rotation=<rot 87.6511 -0.0001 0>
  }
 ]
 upperarm_l [{shapeType=Capsule, length=20, radius=6.5, shapeName=upperarm_l_capsule, center=<10 -0.6451 0.4414>, rotation=<rot 90 0 0>}]
}
constraints [
 {
  bone1Name=spine_05
  bone2Name=spine_04
  swing1LimitDegrees=45
  swing1Motion=Free
  swing2LimitDegrees=45
  swing2Motion=Free
  twistLimitDegrees=45
  twistMotion=Free
 }
 {
  bone1Name=upperarm_l
  bone2Name=clavicle_l
  swing1LimitDegrees=45
  swing1Motion=Free
  swing2LimitDegrees=45
  swing2Motion=Free
  twistLimitDegrees=45
  twistMotion=Free
 }
 {
  bone1Name=head
  bone2Name=neck_02
  swing1LimitDegrees=45
  swing1Motion=Free
  swing2LimitDegrees=45
  swing2Motion=Free
  twistLimitDegrees=45
  twistMotion=Free
 }
 {
  bone1Name=neck_02
  bone2Name=neck_01
  swing1LimitDegrees=45
  swing1Motion=Free
  swing2LimitDegrees=45
  swing2Motion=Free
  twistLimitDegrees=45
  twistMotion=Free
 }
 {
  bone1Name=neck_01
  bone2Name=spine_05
  swing1LimitDegrees=45
  swing1Motion=Free
  swing2LimitDegrees=45
  swing2Motion=Free
  twistLimitDegrees=45
  twistMotion=Free
 }
 {
  bone1Name=dyn_hair_side_fr_l_base
  bone2Name=head
  swing1LimitDegrees=45
  swing1Motion=Free
  swing2LimitDegrees=45
  swing2Motion=Free
  twistLimitDegrees=45
  twistMotion=Free
 }
 {bone1Name=dyn_hair_side_fr_01_l, bone2Name=dyn_hair_side_fr_l_base, swing1LimitDegrees=20, swing2LimitDegrees=20, twistLimitDegrees=0}
 {bone1Name=dyn_hair_side_fr_02_l, bone2Name=dyn_hair_side_fr_01_l, swing1LimitDegrees=30, swing2LimitDegrees=30, twistLimitDegrees=0}
 {bone1Name=dyn_hair_side_fr_03_l, bone2Name=dyn_hair_side_fr_02_l, swing1LimitDegrees=30, swing2LimitDegrees=30, twistLimitDegrees=0}
 {bone1Name=dyn_hair_side_fr_04_l, bone2Name=dyn_hair_side_fr_03_l, swing1LimitDegrees=30, swing2LimitDegrees=30, twistLimitDegrees=0}
 {bone1Name=dyn_hair_side_fr_05_l, bone2Name=dyn_hair_side_fr_04_l, swing1LimitDegrees=30, swing2LimitDegrees=30, twistLimitDegrees=0}
 {
  bone1Name=clavicle_l
  bone2Name=spine_05
  swing1LimitDegrees=45
  swing1Motion=Free
  swing2LimitDegrees=45
  swing2Motion=Free
  twistLimitDegrees=45
  twistMotion=Free
 }
]
