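# Skydive physics asset for the Miss Missile FaceAcc mesh, simulating two hair strands (left and right) against capsule bodies on the head, neck, upper spine, clavicles and upper arms.
# Each strand is a four-bone chain of tightly swing-limited, zero-twist joints hanging from a base bone that is angularly locked to both head and upper neck.
# Torso, neck and arm joints rotate freely, and an extra angularly free constraint links the two strand tips.
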PhysicsAsset /BRCosmetics/Characters/Player/Female/Medium/Bodies/F_MED_MissMissile/Meshes/Parts/F_MED_MissMissile_FaceAcc_Physics_Skydive
bodies {
 base_L_hair_main [{shapeType=Capsule, length=0.505, radius=0.505, shapeName=base_L_hair_main_capsule, rotation=<rot 0 -90 0>}]
 base_R_hair_main [{shapeType=Capsule, length=0.505, radius=0.505, shapeName=base_R_hair_main_capsule, rotation=<rot 0 -90 0>}]
 clavicle_l [{shapeType=Capsule, length=16.0076, radius=6.4543, shapeName=clavicle_l_capsule, center=<2.4781 0.3392 2.9046>, rotation=<rot 90 -175.2522 0>}]
 clavicle_r [{shapeType=Capsule, length=16.0076, radius=6.4543, shapeName=clavicle_r_capsule, center=<-2.4781 -0.3392 -2.9046>, rotation=<rot 90 4.7478 0>}]
 dyn_L_hair_main_1 [{shapeType=Capsule, length=8, radius=6, shapeName=dyn_L_hair_main_1_capsule, center=<8 0 4>, rotation=<rot 90 0 0>}]
 dyn_L_hair_main_2 [{shapeType=Capsule, length=8, radius=6.5, shapeName=dyn_L_hair_main_2_capsule, center=<8 0 3>, rotation=<rot 90 0 0>}]
 dyn_L_hair_main_3 [{shapeType=Capsule, length=8, radius=6.5, shapeName=dyn_L_hair_main_3_capsule, center=<8 0 3>, rotation=<rot 90 0 0>}]
 dyn_L_hair_main_4 [{shapeType=Capsule, length=8, radius=6.5, shapeName=dyn_L_hair_main_4_capsule, center=<7 0 3>, rotation=<rot 90 0 0>}]
 dyn_R_hair_main_1 [{shapeType=Capsule, length=8, radius=6, shapeName=dyn_R_hair_main_1_capsule, center=<8 0 4>, rotation=<rot 90 0 0>}]
 dyn_R_hair_main_2 [{shapeType=Capsule, length=8, radius=6.5, shapeName=dyn_R_hair_main_2_capsule, center=<8 0 3.5>, rotation=<rot 90 0 0>}]
 dyn_R_hair_main_3 [{shapeType=Capsule, length=8, radius=6.5, shapeName=dyn_R_hair_main_3_capsule, center=<8 0 4>, rotation=<rot 90 0 0>}]
 dyn_R_hair_main_4 [{shapeType=Capsule, length=8, radius=6.5, shapeName=dyn_R_hair_main_4_capsule, center=<7 0 4>, rotation=<rot 90 0 0>}]
 head [{shapeType=Capsule, length=5, radius=6, shapeName=head_capsule, rotation=<rot 90 0 0>}]
 neck_01 [{shapeType=Capsule, length=26.6342, radius=7.009, shapeName=neck_01_capsule, center=<1.8387 0.1482 -0.1532>, rotation=<rot -2.2834 89.9997 0>}]
 neck_02 [{shapeType=Capsule, length=11.7889, radius=6.5457, shapeName=neck_02_capsule, rotation=<rot 90 0 0>}]
 spine_04 [{shapeType=Capsule, length=9.9496, radius=13.6998, shapeName=spine_04_capsule, center=<0 1.1695 0>}]
 spine_05 [{shapeType=Capsule, length=53.6222, radius=12, shapeName=spine_05_capsule, center=<10.3717 4.4027 0>}]
 upperarm_l [{shapeType=Capsule, length=20, radius=6.5, shapeName=upperarm_l_capsule, center=<15 0.5 0>, rotation=<rot 90 0 0>}]
 upperarm_r [{shapeType=Capsule, length=20, radius=6.5, shapeName=upperarm_r_capsule, center=<-15 -0.5 0>, rotation=<rot 90 0 0>}]
}
constraints [
 {
  bone1Name=spine_05
  bone2Name=spine_04
  swing1LimitDegrees=45
  swing1Motion=Free
  swing2LimitDegrees=45
  swing2Motion=Free
  twistLimitDegrees=45
  twistMotion=Free
 }
 {
  bone1Name=head
  bone2Name=neck_02
  swing1LimitDegrees=45
  swing1Motion=Free
  swing2LimitDegrees=45
  swing2Motion=Free
  twistLimitDegrees=45
  twistMotion=Free
 }
 {
  bone1Name=base_L_hair_main
  bone2Name=neck_02
  swing1LimitDegrees=45
  swing1Motion=Locked
  swing2LimitDegrees=45
  swing2Motion=Locked
  twistLimitDegrees=45
  twistMotion=Locked
 }
 {
  bone1Name=base_R_hair_main
  bone2Name=neck_02
  swing1LimitDegrees=45
  swing1Motion=Locked
  swing2LimitDegrees=45
  swing2Motion=Locked
  twistLimitDegrees=45
  twistMotion=Locked
 }
 {
  bone1Name=neck_01
  bone2Name=spine_05
  swing1LimitDegrees=45
  swing1Motion=Free
  swing2LimitDegrees=45
  swing2Motion=Free
  twistLimitDegrees=45
  twistMotion=Free
 }
 {
  bone1Name=neck_02
  bone2Name=neck_01
  swing1LimitDegrees=45
  swing1Motion=Free
  swing2LimitDegrees=45
  swing2Motion=Free
  twistLimitDegrees=45
  twistMotion=Free
 }
 {bone1Name=dyn_L_hair_main_2, bone2Name=dyn_L_hair_main_1, swing1LimitDegrees=10, swing2LimitDegrees=10, twistLimitDegrees=0}
 {bone1Name=dyn_L_hair_main_3, bone2Name=dyn_L_hair_main_2, swing1LimitDegrees=10, swing2LimitDegrees=10, twistLimitDegrees=0}
 {bone1Name=dyn_L_hair_main_4, bone2Name=dyn_L_hair_main_3, swing1LimitDegrees=10, swing2LimitDegrees=10, twistLimitDegrees=0}
 {bone1Name=dyn_R_hair_main_1, bone2Name=base_R_hair_main, swing1LimitDegrees=10, swing2LimitDegrees=10, twistLimitDegrees=0}
 {bone1Name=dyn_R_hair_main_2, bone2Name=dyn_R_hair_main_1, swing1LimitDegrees=10, swing2LimitDegrees=10, twistLimitDegrees=0}
 {bone1Name=dyn_R_hair_main_3, bone2Name=dyn_R_hair_main_2, swing1LimitDegrees=10, swing2LimitDegrees=10, twistLimitDegrees=0}
 {bone1Name=dyn_R_hair_main_4, bone2Name=dyn_R_hair_main_3, swing1LimitDegrees=10, swing2LimitDegrees=10, twistLimitDegrees=0}
 {bone1Name=dyn_L_hair_main_1, bone2Name=base_L_hair_main, swing1LimitDegrees=10, swing2LimitDegrees=10, twistLimitDegrees=0}
 {
  bone1Name=base_L_hair_main
  bone2Name=head
  swing1LimitDegrees=45
  swing1Motion=Locked
  swing2LimitDegrees=45
  swing2Motion=Locked
  twistLimitDegrees=45
  twistMotion=Locked
 }
 {
  bone1Name=base_R_hair_main
  bone2Name=head
  swing1LimitDegrees=45
  swing1Motion=Locked
  swing2LimitDegrees=45
  swing2Motion=Locked
  twistLimitDegrees=45
  twistMotion=Locked
 }
 {
  bone1Name=upperarm_l
  bone2Name=clavicle_l
  swing1LimitDegrees=45
  swing1Motion=Free
  swing2LimitDegrees=45
  swing2Motion=Free
  twistLimitDegrees=45
  twistMotion=Free
 }
 {
  bone1Name=upperarm_r
  bone2Name=clavicle_r
  swing1LimitDegrees=45
  swing1Motion=Free
  swing2LimitDegrees=45
  swing2Motion=Free
  twistLimitDegrees=45
  twistMotion=Free
 }
 {
  bone1Name=clavicle_l
  bone2Name=spine_05
  swing1LimitDegrees=45
  swing1Motion=Free
  swing2LimitDegrees=45
  swing2Motion=Free
  twistLimitDegrees=45
  twistMotion=Free
 }
 {
  bone1Name=clavicle_r
  bone2Name=spine_05
  swing1LimitDegrees=45
  swing1Motion=Free
  swing2LimitDegrees=45
  swing2Motion=Free
  twistLimitDegrees=45
  twistMotion=Free
 }
 {
  bone1Name=dyn_R_hair_main_4
  bone2Name=dyn_L_hair_main_4
  swing1LimitDegrees=45
  swing1Motion=Free
  swing2LimitDegrees=45
  swing2Motion=Free
  twistLimitDegrees=45
  twistMotion=Free
 }
]
