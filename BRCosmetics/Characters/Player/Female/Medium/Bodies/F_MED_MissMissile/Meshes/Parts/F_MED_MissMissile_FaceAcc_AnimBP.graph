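# Post-style AnimBP for Miss Missile's FaceAcc part, adding secondary motion to the left and right hair chains.
# Poses copied from the parent mesh pass through rigid-body physics, weighted bone constraints, AnimDynamics and a Control Rig fed crouch-walk and melee values.
# Cached poses and bool blends switch hair treatment for crouching, emotes and backbling, with additive bone offsets correcting placement.
# A skydiving blend selects between separately tuned rigid-body simulations.
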AnimBlueprint /BRCosmetics/Characters/Player/Female/Medium/Bodies/F_MED_MissMissile/Meshes/Parts/F_MED_MissMissile_FaceAcc_AnimBP
target_skeleton=&/F_MED_MissMissile_FaceAcc_Skeleton
graphs [
 AnimGraph {nodes [
   n1 = Anim Dynamics - Bone: dyn_L_hair_main_1!(ComponentPose, Alpha=1, GravityOverride:Vector="0,0,-980", ExternalForce="0,0,0") -> (Pose) {
    angularDampingOverride=1
    angularSpringConstant=15000
    bAngularSpring=true
    bOverrideAngularDamping=true
    bUseGravityOverride=true
    gravityScale=0.35
    simulationSpace=RootRelative
    boundBone {boneName=dyn_L_hair_main_1}
    componentAppliedLinearAccClamp=<2000 2000 2000>
    componentLinearAccScale=<0.5 0.5 0.5>
    componentLinearVelScale=<0.5 0.5 0.5>
    gravityOverride {z=-980}
    physicsBodyDefinitions [{
      boundBone {boneName=dyn_L_hair_main_1}
      boxExtents=<30 25 25>
      constraintSetup {angularLimitsMax=<0 0 2>, angularLimitsMin=<0 -5 -2>, angularTarget=<1 0 0>}
      localJointOffset=<2 0 0>
     }]
   }
   n2 = Anim Dynamics - Bone: dyn_R_hair_main_1!(ComponentPose=@n1, Alpha=1, GravityOverride:Vector="0,0,-980", ExternalForce="0,0,0") -> (Pose) {
    angularDampingOverride=1
    angularSpringConstant=15000
    bAngularSpring=true
    bOverrideAngularDamping=true
    bUseGravityOverride=true
    gravityScale=0.35
    simulationSpace=RootRelative
    boundBone {boneName=dyn_R_hair_main_1}
    componentAppliedLinearAccClamp=<2000 2000 2000>
    componentLinearAccScale=<0.5 0.5 0.5>
    componentLinearVelScale=<0.5 0.5 0.5>
    gravityOverride {z=-980}
    physicsBodyDefinitions [{
      boundBone {boneName=dyn_R_hair_main_1}
      boxExtents=<30 25 25>
      constraintSetup {angularLimitsMax=<0 0 2>, angularLimitsMin=<0 -5 -2>, angularTarget=<1 0 0>}
      localJointOffset=<2 0 0>
     }]
   }
   n3 = Anim Dynamics - Bone: dyn_R_hair_main_2!(ComponentPose=@n2, Alpha=1, GravityOverride:Vector="0,0,-980", ExternalForce="0,0,0") -> (Pose) {
    angularDampingOverride=1
    angularSpringConstant=15000
    bAngularSpring=true
    bOverrideAngularDamping=true
    bUseGravityOverride=true
    gravityScale=0.35
    simulationSpace=RootRelative
    boundBone {boneName=dyn_R_hair_main_2}
    componentAppliedLinearAccClamp=<2000 2000 2000>
    componentLinearAccScale=<0.5 0.5 0.5>
    componentLinearVelScale=<0.5 0.5 0.5>
    gravityOverride {z=-980}
    physicsBodyDefinitions [{
      boundBone {boneName=dyn_R_hair_main_2}
      boxExtents=<30 25 25>
      constraintSetup {angularLimitsMax=<0 0 3>, angularLimitsMin=<0 -5 -3>, angularTarget=<1 0 0>}
      localJointOffset=<2 0 0>
     }]
   }
   n4 = Anim Dynamics - Bone: dyn_R_hair_main_3!(ComponentPose=@n3, Alpha=1, GravityOverride:Vector="0,0,-980", ExternalForce="0,0,0") -> (Pose) {
    angularDampingOverride=1
    angularSpringConstant=15000
    bAngularSpring=true
    bOverrideAngularDamping=true
    bUseGravityOverride=true
    gravityScale=0.35
    simulationSpace=RootRelative
    boundBone {boneName=dyn_R_hair_main_3}
    componentAppliedLinearAccClamp=<2000 2000 2000>
    componentLinearAccScale=<0.5 0.5 0.5>
    componentLinearVelScale=<0.5 0.5 0.5>
    gravityOverride {z=-980}
    physicsBodyDefinitions [{
      boundBone {boneName=dyn_R_hair_main_3}
      boxExtents=<30 25 25>
      constraintSetup {angularLimitsMax=<0 0 4>, angularLimitsMin=<0 -10 -4>, angularTarget=<1 0 0>}
      localJointOffset=<2 0 0>
     }]
   }
   n5 = Anim Dynamics - Bone: dyn_R_hair_main_4!(ComponentPose=@n4, Alpha=1, GravityOverride:Vector="0,0,-980", ExternalForce="0,0,0") -> (Pose) {
    angularDampingOverride=1
    angularSpringConstant=15000
    bAngularSpring=true
    bOverrideAngularDamping=true
    bUseGravityOverride=true
    gravityScale=0.35
    simulationSpace=RootRelative
    boundBone {boneName=dyn_R_hair_main_4}
    componentAppliedLinearAccClamp=<2000 2000 2000>
    componentLinearAccScale=<0.5 0.5 0.5>
    componentLinearVelScale=<0.5 0.5 0.5>
    gravityOverride {z=-980}
    physicsBodyDefinitions [{
      boundBone {boneName=dyn_R_hair_main_4}
      boxExtents=<30 25 25>
      constraintSetup {angularLimitsMax=<0 0 5>, angularLimitsMin=<0 -15 -5>, angularTarget=<1 0 0>}
      localJointOffset=<2 0 0>
     }]
   }
   n6 = Anim Dynamics - Bone: dyn_L_hair_main_2!(ComponentPose=@n5, Alpha=1, GravityOverride:Vector="0,0,-980", ExternalForce="0,0,0") -> (Pose) {
    angularDampingOverride=1
    angularSpringConstant=15000
    bAngularSpring=true
    bOverrideAngularDamping=true
    bUseGravityOverride=true
    gravityScale=0.35
    simulationSpace=RootRelative
    boundBone {boneName=dyn_L_hair_main_2}
    componentAppliedLinearAccClamp=<2000 2000 2000>
    componentLinearAccScale=<0.5 0.5 0.5>
    componentLinearVelScale=<0.5 0.5 0.5>
    gravityOverride {z=-980}
    physicsBodyDefinitions [{
      boundBone {boneName=dyn_L_hair_main_2}
      boxExtents=<30 25 25>
      constraintSetup {angularLimitsMax=<0 0 3>, angularLimitsMin=<0 -5 -3>, angularTarget=<1 0 0>}
      localJointOffset=<2 0 0>
     }]
   }
   n7 = Anim Dynamics - Bone: dyn_L_hair_main_3!(ComponentPose=@n6, Alpha=1, GravityOverride:Vector="0,0,-980", ExternalForce="0,0,0") -> (Pose) {
    angularDampingOverride=1
    angularSpringConstant=15000
    bAngularSpring=true
    bOverrideAngularDamping=true
    bUseGravityOverride=true
    gravityScale=0.35
    simulationSpace=RootRelative
    boundBone {boneName=dyn_L_hair_main_3}
    componentAppliedLinearAccClamp=<2000 2000 2000>
    componentLinearAccScale=<0.5 0.5 0.5>
    componentLinearVelScale=<0.5 0.5 0.5>
    gravityOverride {z=-980}
    physicsBodyDefinitions [{
      boundBone {boneName=dyn_L_hair_main_3}
      boxExtents=<30 25 25>
      constraintSetup {angularLimitsMax=<0 0 4>, angularLimitsMin=<0 -10 -4>, angularTarget=<1 0 0>}
      localJointOffset=<2 0 0>
     }]
   }
   Anim Dynamics - Bone: dyn_L_hair_main_4!(ComponentPose=@n7, Alpha=1, GravityOverride:Vector="0,0,-980", ExternalForce="0,0,0") -> (Pose) {
    angularDampingOverride=1
    angularSpringConstant=15000
    bAngularSpring=true
    bOverrideAngularDamping=true
    bUseGravityOverride=true
    gravityScale=0.35
    simulationSpace=RootRelative
    boundBone {boneName=dyn_L_hair_main_4}
    componentAppliedLinearAccClamp=<2000 2000 2000>
    componentLinearAccScale=<0.5 0.5 0.5>
    componentLinearVelScale=<0.5 0.5 0.5>
    gravityOverride {z=-980}
    physicsBodyDefinitions [{
      boundBone {boneName=dyn_L_hair_main_4}
      boxExtents=<30 25 25>
      constraintSetup {angularLimitsMax=<0 0 5>, angularLimitsMin=<0 -15 -5>, angularTarget=<1 0 0>}
      localJointOffset=<2 0 0>
     }]
   }
   n8 = "Constraint - dyn_L_hair_main_1 (PPP)"!(
    ComponentPose=Local To Component(LocalPose=Use cached pose 'body'())
    Alpha=0.6
    ConstraintWeights_0=0.1
    ConstraintWeights_1:float=0.7
    ConstraintWeights_2:float=0.2
   ) -> (Pose) {
    boneToModify {boneName=dyn_L_hair_main_1}
    constraintSetup [{targetBone {boneName=head}}, {targetBone {boneName=spine_05}}, {targetBone {boneName=neck_01}}]
    constraintWeights [1,1,1]
   }
   n9 = "Constraint - dyn_L_hair_main_2 (PP)"!(ComponentPose=@n8, Alpha=1, ConstraintWeights_0=0, ConstraintWeights_1:float=1) -> (Pose) {
    boneToModify {boneName=dyn_L_hair_main_2}
    constraintSetup [{targetBone {boneName=clavicle_r}}, {targetBone {boneName=spine_05}}]
    constraintWeights [1,1]
   }
   n10 = "Constraint - dyn_L_hair_main_3 (PPP)"!(ComponentPose=@n9, Alpha=1, ConstraintWeights_0=0, ConstraintWeights_1:float=1, ConstraintWeights_2:float=0) -> (Pose) {
    boneToModify {boneName=dyn_L_hair_main_3}
    constraintSetup [{targetBone {boneName=head}}, {targetBone {boneName=spine_05}}, {targetBone {boneName=clavicle_l}}]
    constraintWeights [1,1,1]
   }
   n11 = "Constraint - dyn_L_hair_main_4 (P)"(ComponentPose=@n10, Alpha=1, ConstraintWeights_0=1) {boneToModify {boneName=dyn_L_hair_main_4}, constraintSetup [{targetBone {boneName=spine_05}}], constraintWeights [1]}
   n12 = "Constraint - dyn_R_hair_main_1 (PPP)"!(ComponentPose=@n11, Alpha=0.6, ConstraintWeights_0=0.1, ConstraintWeights_1:float=0.7, ConstraintWeights_2:float=0.2) -> (Pose) {
    boneToModify {boneName=dyn_R_hair_main_1}
    constraintSetup [{targetBone {boneName=head}}, {targetBone {boneName=spine_05}}, {targetBone {boneName=neck_01}}]
    constraintWeights [1,1,1]
   }
   n13 = "Constraint - dyn_R_hair_main_2 (PP)"!(ComponentPose=@n12, Alpha=1, ConstraintWeights_0=0, ConstraintWeights_1:float=1) -> (Pose) {
    boneToModify {boneName=dyn_R_hair_main_2}
    constraintSetup [{targetBone {boneName=head}}, {targetBone {boneName=spine_05}}]
    constraintWeights [1,1]
   }
   n14 = "Constraint - dyn_R_hair_main_3 (PPP)"!(ComponentPose=@n13, Alpha=1, ConstraintWeights_0=0, ConstraintWeights_1:float=1, ConstraintWeights_2:float=0) -> (Pose) {
    boneToModify {boneName=dyn_R_hair_main_3}
    constraintSetup [{targetBone {boneName=head}}, {targetBone {boneName=spine_05}}, {targetBone {boneName=clavicle_r}}]
    constraintWeights [1,1,1]
   }
   n15 = "Constraint - dyn_R_hair_main_4 (P)"(ComponentPose=@n14, Alpha=1, ConstraintWeights_0=1) {boneToModify {boneName=dyn_R_hair_main_4}, constraintSetup [{targetBone {boneName=spine_05}}], constraintWeights [1]}
   n16 = Component To Local(ComponentPose=@n15)
   n17 = Use cached pose 'body'()
   n18 = Get isWearingBackblingBool() -> (isWearingBackblingBool:bool)
   n19 = RigidBody!(
    ComponentPose=Local To Component(LocalPose=Blend Poses by bool(BlendPose_0=@n16, BlendPose_1=@n17, bActiveValue=@n18))
    Alpha
    OverrideWorldGravity:Vector="0,0,0"
    ExternalForce
   ) -> (Pose) {
    bOverrideWorldGravity=true
    overridePhysicsAsset=&/F_MED_MissMissile_FaceAcc_Physics
    simulationSpace=BaseBoneSpace
    baseBoneRef {boneName=spine_05}
    simSpaceSettings {velocityScaleZ=0.2, worldAlpha=1}
   }
   Component To Local(ComponentPose=@n19)
   n20 = "Constraint - dyn_Const_Jnt_1 (PPP)"!(ComponentPose, Alpha=1, ConstraintWeights_0=0.4359, ConstraintWeights_1:float=0.3971, ConstraintWeights_2:float=0.167) -> (Pose) {
    boneToModify {boneName=dyn_Const_Jnt_1}
    constraintSetup [{targetBone {boneName=neck_02}}, {targetBone {boneName=spine_05}}, {targetBone {boneName=neck_01}}]
    constraintWeights [1,1,1]
   }
   n21 = "Constraint - dyn_Const_Jnt_2 (PPP)"!(ComponentPose=@n20, Alpha=1, ConstraintWeights_0=0.4325, ConstraintWeights_1:float=0.4011, ConstraintWeights_2:float=0.1664) -> (Pose) {
    boneToModify {boneName=dyn_Const_Jnt_2}
    constraintSetup [{targetBone {boneName=neck_02}}, {targetBone {boneName=spine_05}}, {targetBone {boneName=neck_01}}]
    constraintWeights [1,1,1]
   }
   n22 = "Constraint - dyn_Const_Jnt_3 (PPP)"!(ComponentPose=@n21, Alpha=1, ConstraintWeights_0=0.7324, ConstraintWeights_1:float=0.1543, ConstraintWeights_2:float=0.1132) -> (Pose) {
    boneToModify {boneName=dyn_Const_Jnt_3}
    constraintSetup [{targetBone {boneName=spine_05}}, {targetBone {boneName=neck_02}}, {targetBone {boneName=neck_01}}]
    constraintWeights [1,1,1]
   }
   "Constraint - dyn_Const_Jnt_4 (PP)"!(ComponentPose=@n22, Alpha=1, ConstraintWeights_0=0.8835, ConstraintWeights_1:float=0.1165) -> (Pose) {
    boneToModify {boneName=dyn_Const_Jnt_4}
    constraintSetup [{targetBone {boneName=spine_05}}, {targetBone {boneName=spine_04}}]
    constraintWeights [1,1]
   }
   n23 = "Constraint - dyn_Const_Jnt_1 (PPP)"(Alpha=1, ConstraintWeights_0=0.4359, ConstraintWeights_1=0.3971, ConstraintWeights_2=0.167) {
    boneToModify {boneName=dyn_Const_Jnt_1}
    constraintSetup [{targetBone {boneName=neck_02}}, {targetBone {boneName=spine_05}}, {targetBone {boneName=neck_01}}]
    constraintWeights [1,1,1]
   }
   n24 = "Constraint - dyn_Const_Jnt_2 (PPP)"(ComponentPose=@n23, Alpha=1, ConstraintWeights_0=0.4325, ConstraintWeights_1=0.4011, ConstraintWeights_2=0.1664) {
    boneToModify {boneName=dyn_Const_Jnt_2}
    constraintSetup [{targetBone {boneName=neck_02}}, {targetBone {boneName=spine_05}}, {targetBone {boneName=neck_01}}]
    constraintWeights [1,1,1]
   }
   n25 = "Constraint - dyn_Const_Jnt_3 (PPP)"(ComponentPose=@n24, Alpha=1, ConstraintWeights_0=0.7324, ConstraintWeights_1=0.1543, ConstraintWeights_2=0.1132) {
    boneToModify {boneName=dyn_Const_Jnt_3}
    constraintSetup [{targetBone {boneName=spine_05}}, {targetBone {boneName=neck_02}}, {targetBone {boneName=neck_01}}]
    constraintWeights [1,1,1]
   }
   "Constraint - dyn_Const_Jnt_4 (PP)"(ComponentPose=@n25, Alpha=1, ConstraintWeights_0=0.8835, ConstraintWeights_1=0.1165) {
    boneToModify {boneName=dyn_Const_Jnt_4}
    constraintSetup [{targetBone {boneName=spine_05}}, {targetBone {boneName=spine_04}}]
    constraintWeights [1,1]
   }
   Control Rig() {controlRigAssetReference=~}
   Input Pose()
   n26 = "Constraint - dyn_L_hair_main_1 (PPP)"(
    ComponentPose=Local To Component(LocalPose=Use cached pose 'body'())
    Alpha=1
    ConstraintWeights_0=0.2
    ConstraintWeights_1=0.3
    ConstraintWeights_2=0.5
   ) {
    boneToModify {boneName=dyn_L_hair_main_1}
    constraintSetup [{targetBone {boneName=head}}, {targetBone {boneName=spine_05}}, {targetBone {boneName=neck_01}}]
    constraintWeights [1,1,1]
   }
   n27 = "Constraint - dyn_L_hair_main_2 (PP)"(ComponentPose=@n26, Alpha=1, ConstraintWeights_0=0, ConstraintWeights_1=1) {
    boneToModify {boneName=dyn_L_hair_main_2}
    constraintSetup [{targetBone {boneName=clavicle_r}}, {targetBone {boneName=spine_05}}]
    constraintWeights [1,1]
   }
   n28 = "Constraint - dyn_L_hair_main_3 (PPP)"(ComponentPose=@n27, Alpha=1, ConstraintWeights_0=0, ConstraintWeights_1=1, ConstraintWeights_2=0) {
    boneToModify {boneName=dyn_L_hair_main_3}
    constraintSetup [{targetBone {boneName=head}}, {targetBone {boneName=spine_05}}, {targetBone {boneName=clavicle_l}}]
    constraintWeights [1,1,1]
   }
   n29 = "Constraint - dyn_L_hair_main_4 (P)"(ComponentPose=@n28, Alpha=1, ConstraintWeights_0=1) {boneToModify {boneName=dyn_L_hair_main_4}, constraintSetup [{targetBone {boneName=spine_05}}], constraintWeights [1]}
   n30 = "Constraint - dyn_R_hair_main_1 (PPP)"(ComponentPose=@n29, Alpha=1, ConstraintWeights_0=0.2, ConstraintWeights_1=0.3, ConstraintWeights_2=0.5) {
    boneToModify {boneName=dyn_R_hair_main_1}
    constraintSetup [{targetBone {boneName=head}}, {targetBone {boneName=spine_05}}, {targetBone {boneName=neck_01}}]
    constraintWeights [1,1,1]
   }
   n31 = "Constraint - dyn_R_hair_main_2 (PP)"(ComponentPose=@n30, Alpha=1, ConstraintWeights_0=0, ConstraintWeights_1=1) {
    boneToModify {boneName=dyn_R_hair_main_2}
    constraintSetup [{targetBone {boneName=head}}, {targetBone {boneName=spine_05}}]
    constraintWeights [1,1]
   }
   n32 = "Constraint - dyn_R_hair_main_3 (PPP)"(ComponentPose=@n31, Alpha=1, ConstraintWeights_0=0, ConstraintWeights_1=1, ConstraintWeights_2=0) {
    boneToModify {boneName=dyn_R_hair_main_3}
    constraintSetup [{targetBone {boneName=head}}, {targetBone {boneName=spine_05}}, {targetBone {boneName=clavicle_r}}]
    constraintWeights [1,1,1]
   }
   n33 = "Constraint - dyn_R_hair_main_4 (P)"(ComponentPose=@n32, Alpha=1, ConstraintWeights_0=1) {boneToModify {boneName=dyn_R_hair_main_4}, constraintSetup [{targetBone {boneName=spine_05}}], constraintWeights [1]}
   n34 = "Transform (Modify) Bone - Bone: base_R_hair_main"(ComponentPose=@n33, Alpha=1, Translation="-2.384,-0.000,2.629", Rotation="-1.986,0.000,-0.000", Scale="1,1,1") {
    rotationMode=BMM_Additive
    rotationSpace=BCS_BoneSpace
    translationMode=BMM_Additive
    translationSpace=BCS_BoneSpace
    boneToModify {boneName=base_R_hair_main}
    rotation=<rot -1.986 0 0>
    translation {x=-2.384, z=2.629}
   }
   n35 = "Transform (Modify) Bone - Bone: base_L_hair_main"(ComponentPose=@n34, Alpha=1, Translation="-1.654,-0.000,1.788", Rotation="-1.773,0.000,-0.000", Scale="1,1,1") {
    rotationMode=BMM_Additive
    rotationSpace=BCS_BoneSpace
    translationMode=BMM_Additive
    translationSpace=BCS_BoneSpace
    boneToModify {boneName=base_L_hair_main}
    rotation=<rot -1.7731 0 0>
    translation {x=-1.654, z=1.788}
   }
   n36 = Component To Local(ComponentPose=@n35)
   n37 = Use cached pose 'body'()
   n38 = Get isWearingBackblingBool()
   n39 = "Transform (Modify) Bone - Bone: dyn_L_hair_main_2"(
    ComponentPose=Local To Component(LocalPose=Blend Poses by bool(BlendPose_0=@n36, BlendPose_1=@n37, bActiveValue=@n38))
    Alpha=1
    Translation="0,0,0"
    Rotation="-26.115,-3.503,11.121"
    Scale="1,1,1"
   ) {
    rotationMode=BMM_Additive
    rotationSpace=BCS_BoneSpace
    translationMode=BMM_Additive
    translationSpace=BCS_BoneSpace
    boneToModify {boneName=dyn_L_hair_main_2}
    rotation=<rot -26.115 -3.503 11.121>
   }
   n40 = "Transform (Modify) Bone - Bone: dyn_R_hair_main_2"(ComponentPose=@n39, Alpha=1, Translation="-0.152,0.085,-0.322", Rotation="-23.666,6.743,-16.413", Scale="1,1,1") {
    rotationMode=BMM_Additive
    rotationSpace=BCS_BoneSpace
    translationMode=BMM_Additive
    translationSpace=BCS_BoneSpace
    boneToModify {boneName=dyn_R_hair_main_2}
    rotation=<rot -23.666 6.743 -16.413>
    translation=<-0.152 0.085 -0.322>
   }
   Output Pose(Result=Component To Local(ComponentPose=@n40)) {layerGroup=None}
   n41 = RigidBody(ComponentPose=Local To Component(LocalPose=Copy Pose From Mesh() {bUseAttachedParent=true}), OverrideWorldGravity="0,0,0") {
    bOverrideWorldGravity=true
    overridePhysicsAsset=&/F_MED_MissMissile_FaceAcc_BackHairPhysics
    simulationSpace=BaseBoneSpace
    baseBoneRef {boneName=spine_05}
    simSpaceSettings {
     maxAngularAcceleration=500
     maxAngularVelocity=500
     maxLinearAcceleration=500
     maxLinearVelocity=500
     worldAlpha=1
     externalAngularVelocity=<1 1 1>
     externalLinearDragV=<1 1 1>
     externalLinearVelocity=<1 1 1>
    }
   }
   Save cached pose 'PreAnim'(Pose=Component To Local(ComponentPose=@n41)) {cacheName=PreAnim}
   Save cached pose 'Pose'(Pose=Copy Pose From Mesh(SourceMeshComponent=Get Skel Mesh()) {bUseAttachedParent=true}) {cacheName=Pose}
   n42 = "Constraint - dyn_L_hair_main_1 (PPPP)"!(
    ComponentPose=Local To Component(LocalPose=Use cached pose 'PreAnim'())
    Alpha=1
    ConstraintWeights_0=0.1
    ConstraintWeights_1:float=0.9
    ConstraintWeights_2:float=0
    ConstraintWeights_3:float=0
   ) -> (Pose) {
    boneToModify {boneName=dyn_L_hair_main_1}
    constraintSetup [{targetBone {boneName=head}}, {targetBone {boneName=spine_05}}, {targetBone {boneName=neck_01}}, {targetBone {boneName=clavicle_l}}]
    constraintWeights [1,1,1,1]
   }
   n43 = "Constraint - dyn_L_hair_main_2 (PPPP)"!(
    ComponentPose=@n42
    Alpha=1
    ConstraintWeights_0=0.9
    ConstraintWeights_1:float=0
    ConstraintWeights_2:float=0
    ConstraintWeights_3:float=0.1
   ) -> (Pose) {
    boneToModify {boneName=dyn_L_hair_main_2}
    constraintSetup [{targetBone {boneName=spine_05}}, {targetBone {boneName=neck_02}}, {targetBone {boneName=neck_01}}, {targetBone {boneName=clavicle_l}}]
    constraintWeights [1,1,1,1]
   }
   n44 = "Constraint - dyn_L_hair_main_3 (PPP)"(ComponentPose=@n43, Alpha=1, ConstraintWeights_0=0.5, ConstraintWeights_1=0.5, ConstraintWeights_2=0) {
    boneToModify {boneName=dyn_L_hair_main_3}
    constraintSetup [{targetBone {boneName=clavicle_l}}, {targetBone {boneName=spine_05}}, {targetBone {boneName=neck_01}}]
    constraintWeights [1,1,1]
   }
   n45 = "Constraint - dyn_L_hair_main_4 (PPPPP)"!(
    ComponentPose=@n44
    Alpha=1
    ConstraintWeights_0=0.5
    ConstraintWeights_1:float=0.5
    ConstraintWeights_2:float=0
    ConstraintWeights_3:float=0
    ConstraintWeights_4:float=0
   ) -> (Pose) {
    boneToModify {boneName=dyn_L_hair_main_4}
    constraintSetup [
     {targetBone {boneName=clavicle_l}}
     {targetBone {boneName=spine_05}}
     {targetBone {boneName=spine_04}}
     {targetBone {boneName=upperarm_twist_01_l}}
     {targetBone {boneName=deform_upperarm_bk_l}}
    ]
    constraintWeights [1,1,1,1,1]
   }
   n46 = "Constraint - dyn_R_hair_main_1 (PPPP)"!(
    ComponentPose=@n45
    Alpha=1
    ConstraintWeights_0=0.1
    ConstraintWeights_1:float=0.9
    ConstraintWeights_2:float=0
    ConstraintWeights_3:float=0
   ) -> (Pose) {
    boneToModify {boneName=dyn_R_hair_main_1}
    constraintSetup [{targetBone {boneName=head}}, {targetBone {boneName=spine_05}}, {targetBone {boneName=clavicle_r}}, {targetBone {boneName=neck_01}}]
    constraintWeights [1,1,1,1]
   }
   n47 = "Constraint - dyn_R_hair_main_2 (PPPP)"!(
    ComponentPose=@n46
    Alpha=1
    ConstraintWeights_0=0.9
    ConstraintWeights_1:float=0
    ConstraintWeights_2:float=0
    ConstraintWeights_3:float=0.1
   ) -> (Pose) {
    boneToModify {boneName=dyn_R_hair_main_2}
    constraintSetup [{targetBone {boneName=spine_05}}, {targetBone {boneName=neck_02}}, {targetBone {boneName=neck_01}}, {targetBone {boneName=clavicle_r}}]
    constraintWeights [1,1,1,1]
   }
   n48 = "Constraint - dyn_R_hair_main_3 (PPP)"(ComponentPose=@n47, Alpha=1, ConstraintWeights_0=0.5, ConstraintWeights_1=0.5, ConstraintWeights_2=0) {
    boneToModify {boneName=dyn_R_hair_main_3}
    constraintSetup [{targetBone {boneName=clavicle_r}}, {targetBone {boneName=spine_05}}, {targetBone {boneName=deform_upperarm_up_r}}]
    constraintWeights [1,1,1]
   }
   n49 = "Constraint - dyn_R_hair_main_4 (PPPPP)"!(
    ComponentPose=@n48
    Alpha=1
    ConstraintWeights_0=0.2
    ConstraintWeights_1:float=0.8
    ConstraintWeights_2:float=0
    ConstraintWeights_3:float=0
    ConstraintWeights_4:float=0
   ) -> (Pose) {
    boneToModify {boneName=dyn_R_hair_main_4}
    constraintSetup [
     {targetBone {boneName=clavicle_r}}
     {targetBone {boneName=spine_05}}
     {targetBone {boneName=spine_04}}
     {targetBone {boneName=upperarm_twist_01_r}}
     {targetBone {boneName=deform_upperarm_bk_r}}
    ]
    constraintWeights [1,1,1,1,1]
   }
   n50 = Anim Dynamics - Bone: dyn_R_hair_main_1(ComponentPose=@n49, Alpha=0.8, GravityOverride="0,0,-980", ExternalForce="0,0,0") {
    angularDampingOverride=0.9999
    angularSpringConstant=8000
    bAngularSpring=true
    bOverrideAngularDamping=true
    bUseGravityOverride=true
    gravityScale=0.35
    simulationSpace=RootRelative
    boundBone {boneName=dyn_R_hair_main_1}
    componentAppliedLinearAccClamp=<4000 4000 4000>
    componentLinearAccScale=<1 1 1>
    componentLinearVelScale=<1 1 1>
    gravityOverride {z=-980}
    physicsBodyDefinitions [{
      boundBone {boneName=dyn_R_hair_main_1}
      boxExtents=<20 30 30>
      constraintSetup {angularLimitsMax=<0 0 4>, angularLimitsMin=<0 0 -4>, angularTarget=<1 0 0>}
      localJointOffset=<2 0 0>
     }]
   }
   n51 = Anim Dynamics - Bone: dyn_R_hair_main_2(ComponentPose=@n50, Alpha=1, GravityOverride="0,0,-980", ExternalForce="0,0,0") {
    angularDampingOverride=0.9999
    angularSpringConstant=8000
    bAngularSpring=true
    bOverrideAngularDamping=true
    bUseGravityOverride=true
    gravityScale=0.35
    simulationSpace=RootRelative
    boundBone {boneName=dyn_R_hair_main_2}
    componentAppliedLinearAccClamp=<4000 4000 4000>
    componentLinearAccScale=<1 1 1>
    componentLinearVelScale=<1 1 1>
    gravityOverride {z=-980}
    physicsBodyDefinitions [{
      boundBone {boneName=dyn_R_hair_main_2}
      boxExtents=<20 30 30>
      constraintSetup {angularLimitsMax=<0 0 9>, angularLimitsMin=<0 0 -9>, angularTarget=<1 0 0>}
      localJointOffset=<4 0 0>
     }]
   }
   n52 = Anim Dynamics - Bone: dyn_R_hair_main_3(ComponentPose=@n51, Alpha=1, GravityOverride="0,0,-980", ExternalForce="0,0,0") {
    angularDampingOverride=0.9999
    angularSpringConstant=8000
    bAngularSpring=true
    bOverrideAngularDamping=true
    bUseGravityOverride=true
    gravityScale=0.35
    simulationSpace=RootRelative
    boundBone {boneName=dyn_R_hair_main_3}
    componentAppliedLinearAccClamp=<4000 4000 4000>
    componentLinearAccScale=<1 1 1>
    componentLinearVelScale=<1 1 1>
    gravityOverride {z=-980}
    physicsBodyDefinitions [{
      boundBone {boneName=dyn_R_hair_main_3}
      boxExtents=<20 30 30>
      constraintSetup {angularLimitsMax=<0 0 9>, angularLimitsMin=<0 0 -9>, angularTarget=<1 0 0>}
      localJointOffset=<4 0 0>
     }]
   }
   n53 = Anim Dynamics - Bone: dyn_R_hair_main_4(ComponentPose=@n52, Alpha=1, GravityOverride="0,0,-980", ExternalForce="0,0,0") {
    angularDampingOverride=0.9999
    angularSpringConstant=8000
    bAngularSpring=true
    bOverrideAngularDamping=true
    bUseGravityOverride=true
    gravityScale=0.35
    simulationSpace=RootRelative
    boundBone {boneName=dyn_R_hair_main_4}
    componentAppliedLinearAccClamp=<4000 4000 4000>
    componentLinearAccScale=<1 1 1>
    componentLinearVelScale=<1 1 1>
    gravityOverride {z=-980}
    physicsBodyDefinitions [{
      boundBone {boneName=dyn_R_hair_main_4}
      boxExtents=<20 30 30>
      constraintSetup {angularLimitsMax=<0 0 9>, angularLimitsMin=<0 0 -9>, angularTarget=<1 0 0>}
      localJointOffset=<4 0 0>
     }]
   }
   n54 = Anim Dynamics - Bone: dyn_L_hair_main_1(ComponentPose=@n53, Alpha=0.8, GravityOverride="0,0,-980", ExternalForce="0,0,0") {
    angularDampingOverride=0.9999
    angularSpringConstant=8000
    bAngularSpring=true
    bOverrideAngularDamping=true
    bUseGravityOverride=true
    gravityScale=0.35
    simulationSpace=RootRelative
    boundBone {boneName=dyn_L_hair_main_1}
    componentAppliedLinearAccClamp=<4000 4000 4000>
    componentLinearAccScale=<1 1 1>
    componentLinearVelScale=<1 1 1>
    gravityOverride {z=-980}
    physicsBodyDefinitions [{
      boundBone {boneName=dyn_L_hair_main_1}
      boxExtents=<20 30 30>
      constraintSetup {angularLimitsMax=<0 0 4>, angularLimitsMin=<0 0 -4>, angularTarget=<1 0 0>}
      localJointOffset=<2 0 0>
     }]
   }
   n55 = Anim Dynamics - Bone: dyn_L_hair_main_2(ComponentPose=@n54, Alpha=1, GravityOverride="0,0,-980", ExternalForce="0,0,0") {
    angularDampingOverride=0.9999
    angularSpringConstant=8000
    bAngularSpring=true
    bOverrideAngularDamping=true
    bUseGravityOverride=true
    gravityScale=0.35
    simulationSpace=RootRelative
    boundBone {boneName=dyn_L_hair_main_2}
    componentAppliedLinearAccClamp=<4000 4000 4000>
    componentLinearAccScale=<1 1 1>
    componentLinearVelScale=<1 1 1>
    gravityOverride {z=-980}
    physicsBodyDefinitions [{
      boundBone {boneName=dyn_L_hair_main_2}
      boxExtents=<20 30 30>
      constraintSetup {angularLimitsMax=<0 0 9>, angularLimitsMin=<0 0 -9>, angularTarget=<1 0 0>}
      localJointOffset=<4 0 0>
     }]
   }
   n56 = Anim Dynamics - Bone: dyn_L_hair_main_3(ComponentPose=@n55, Alpha=1, GravityOverride="0,0,-980", ExternalForce="0,0,0") {
    angularDampingOverride=0.9999
    angularSpringConstant=8000
    bAngularSpring=true
    bOverrideAngularDamping=true
    bUseGravityOverride=true
    gravityScale=0.35
    simulationSpace=RootRelative
    boundBone {boneName=dyn_L_hair_main_3}
    componentAppliedLinearAccClamp=<4000 4000 4000>
    componentLinearAccScale=<1 1 1>
    componentLinearVelScale=<1 1 1>
    gravityOverride {z=-980}
    physicsBodyDefinitions [{
      boundBone {boneName=dyn_L_hair_main_3}
      boxExtents=<20 30 30>
      constraintSetup {angularLimitsMax=<0 0 9>, angularLimitsMin=<0 0 -9>, angularTarget=<1 0 0>}
      localJointOffset=<4 0 0>
     }]
   }
   n57 = Anim Dynamics - Bone: dyn_L_hair_main_4(ComponentPose=@n56, Alpha=1, GravityOverride="0,0,-980", ExternalForce="0,0,0") {
    angularDampingOverride=0.9999
    angularSpringConstant=8000
    bAngularSpring=true
    bOverrideAngularDamping=true
    bUseGravityOverride=true
    gravityScale=0.35
    simulationSpace=RootRelative
    boundBone {boneName=dyn_L_hair_main_4}
    componentAppliedLinearAccClamp=<4000 4000 4000>
    componentLinearAccScale=<1 1 1>
    componentLinearVelScale=<1 1 1>
    gravityOverride {z=-980}
    physicsBodyDefinitions [{
      boundBone {boneName=dyn_L_hair_main_4}
      boxExtents=<20 30 30>
      constraintSetup {angularLimitsMax=<0 0 9>, angularLimitsMin=<0 0 -9>, angularTarget=<1 0 0>}
      localJointOffset=<4 0 0>
     }]
   }
   n58 = Control Rig!(
    Source=Component To Local(ComponentPose=@n57)
    Alpha
    CrouchHeadUP walk:bool=Get CrouchWalk() -> (CrouchWalk:bool)
    Pickaxe:double=Get PlayingMeleeAnim() -> (PlayingMeleeAnim:double)
   ) -> (Pose) {controlRigAssetReference {blueprintRigClass=&/F_MED_MissMissile_FaceAcc_CtrlRig_Post.F_MED_MissMissile_FaceAcc_CtrlRig_Post_C}}
   n59 = Use cached pose 'PreAnim'()
   n60 = Get bIsCrouching()
   n61 = "Transform (Modify) Bone - Bone: dyn_L_hair_main_1"!(
    ComponentPose=Local To Component(LocalPose=Blend Poses by bool(BlendPose_0=@n58, BlendPose_1=@n59, bActiveValue=@n60))
    bAlphaBoolEnabled:bool=True
    Translation="0,0,0"
    Rotation="10,0,0"
    Scale="1,1,1"
   ) -> (Pose) {
    alphaInputType=Bool
    rotationMode=BMM_Additive
    rotationSpace=BCS_BoneSpace
    translationMode=BMM_Additive
    translationSpace=BCS_BoneSpace
    boneToModify {boneName=dyn_L_hair_main_1}
    rotation {pitch=10}
   }
   n62 = "Transform (Modify) Bone - Bone: dyn_R_hair_main_1"!(ComponentPose=@n61, bAlphaBoolEnabled:bool=True, Translation="0,0,0", Rotation="10,0,0", Scale="1,1,1") -> (Pose) {
    alphaInputType=Bool
    rotationMode=BMM_Additive
    rotationSpace=BCS_BoneSpace
    translationMode=BMM_Additive
    translationSpace=BCS_BoneSpace
    boneToModify {boneName=dyn_R_hair_main_1}
    rotation {pitch=10}
   }
   Save cached pose 'Crouch Output'(Pose=Component To Local(ComponentPose=@n62)) {cacheName=Crouch Output}
   n63 = RigidBody(ComponentPose=Local To Component(LocalPose=Use cached pose 'Crouch Output'()), OverrideWorldGravity="0,0,0") {
    bOverrideWorldGravity=true
    overridePhysicsAsset=&F_MED_MissMissile_FaceAcc_BackHairPhysics
    simulationSpace=BaseBoneSpace
    baseBoneRef {boneName=spine_05}
    simSpaceSettings {
     maxAngularAcceleration=500
     maxAngularVelocity=500
     maxLinearAcceleration=500
     maxLinearVelocity=500
     worldAlpha=1
     externalAngularVelocity=<1 1 1>
     externalLinearDragV=<1 1 1>
     externalLinearVelocity=<1 1 1>
    }
   }
   n64 = "Transform (Modify) Bone - Bone: dyn_R_hair_main_1"(ComponentPose=@n63, bAlphaBoolEnabled=True, Translation="-1.957,0.000,7.668", Rotation="-20.275,0.000,-0.000", Scale="1,1,1") {
    alphaInputType=Bool
    rotationMode=BMM_Additive
    translationMode=BMM_Additive
    alphaBoolBlend {blendInTime=10, blendOutTime=10}
    alphaScaleBiasClamp {bInterpResult=true}
    boneToModify {boneName=dyn_R_hair_main_1}
    rotation {pitch=-20.275}
    translation {x=-1.957, z=7.668}
   }
   n65 = "Transform (Modify) Bone - Bone: dyn_R_hair_main_2"!(ComponentPose=@n64, bAlphaBoolEnabled:bool=True, Translation="0,0,0", Rotation="6.957,0.000,0.000", Scale="1,1,1") -> (Pose) {
    alphaInputType=Bool
    rotationMode=BMM_Additive
    translationMode=BMM_Additive
    alphaBoolBlend {blendInTime=10, blendOutTime=10}
    alphaScaleBiasClamp {bInterpResult=true}
    boneToModify {boneName=dyn_R_hair_main_2}
    rotation {pitch=6.957}
   }
   n66 = "Transform (Modify) Bone - Bone: dyn_R_hair_main_3"!(ComponentPose=@n65, bAlphaBoolEnabled:bool=True, Translation="0,0,0", Rotation="16.917,0.000,0.000", Scale="1,1,1") -> (Pose) {
    alphaInputType=Bool
    rotationMode=BMM_Additive
    translationMode=BMM_Additive
    alphaBoolBlend {blendInTime=10, blendOutTime=10}
    alphaScaleBiasClamp {bInterpResult=true}
    boneToModify {boneName=dyn_R_hair_main_3}
    rotation {pitch=16.917}
   }
   n67 = "Transform (Modify) Bone - Bone: dyn_L_hair_main_1"(ComponentPose=@n66, bAlphaBoolEnabled=True, Translation="-2.501,-0.022,5.145", Rotation="-9.777,-3.290,7.385", Scale="1,1,1") {
    alphaInputType=Bool
    rotationMode=BMM_Additive
    translationMode=BMM_Additive
    alphaBoolBlend {blendInTime=10, blendOutTime=10}
    alphaScaleBiasClamp {bInterpResult=true}
    boneToModify {boneName=dyn_L_hair_main_1}
    rotation=<rot -9.777 -3.29 7.385>
    translation=<-2.501 -0.022 5.145>
   }
   n68 = "Transform (Modify) Bone - Bone: dyn_L_hair_main_2"!(ComponentPose=@n67, bAlphaBoolEnabled:bool=True, Translation="0,0,0", Rotation="-3.651,0.023,0.041", Scale="1,1,1") -> (Pose) {
    alphaInputType=Bool
    rotationMode=BMM_Additive
    translationMode=BMM_Additive
    alphaBoolBlend {blendInTime=10, blendOutTime=10}
    alphaScaleBiasClamp {bInterpResult=true}
    boneToModify {boneName=dyn_L_hair_main_2}
    rotation=<rot -3.651 0.023 0.041>
   }
   n69 = "Transform (Modify) Bone - Bone: dyn_L_hair_main_3"!(ComponentPose=@n68, bAlphaBoolEnabled:bool=True, Translation="0,0,0", Rotation="22.594,0.000,0.000", Scale="1,1,1") -> (Pose) {
    alphaInputType=Bool
    rotationMode=BMM_Additive
    translationMode=BMM_Additive
    alphaBoolBlend {blendInTime=10, blendOutTime=10}
    alphaScaleBiasClamp {bInterpResult=true}
    boneToModify {boneName=dyn_L_hair_main_3}
    rotation {pitch=22.594}
   }
   Save cached pose 'body'(Pose=Blend Poses by bool(BlendPose_0=Component To Local(ComponentPose=@n69), BlendPose_1=Use cached pose 'Crouch Output'(), bActiveValue=Get bIsPlayingEmote())) {cacheName=body}
   n70 = RigidBody(ComponentPose=Local To Component(LocalPose=Use cached pose 'Pose'()), OverrideWorldGravity="0,0,0") {
    bOverrideWorldGravity=true
    overridePhysicsAsset=&/F_MED_MissMissile_FaceAcc_Physics_Skydive
    simulationSpace=BaseBoneSpace
    baseBoneRef {boneName=spine_05}
    simSpaceSettings {maxAngularAcceleration=500, maxAngularVelocity=500, maxLinearAcceleration=500, maxLinearVelocity=500, worldAlpha=1}
   }
   n71 = Component To Local(ComponentPose=@n70)
   n72 = RigidBody(ComponentPose=Local To Component(LocalPose=Use cached pose 'Pose'()), OverrideWorldGravity="0,0,0") {
    bOverrideWorldGravity=true
    overridePhysicsAsset=&F_MED_MissMissile_FaceAcc_BackHairPhysics
    simulationSpace=BaseBoneSpace
    baseBoneRef {boneName=spine_05}
    simSpaceSettings {
     dampingAlpha=0.8
     maxAngularAcceleration=500
     maxAngularVelocity=500
     maxLinearAcceleration=500
     maxLinearVelocity=500
     velocityScaleZ=0.6
     worldAlpha=0.5
     externalAngularVelocity=<1 1 1>
     externalLinearDragV=<1 1 1>
     externalLinearVelocity=<1 1 1>
    }
   }
   n73 = Component To Local(ComponentPose=@n72)
   n74 = Get bIsSkyDiving()
   n75 = Anim Dynamics - Bone: dyn_L_hair_main_1!(
    ComponentPose=Local To Component(LocalPose=Blend Poses by bool(BlendPose_0=@n71, BlendPose_1=@n73, bActiveValue=@n74))
    Alpha=1
    GravityOverride="0,0,-980"
    SimSpaceSettings:Anim Phys Sim Space Settings Structure=(SimSpaceAngularAlpha=0,MaxAngularVelocity=10000,MaxAngularAcceleration=10000,ExternalAngularVelocity=(X=0,Y=0,Z=0))
    ExternalForce="0,0,0"
   ) -> (Pose) {
    angularDampingOverride=0.9999
    angularSpringConstant=12000
    bAngularSpring=true
    bOverrideAngularDamping=true
    bUseGravityOverride=true
    gravityScale=0.35
    simulationSpace=RootRelative
    boundBone {boneName=dyn_L_hair_main_1}
    componentAppliedLinearAccClamp=<2000 2000 2000>
    componentLinearAccScale=<0.5 0.5 0.5>
    componentLinearVelScale=<0.5 0.5 0.5>
    gravityOverride {z=-980}
    physicsBodyDefinitions [{
      boundBone {boneName=dyn_L_hair_main_1}
      boxExtents=<30 25 25>
      constraintSetup {angularLimitsMax=<0 0 10>, angularLimitsMin=<0 -5 -10>, angularTarget=<1 0 0>}
      localJointOffset=<4 0 0>
     }]
   }
   n76 = Anim Dynamics - Bone: dyn_R_hair_main_1!(
    ComponentPose=@n75
    Alpha=1
    GravityOverride="0,0,-980"
    SimSpaceSettings:Anim Phys Sim Space Settings Structure=(SimSpaceAngularAlpha=0,MaxAngularVelocity=10000,MaxAngularAcceleration=10000,ExternalAngularVelocity=(X=0,Y=0,Z=0))
    ExternalForce="0,0,0"
   ) -> (Pose) {
    angularDampingOverride=0.9999
    angularSpringConstant=12000
    bAngularSpring=true
    bOverrideAngularDamping=true
    bUseGravityOverride=true
    gravityScale=0.35
    simulationSpace=RootRelative
    boundBone {boneName=dyn_R_hair_main_1}
    componentAppliedLinearAccClamp=<2000 2000 2000>
    componentLinearAccScale=<0.5 0.5 0.5>
    componentLinearVelScale=<0.5 0.5 0.5>
    gravityOverride {z=-980}
    physicsBodyDefinitions [{
      boundBone {boneName=dyn_R_hair_main_1}
      boxExtents=<30 25 25>
      constraintSetup {angularLimitsMax=<0 0 10>, angularLimitsMin=<0 -5 -10>, angularTarget=<1 0 0>}
      localJointOffset=<4 0 0>
     }]
   }
   n77 = Anim Dynamics - Bone: dyn_R_hair_main_2!(
    ComponentPose=@n76
    Alpha=1
    GravityOverride="0,0,-980"
    SimSpaceSettings:Anim Phys Sim Space Settings Structure=(SimSpaceAngularAlpha=0,MaxAngularVelocity=10000,MaxAngularAcceleration=10000,ExternalAngularVelocity=(X=0,Y=0,Z=0))
    ExternalForce="0,0,0"
   ) -> (Pose) {
    angularDampingOverride=0.9999
    angularSpringConstant=12000
    bAngularSpring=true
    bOverrideAngularDamping=true
    bUseGravityOverride=true
    gravityScale=0.35
    simulationSpace=RootRelative
    boundBone {boneName=dyn_R_hair_main_2}
    componentAppliedLinearAccClamp=<2000 2000 2000>
    componentLinearAccScale=<0.5 0.5 0.5>
    componentLinearVelScale=<0.5 0.5 0.5>
    gravityOverride {z=-980}
    physicsBodyDefinitions [{
      boundBone {boneName=dyn_R_hair_main_2}
      boxExtents=<30 25 25>
      constraintSetup {angularLimitsMax=<0 0 10>, angularLimitsMin=<0 -10 -10>, angularTarget=<1 0 0>}
      localJointOffset=<4 0 0>
     }]
   }
   n78 = Anim Dynamics - Bone: dyn_R_hair_main_3!(
    ComponentPose=@n77
    Alpha=1
    GravityOverride="0,0,-980"
    SimSpaceSettings:Anim Phys Sim Space Settings Structure=(SimSpaceAngularAlpha=0,MaxAngularVelocity=10000,MaxAngularAcceleration=10000,ExternalAngularVelocity=(X=0,Y=0,Z=0))
    ExternalForce="0,0,0"
   ) -> (Pose) {
    angularDampingOverride=0.9999
    angularSpringConstant=12000
    bAngularSpring=true
    bOverrideAngularDamping=true
    bUseGravityOverride=true
    gravityScale=0.35
    simulationSpace=RootRelative
    boundBone {boneName=dyn_R_hair_main_3}
    componentAppliedLinearAccClamp=<2000 2000 2000>
    componentLinearAccScale=<0.5 0.5 0.5>
    componentLinearVelScale=<0.5 0.5 0.5>
    gravityOverride {z=-980}
    physicsBodyDefinitions [{
      boundBone {boneName=dyn_R_hair_main_3}
      boxExtents=<30 25 25>
      constraintSetup {angularLimitsMax=<0 0 10>, angularLimitsMin=<0 -15 -10>, angularTarget=<1 0 0>}
      localJointOffset=<4 0 0>
     }]
   }
   n79 = Anim Dynamics - Bone: dyn_R_hair_main_4!(
    ComponentPose=@n78
    Alpha=1
    GravityOverride="0,0,-980"
    SimSpaceSettings:Anim Phys Sim Space Settings Structure=(SimSpaceAngularAlpha=0,MaxAngularVelocity=10000,MaxAngularAcceleration=10000,ExternalAngularVelocity=(X=0,Y=0,Z=0))
    ExternalForce="0,0,0"
   ) -> (Pose) {
    angularDampingOverride=0.9999
    angularSpringConstant=12000
    bAngularSpring=true
    bOverrideAngularDamping=true
    bUseGravityOverride=true
    gravityScale=0.35
    simulationSpace=RootRelative
    boundBone {boneName=dyn_R_hair_main_4}
    componentAppliedLinearAccClamp=<2000 2000 2000>
    componentLinearAccScale=<0.5 0.5 0.5>
    componentLinearVelScale=<0.5 0.5 0.5>
    gravityOverride {z=-980}
    physicsBodyDefinitions [{
      boundBone {boneName=dyn_R_hair_main_4}
      boxExtents=<30 25 25>
      constraintSetup {angularLimitsMax=<0 0 10>, angularLimitsMin=<0 -20 -10>, angularTarget=<1 0 0>}
      localJointOffset=<4 0 0>
     }]
   }
   n80 = Anim Dynamics - Bone: dyn_L_hair_main_2!(
    ComponentPose=@n79
    Alpha=1
    GravityOverride="0,0,-980"
    SimSpaceSettings:Anim Phys Sim Space Settings Structure=(SimSpaceAngularAlpha=0,MaxAngularVelocity=10000,MaxAngularAcceleration=10000,ExternalAngularVelocity=(X=0,Y=0,Z=0))
    ExternalForce="0,0,0"
   ) -> (Pose) {
    angularDampingOverride=0.9999
    angularSpringConstant=12000
    bAngularSpring=true
    bOverrideAngularDamping=true
    bUseGravityOverride=true
    gravityScale=0.35
    simulationSpace=RootRelative
    boundBone {boneName=dyn_L_hair_main_2}
    componentAppliedLinearAccClamp=<2000 2000 2000>
    componentLinearAccScale=<0.5 0.5 0.5>
    componentLinearVelScale=<0.5 0.5 0.5>
    gravityOverride {z=-980}
    physicsBodyDefinitions [{
      boundBone {boneName=dyn_L_hair_main_2}
      boxExtents=<30 25 25>
      constraintSetup {angularLimitsMax=<0 0 10>, angularLimitsMin=<0 -10 -10>, angularTarget=<1 0 0>}
      localJointOffset=<4 0 0>
     }]
   }
   n81 = Anim Dynamics - Bone: dyn_L_hair_main_3!(
    ComponentPose=@n80
    Alpha=1
    GravityOverride="0,0,-980"
    SimSpaceSettings:Anim Phys Sim Space Settings Structure=(SimSpaceAngularAlpha=0,MaxAngularVelocity=10000,MaxAngularAcceleration=10000,ExternalAngularVelocity=(X=0,Y=0,Z=0))
    ExternalForce="0,0,0"
   ) -> (Pose) {
    angularDampingOverride=0.9999
    angularSpringConstant=12000
    bAngularSpring=true
    bOverrideAngularDamping=true
    bUseGravityOverride=true
    gravityScale=0.35
    simulationSpace=RootRelative
    boundBone {boneName=dyn_L_hair_main_3}
    componentAppliedLinearAccClamp=<2000 2000 2000>
    componentLinearAccScale=<0.5 0.5 0.5>
    componentLinearVelScale=<0.5 0.5 0.5>
    gravityOverride {z=-980}
    physicsBodyDefinitions [{
      boundBone {boneName=dyn_L_hair_main_3}
      boxExtents=<30 25 25>
      constraintSetup {angularLimitsMax=<0 0 10>, angularLimitsMin=<0 -15 -10>, angularTarget=<1 0 0>}
      localJointOffset=<4 0 0>
     }]
   }
   n82 = Anim Dynamics - Bone: dyn_L_hair_main_4!(
    ComponentPose=@n81
    Alpha=1
    GravityOverride="0,0,-980"
    SimSpaceSettings:Anim Phys Sim Space Settings Structure=(SimSpaceAngularAlpha=0,MaxAngularVelocity=10000,MaxAngularAcceleration=10000,ExternalAngularVelocity=(X=0,Y=0,Z=0))
    ExternalForce="0,0,0"
   ) -> (Pose) {
    angularDampingOverride=0.9999
    angularSpringConstant=12000
    bAngularSpring=true
    bOverrideAngularDamping=true
    bUseGravityOverride=true
    gravityScale=0.35
    simulationSpace=RootRelative
    boundBone {boneName=dyn_L_hair_main_4}
    componentAppliedLinearAccClamp=<2000 2000 2000>
    componentLinearAccScale=<0.5 0.5 0.5>
    componentLinearVelScale=<0.5 0.5 0.5>
    gravityOverride {z=-980}
    physicsBodyDefinitions [{
      boundBone {boneName=dyn_L_hair_main_4}
      boxExtents=<30 25 25>
      constraintSetup {angularLimitsMax=<0 0 10>, angularLimitsMin=<0 -20 -10>, angularTarget=<1 0 0>}
      localJointOffset=<4 0 0>
     }]
   }
   n83 = RigidBody!(
    ComponentPose=@n82
    Alpha
    OverrideWorldGravity="0,0,0"
    ExternalForce
    SimSpaceSettings:Sim Space Settings Structure=(WorldAlpha=1,VelocityScaleZ=0.2,DampingAlpha=1,MaxLinearVelocity=200,MaxAngularVelocity=2,MaxLinearAcceleration=1000,MaxAngularAcceleration=10,ExternalLinearDragV=(X=0,Y=0,Z=0),ExternalLinearVelocity=(X=0,Y=0,Z=0),ExternalAngularVelocity=(X=0,Y=0,Z=0))
   ) -> (Pose) {
    bOverrideWorldGravity=true
    overridePhysicsAsset=&F_MED_MissMissile_FaceAcc_Physics
    simulationSpace=BaseBoneSpace
    baseBoneRef {boneName=spine_05}
    simSpaceSettings {maxAngularAcceleration=10, maxAngularVelocity=2, maxLinearAcceleration=1000, maxLinearVelocity=200, velocityScaleZ=0.2, worldAlpha=1}
   }
   n84 = Anim Dynamics - Bone: dyn_hairfront_01!(ComponentPose=@n83, Alpha=1, GravityOverride:Vector="0,0,-980", ExternalForce="0,0,0") -> (Pose) {
    angularDampingOverride=1
    angularSpringConstant=5000
    bAngularSpring=true
    bOverrideAngularDamping=true
    bUseGravityOverride=true
    gravityScale=0.35
    simulationSpace=RootRelative
    boundBone {boneName=dyn_hairfront_01}
    componentAppliedLinearAccClamp=<3000 3000 3000>
    componentLinearAccScale=<1 1 1>
    componentLinearVelScale=<1 1 1>
    gravityOverride {z=-980}
    physicsBodyDefinitions [{
      boundBone {boneName=dyn_hairfront_01}
      boxExtents=<20 20 20>
      constraintSetup {angularLimitsMax=<0 -2 15>, angularLimitsMin=<0 -20 -15>, angularTarget=<1 0 0>}
      localJointOffset=<3 0 0>
     }]
   }
   n85 = Control Rig!(
    Source=Component To Local(ComponentPose=@n84)
    Alpha
    Pickaxe=Get PlayingMeleeAnim()
    Jumping:double=Get IsJumping() -> (IsJumping:double)
    skydiving:bool=Get bIsSkyDiving()
    Istargeting:double=Get isTargeting float() -> (isTargeting float:double)
    skydivingFloating:double=Get isSkydiving_Floating() -> (isSkydiving_Floating:double)
    SkydiveMode:double=Get Skydivemode() -> (Skydivemode:double)
    IsPistol:double=Get isUsingPistol() -> (isUsingPistol:double)
    TacticalJump:double=Get TacticalJump() -> (TacticalJump:double)
    FrontEndIdle:double=Get FrontEndidle() -> (FrontEndidle:double)
   ) -> (Pose) {controlRigAssetReference {blueprintRigClass=&/F_MED_MissMissile_FaceAcc_CtrlRig_Pre.F_MED_MissMissile_FaceAcc_CtrlRig_Pre_C}}
   Save cached pose 'PreAnim'(Pose=@n85) {cacheName=PreAnim}
   n86 = "Constraint - dyn_L_hair_main_1 (PPPP)"(
    ComponentPose=Local To Component(LocalPose=Use cached pose 'PreAnim'())
    Alpha=1
    ConstraintWeights_0=0.1
    ConstraintWeights_1=0.9
    ConstraintWeights_2=0
    ConstraintWeights_3=0
   ) {
    boneToModify {boneName=dyn_L_hair_main_1}
    constraintSetup [{targetBone {boneName=head}}, {targetBone {boneName=spine_05}}, {targetBone {boneName=neck_01}}, {targetBone {boneName=clavicle_l}}]
    constraintWeights [1,1,1,1]
   }
   n87 = "Constraint - dyn_L_hair_main_2 (PPPP)"(ComponentPose=@n86, Alpha=1, ConstraintWeights_0=0.9, ConstraintWeights_1=0, ConstraintWeights_2=0, ConstraintWeights_3=0.1) {
    boneToModify {boneName=dyn_L_hair_main_2}
    constraintSetup [{targetBone {boneName=spine_05}}, {targetBone {boneName=neck_02}}, {targetBone {boneName=neck_01}}, {targetBone {boneName=clavicle_l}}]
    constraintWeights [1,1,1,1]
   }
   n88 = "Constraint - dyn_L_hair_main_3 (PPP)"(ComponentPose=@n87, Alpha=1, ConstraintWeights_0=0.5, ConstraintWeights_1=0.5, ConstraintWeights_2=0) {
    boneToModify {boneName=dyn_L_hair_main_3}
    constraintSetup [{targetBone {boneName=clavicle_l}}, {targetBone {boneName=spine_05}}, {targetBone {boneName=neck_01}}]
    constraintWeights [1,1,1]
   }
   n89 = "Constraint - dyn_L_hair_main_4 (PPPPP)"(
    ComponentPose=@n88
    Alpha=1
    ConstraintWeights_0=0.5
    ConstraintWeights_1=0.5
    ConstraintWeights_2=0
    ConstraintWeights_3=0
    ConstraintWeights_4=0
   ) {
    boneToModify {boneName=dyn_L_hair_main_4}
    constraintSetup [
     {targetBone {boneName=clavicle_l}}
     {targetBone {boneName=spine_05}}
     {targetBone {boneName=spine_04}}
     {targetBone {boneName=upperarm_twist_01_l}}
     {targetBone {boneName=deform_upperarm_bk_l}}
    ]
    constraintWeights [1,1,1,1,1]
   }
   n90 = "Constraint - dyn_R_hair_main_1 (PPPP)"(ComponentPose=@n89, Alpha=1, ConstraintWeights_0=0.1, ConstraintWeights_1=0.9, ConstraintWeights_2=0, ConstraintWeights_3=0) {
    boneToModify {boneName=dyn_R_hair_main_1}
    constraintSetup [{targetBone {boneName=head}}, {targetBone {boneName=spine_05}}, {targetBone {boneName=clavicle_r}}, {targetBone {boneName=neck_01}}]
    constraintWeights [1,1,1,1]
   }
   n91 = "Constraint - dyn_R_hair_main_2 (PPPP)"(ComponentPose=@n90, Alpha=1, ConstraintWeights_0=0.9, ConstraintWeights_1=0, ConstraintWeights_2=0, ConstraintWeights_3=0.1) {
    boneToModify {boneName=dyn_R_hair_main_2}
    constraintSetup [{targetBone {boneName=spine_05}}, {targetBone {boneName=neck_02}}, {targetBone {boneName=neck_01}}, {targetBone {boneName=clavicle_r}}]
    constraintWeights [1,1,1,1]
   }
   n92 = "Constraint - dyn_R_hair_main_3 (PPP)"(ComponentPose=@n91, Alpha=1, ConstraintWeights_0=0.5, ConstraintWeights_1=0.5, ConstraintWeights_2=0) {
    boneToModify {boneName=dyn_R_hair_main_3}
    constraintSetup [{targetBone {boneName=clavicle_r}}, {targetBone {boneName=spine_05}}, {targetBone {boneName=deform_upperarm_up_r}}]
    constraintWeights [1,1,1]
   }
   n93 = "Constraint - dyn_R_hair_main_4 (PPPPP)"(
    ComponentPose=@n92
    Alpha=1
    ConstraintWeights_0=0.2
    ConstraintWeights_1=0.8
    ConstraintWeights_2=0
    ConstraintWeights_3=0
    ConstraintWeights_4=0
   ) {
    boneToModify {boneName=dyn_R_hair_main_4}
    constraintSetup [
     {targetBone {boneName=clavicle_r}}
     {targetBone {boneName=spine_05}}
     {targetBone {boneName=spine_04}}
     {targetBone {boneName=upperarm_twist_01_r}}
     {targetBone {boneName=deform_upperarm_bk_r}}
    ]
    constraintWeights [1,1,1,1,1]
   }
   n94 = Anim Dynamics - Bone: dyn_R_hair_main_1!(ComponentPose=@n93, Alpha=0.8, GravityOverride="0,0,-980", ExternalForce="0,0,0") -> (Pose) {
    angularDampingOverride=0.9999
    angularSpringConstant=8000
    bAngularSpring=true
    bOverrideAngularDamping=true
    bUseGravityOverride=true
    gravityScale=0.35
    simulationSpace=RootRelative
    boundBone {boneName=dyn_R_hair_main_1}
    componentAppliedLinearAccClamp=<4000 4000 4000>
    componentLinearAccScale=<1 1 1>
    componentLinearVelScale=<1 1 1>
    gravityOverride {z=-980}
    physicsBodyDefinitions [{
      boundBone {boneName=dyn_R_hair_main_1}
      boxExtents=<20 30 30>
      constraintSetup {angularLimitsMax=<0 0 4>, angularLimitsMin=<0 0 -4>, angularTarget=<1 0 0>}
      localJointOffset=<2 0 0>
     }]
   }
   n95 = Anim Dynamics - Bone: dyn_R_hair_main_2!(ComponentPose=@n94, Alpha=1, GravityOverride="0,0,-980", ExternalForce="0,0,0") -> (Pose) {
    angularDampingOverride=0.9999
    angularSpringConstant=8000
    bAngularSpring=true
    bOverrideAngularDamping=true
    bUseGravityOverride=true
    gravityScale=0.35
    simulationSpace=RootRelative
    boundBone {boneName=dyn_R_hair_main_2}
    componentAppliedLinearAccClamp=<4000 4000 4000>
    componentLinearAccScale=<1 1 1>
    componentLinearVelScale=<1 1 1>
    gravityOverride {z=-980}
    physicsBodyDefinitions [{
      boundBone {boneName=dyn_R_hair_main_2}
      boxExtents=<20 30 30>
      constraintSetup {angularLimitsMax=<0 0 9>, angularLimitsMin=<0 0 -9>, angularTarget=<1 0 0>}
      localJointOffset=<4 0 0>
     }]
   }
   n96 = Anim Dynamics - Bone: dyn_R_hair_main_3!(ComponentPose=@n95, Alpha=1, GravityOverride="0,0,-980", ExternalForce="0,0,0") -> (Pose) {
    angularDampingOverride=0.9999
    angularSpringConstant=8000
    bAngularSpring=true
    bOverrideAngularDamping=true
    bUseGravityOverride=true
    gravityScale=0.35
    simulationSpace=RootRelative
    boundBone {boneName=dyn_R_hair_main_3}
    componentAppliedLinearAccClamp=<4000 4000 4000>
    componentLinearAccScale=<1 1 1>
    componentLinearVelScale=<1 1 1>
    gravityOverride {z=-980}
    physicsBodyDefinitions [{
      boundBone {boneName=dyn_R_hair_main_3}
      boxExtents=<20 30 30>
      constraintSetup {angularLimitsMax=<0 0 9>, angularLimitsMin=<0 0 -9>, angularTarget=<1 0 0>}
      localJointOffset=<4 0 0>
     }]
   }
   n97 = Anim Dynamics - Bone: dyn_R_hair_main_4!(ComponentPose=@n96, Alpha=1, GravityOverride="0,0,-980", ExternalForce="0,0,0") -> (Pose) {
    angularDampingOverride=0.9999
    angularSpringConstant=8000
    bAngularSpring=true
    bOverrideAngularDamping=true
    bUseGravityOverride=true
    gravityScale=0.35
    simulationSpace=RootRelative
    boundBone {boneName=dyn_R_hair_main_4}
    componentAppliedLinearAccClamp=<4000 4000 4000>
    componentLinearAccScale=<1 1 1>
    componentLinearVelScale=<1 1 1>
    gravityOverride {z=-980}
    physicsBodyDefinitions [{
      boundBone {boneName=dyn_R_hair_main_4}
      boxExtents=<20 30 30>
      constraintSetup {angularLimitsMax=<0 0 9>, angularLimitsMin=<0 0 -9>, angularTarget=<1 0 0>}
      localJointOffset=<4 0 0>
     }]
   }
   n98 = Anim Dynamics - Bone: dyn_L_hair_main_1!(ComponentPose=@n97, Alpha=0.8, GravityOverride="0,0,-980", ExternalForce="0,0,0") -> (Pose) {
    angularDampingOverride=0.9999
    angularSpringConstant=8000
    bAngularSpring=true
    bOverrideAngularDamping=true
    bUseGravityOverride=true
    gravityScale=0.35
    simulationSpace=RootRelative
    boundBone {boneName=dyn_L_hair_main_1}
    componentAppliedLinearAccClamp=<4000 4000 4000>
    componentLinearAccScale=<1 1 1>
    componentLinearVelScale=<1 1 1>
    gravityOverride {z=-980}
    physicsBodyDefinitions [{
      boundBone {boneName=dyn_L_hair_main_1}
      boxExtents=<20 30 30>
      constraintSetup {angularLimitsMax=<0 0 4>, angularLimitsMin=<0 0 -4>, angularTarget=<1 0 0>}
      localJointOffset=<2 0 0>
     }]
   }
   n99 = Anim Dynamics - Bone: dyn_L_hair_main_2!(ComponentPose=@n98, Alpha=1, GravityOverride="0,0,-980", ExternalForce="0,0,0") -> (Pose) {
    angularDampingOverride=0.9999
    angularSpringConstant=8000
    bAngularSpring=true
    bOverrideAngularDamping=true
    bUseGravityOverride=true
    gravityScale=0.35
    simulationSpace=RootRelative
    boundBone {boneName=dyn_L_hair_main_2}
    componentAppliedLinearAccClamp=<4000 4000 4000>
    componentLinearAccScale=<1 1 1>
    componentLinearVelScale=<1 1 1>
    gravityOverride {z=-980}
    physicsBodyDefinitions [{
      boundBone {boneName=dyn_L_hair_main_2}
      boxExtents=<20 30 30>
      constraintSetup {angularLimitsMax=<0 0 9>, angularLimitsMin=<0 0 -9>, angularTarget=<1 0 0>}
      localJointOffset=<4 0 0>
     }]
   }
   n100 = Anim Dynamics - Bone: dyn_L_hair_main_3!(ComponentPose=@n99, Alpha=1, GravityOverride="0,0,-980", ExternalForce="0,0,0") -> (Pose) {
    angularDampingOverride=0.9999
    angularSpringConstant=8000
    bAngularSpring=true
    bOverrideAngularDamping=true
    bUseGravityOverride=true
    gravityScale=0.35
    simulationSpace=RootRelative
    boundBone {boneName=dyn_L_hair_main_3}
    componentAppliedLinearAccClamp=<4000 4000 4000>
    componentLinearAccScale=<1 1 1>
    componentLinearVelScale=<1 1 1>
    gravityOverride {z=-980}
    physicsBodyDefinitions [{
      boundBone {boneName=dyn_L_hair_main_3}
      boxExtents=<20 30 30>
      constraintSetup {angularLimitsMax=<0 0 9>, angularLimitsMin=<0 0 -9>, angularTarget=<1 0 0>}
      localJointOffset=<4 0 0>
     }]
   }
   n101 = Anim Dynamics - Bone: dyn_L_hair_main_4!(ComponentPose=@n100, Alpha=1, GravityOverride="0,0,-980", ExternalForce="0,0,0") -> (Pose) {
    angularDampingOverride=0.9999
    angularSpringConstant=8000
    bAngularSpring=true
    bOverrideAngularDamping=true
    bUseGravityOverride=true
    gravityScale=0.35
    simulationSpace=RootRelative
    boundBone {boneName=dyn_L_hair_main_4}
    componentAppliedLinearAccClamp=<4000 4000 4000>
    componentLinearAccScale=<1 1 1>
    componentLinearVelScale=<1 1 1>
    gravityOverride {z=-980}
    physicsBodyDefinitions [{
      boundBone {boneName=dyn_L_hair_main_4}
      boxExtents=<20 30 30>
      constraintSetup {angularLimitsMax=<0 0 9>, angularLimitsMin=<0 0 -9>, angularTarget=<1 0 0>}
      localJointOffset=<4 0 0>
     }]
   }
   n102 = Control Rig!(Source=Component To Local(ComponentPose=@n101), Alpha, CrouchHeadUP walk:bool=Get CrouchWalk(), Pickaxe=Get PlayingMeleeAnim()) -> (Pose) {controlRigAssetReference {blueprintRigClass=&F_MED_MissMissile_FaceAcc_CtrlRig_Post_C}}
   Save cached pose 'Crouch Output'(Pose=Blend Poses by bool(BlendPose_0=@n102, BlendPose_1=Use cached pose 'PreAnim'(), bActiveValue=Get bIsCrouching())) {cacheName=Crouch Output}
   n103 = RigidBody!(ComponentPose=Local To Component(LocalPose=Use cached pose 'Crouch Output'()), Alpha, OverrideWorldGravity="0,0,0", ExternalForce) -> (Pose) {
    bOverrideWorldGravity=true
    overridePhysicsAsset=&/F_MED_MissMissile_FaceAcc_BackHair_Emotes_Physics
    simulationSpace=BaseBoneSpace
    baseBoneRef {boneName=neck_02}
    simSpaceSettings {
     maxAngularAcceleration=500
     maxAngularVelocity=500
     maxLinearAcceleration=500
     maxLinearVelocity=500
     worldAlpha=1
     externalAngularVelocity=<1 1 1>
     externalLinearDragV=<1 1 1>
     externalLinearVelocity=<1 1 1>
    }
   }
   n104 = "Transform (Modify) Bone - Bone: dyn_R_hair_main_1"(ComponentPose=@n103, bAlphaBoolEnabled=True, Translation="1.324,-0.077,6", Rotation="-20,-0.515,2.263", Scale="1,1,1") {
    alphaInputType=Bool
    rotationMode=BMM_Additive
    rotationSpace=BCS_BoneSpace
    translationMode=BMM_Additive
    translationSpace=BCS_BoneSpace
    alphaScaleBiasClamp {bInterpResult=true}
    boneToModify {boneName=dyn_R_hair_main_1}
    rotation=<rot -20 -0.515 2.263>
    translation=<1.324 -0.077 6>
   }
   n105 = "Transform (Modify) Bone - Bone: dyn_R_hair_main_2"(ComponentPose=@n104, bAlphaBoolEnabled=True, Translation="0,0,0", Rotation="0,-0.017,-0.08", Scale="1,1,1") {
    alphaInputType=Bool
    rotationMode=BMM_Additive
    rotationSpace=BCS_BoneSpace
    translationMode=BMM_Additive
    translationSpace=BCS_BoneSpace
    alphaBoolBlend {blendInTime=10, blendOutTime=10}
    alphaScaleBiasClamp {bInterpResult=true}
    boneToModify {boneName=dyn_R_hair_main_2}
    rotation {roll=-0.08, yaw=-0.017}
   }
   n106 = "Transform (Modify) Bone - Bone: dyn_R_hair_main_3"(ComponentPose=@n105, bAlphaBoolEnabled=True, Translation="0,0,0", Rotation="9.125,-0.026,-0.073", Scale="1,1,1") {
    alphaInputType=Bool
    rotationMode=BMM_Additive
    rotationSpace=BCS_BoneSpace
    translationMode=BMM_Additive
    translationSpace=BCS_BoneSpace
    alphaBoolBlend {blendInTime=10, blendOutTime=10}
    alphaScaleBiasClamp {bInterpResult=true}
    boneToModify {boneName=dyn_R_hair_main_3}
    rotation=<rot 9.125 -0.026 -0.073>
   }
   n107 = "Transform (Modify) Bone - Bone: dyn_L_hair_main_1"(ComponentPose=@n106, bAlphaBoolEnabled=True, Translation="0,0,5", Rotation="-20,0,0", Scale="1,1,1") {
    alphaInputType=Bool
    rotationMode=BMM_Additive
    rotationSpace=BCS_BoneSpace
    translationMode=BMM_Additive
    translationSpace=BCS_BoneSpace
    alphaBoolBlend {blendInTime=10, blendOutTime=10}
    alphaScaleBiasClamp {bInterpResult=true}
    boneToModify {boneName=dyn_L_hair_main_1}
    rotation {pitch=-20}
    translation {z=5}
   }
   n108 = "Transform (Modify) Bone - Bone: dyn_L_hair_main_2"(ComponentPose=@n107, bAlphaBoolEnabled=True, Translation="0,0,0", Rotation="0,0,0", Scale="1,1,1") {
    alphaInputType=Bool
    rotationMode=BMM_Additive
    translationMode=BMM_Additive
    alphaBoolBlend {blendInTime=10, blendOutTime=10}
    alphaScaleBiasClamp {bInterpResult=true}
    boneToModify {boneName=dyn_L_hair_main_2}
   }
   n109 = "Transform (Modify) Bone - Bone: dyn_L_hair_main_3"(ComponentPose=@n108, bAlphaBoolEnabled=True, Translation="0,0,0", Rotation="-0.000,0.000,11.650", Scale="1,1,1") {
    alphaInputType=Bool
    rotationMode=BMM_Additive
    translationMode=BMM_Additive
    alphaBoolBlend {blendInTime=10, blendOutTime=10}
    alphaScaleBiasClamp {bInterpResult=true}
    boneToModify {boneName=dyn_L_hair_main_3}
    rotation {roll=11.65}
   }
   Save cached pose 'body'(Pose=Blend Poses by bool(BlendPose_0=Component To Local(ComponentPose=@n109), BlendPose_1=Use cached pose 'Crouch Output'(), bActiveValue=Get bIsPlayingEmote())) {cacheName=body}
   Get IsJumping()
   Get IsJumping()
   Get IsJumping()
  ]}
 EventGraph {nodes [
   Event OnBeginSkydiving()
   > Set isSkydiving(isSkydiving:double=1) -> (Output_Get:double)
   Event OnEndSkydiving()
   > Set isSkydiving(isSkydiving=0)
   Event OnEndFalling()
   > Set IsJumping(IsJumping:double=0) -> (Output_Get:double)
   n1 = Event OnCharacterCustomizationLoaded()
   n2 = CheckForBackbling()
   Set isWearingBackblingBool(
    execute=[@n1.then, @n6.then_6]
    isWearingBackblingBool:bool=OR Boolean!(A=@n2, B=@n2.bIsBackblingACape, C:bool=@n2.bIsBackblingLong) -> (ReturnValue)
   ) -> (Output_Get:bool)
   Event OnBeginJumping()
   n3 = Get bIsSkyDiving()
   > Set IsJumping(IsJumping="To Float (Boolean)"(InBool=NOT Boolean(A=@n3)))
   Event OnEnterGrindRail() {is=title K2Node_Event}
   > Set GrindRail(GrindRail:double=1) -> (Output_Get:double)
   Event OnWeaponChanged()
   > Set isUsingPistol(isUsingPistol:double=0) -> (Output_Get:double)
   > n4 = Switch on EFortWeaponCoreAnimation!(execute, Selection:EFortWeaponCoreAnimation Enum=Get WeaponCoreAnim!() -> (WeaponCoreAnim)) -> (
    Melee:Exec
    Pistol:Exec
    Shotgun:Exec
    PaperBlueprint:Exec
    Rifle:Exec
    MeleeOneHand:Exec
    MachinePistol:Exec
    RocketLauncher:Exec
    GrenadeLauncher:Exec
    GoingCommando:Exec
    AssaultRifle:Exec
    TacticalShotgun:Exec
    SniperRifle:Exec
    TrapPlacement:Exec
    ShoulderLauncher:Exec
    AbilityDecoTool:Exec
    Crossbow:Exec
    C4:Exec
    RemoteControl:Exec
    DualWield:Exec
    AR_BullPup:Exec
    AR_ForwardGrip:Exec
    MedPackPaddles:Exec
    SMG_P90:Exec
    AR_DrumGun:Exec
    Consumable_Small:Exec
    Consumable_Large:Exec
    Balloon:Exec
    MountedTurret:Exec
    CreativeTool:Exec
    ExplosiveBow:Exec
    AshtonIndigo:Exec
    AshtonChicago:Exec
    MeleeDualWield:Exec
    Unarmed:Exec
   )
   Set isUsingPistol(execute=@n4.Pistol, isUsingPistol=1)
   n5 = Event BlueprintUpdateAnimation()
   > Set Delta Time X(Delta Time X=@n5.DeltaTimeX)
   > n6 = Sequence!(execute) -> (
    then_0
    then_1
    then_2:Exec
    then_3:Exec
    then_4:Exec
    then_5:Exec
    then_6:Exec
    then_7:Exec
    then_8:Exec
    then_9:Exec
    then_10:Exec
    then_11:Exec
   )
   > n7 = "Gravity Override State Function (Parameter Asset) (Nativized)"(
    CharacterPartAnimInstance=Self-Reference()
    GameTimeInSeconds=GetGameTimeInSeconds()
    Parameters=&/F_MED_MissMissile_FaceAcc_Parameter
   )
   > n8 = Set GravityOverride(GravityOverride=@n7)
   > Set RBAN Settings Out(RBAN Settings Out=@n7.OutRigidBodyAnimNodeParameters)
   > Set GravityOverride(GravityOverride=SelectVector(A="0,0,-5000", B=@n8, bPickA=Get isWearingBackblingBool()))
   > Set Sim Space Settings(Sim Space Settings:Anim Phys Sim Space Settings Structure=Get Sim Space Settings_BR() -> (Sim Space Settings_BR:Anim Phys Sim Space Settings Structure)) -> (Output_Get:Anim Phys Sim Space Settings Structure)
   > Branch(Condition=Get bIsInFrontEnd())
   > Set Sim Space Settings(Sim Space Settings=Get Sim Space Settings_FE() -> (Sim Space Settings_FE:Anim Phys Sim Space Settings Structure))
   n9 = Set Head Up(
    execute=@n6.then_1
    Head Up:double=GetAlphaFromReferenceTransformDelta(SkeletalMeshComponent=Get Skel Mesh(), BoneName=spine_05, BaseName=head, InputRange=(X=0,Y=90), TwistAxis="0,0,1")
   ) -> (Output_Get:double)
   n10 = Get CrouchTargeting() -> (CrouchTargeting:bool)
   > Set CrouchingHeadUP(CrouchingHeadUP:double=SelectFloat(A=@n9, B=0, bPickA=@n10)) -> (Output_Get:double)
   n11 = NOT Boolean(A=Get bIsCrouching())
   n12 = Get bIsTargeting()
   > Set AimUp(AimUp:double=SelectFloat(A=@n9, B=0, bPickA=AND Boolean(A=@n11, B=@n12))) -> (Output_Get:double)
   > Set CrouchWalk(CrouchWalk:bool=AND Boolean(A=Get bIsMoving2D(), B=Get bIsCrouching())) -> (Output_Get:bool)
   n13 = NOT Boolean(A=Get bIsTargeting())
   n14 = Get bIsCrouching()
   n15 = Get CrouchWalk()
   > Set CrouchingHeadUp_Alpha(CrouchingHeadUp_Alpha:double=SelectFloat(A=Get Head Up() -> (Head Up:double), B=0, bPickA=AND Boolean!(A=@n15, B=@n14, C:bool=@n13) -> (ReturnValue))) -> (Output_Get:double)
   n16 = Branch(execute=@n6.then_10, Condition=Get bIsSkydiveDiveMode())
   > Set Skydivemode(Skydivemode:double=1) -> (Output_Get:double)
   Set Skydivemode(execute=@n16.else, Skydivemode=0)
   n17 = Get bIsTargeting()
   Set isTargeting float(
    execute=@n6.then_11
    isTargeting float:double=FInterpTo(Current=Get isTargeting float(), Target="To Float (Boolean)"(InBool=@n17), DeltaTime=Get Delta Time X(), InterpSpeed=15)
   ) -> (Output_Get:double)
   > Set TacticalJump(TacticalJump:double=SelectFloat(A=Get IsJumping(), B=0, bPickA=Get bIsTacticalSprinting())) -> (Output_Get:double)
   n18 = NOT Boolean(A=Get bIsPlayingEmote())
   n19 = Get bIsInFrontEnd()
   > Set FrontEndidle(FrontEndidle:double="To Float (Boolean)"(InBool=AND Boolean!(A=@n19, B=@n18) -> (ReturnValue))) -> (Output_Get:double)
   > Set Dyn_Alpha(Dyn_Alpha:double=0.45) -> (Output_Get:double)
   n20 = Get bIsPlayingEmote()
   Set TargetingGun(execute=@n6.then_2, TargetingGun:bool=AND Boolean(A=Get bIsTargeting(), B=NOT Boolean(A=@n20))) -> (Output_Get:bool)
   > Set CrouchTargeting(CrouchTargeting:bool=AND Boolean(A=Get bIsCrouching(), B=Get bIsTargeting())) -> (Output_Get:bool)
   n21 = Set Head Down(
    execute=@n6.then_3
    Head Down:double=GetAlphaFromReferenceTransformDelta(SkeletalMeshComponent=Get Skel Mesh(), BoneName=spine_05, BaseName=head, InputRange=(X=-90,Y=0), TwistAxis="0,0,-1")
   ) -> (Output_Get:double)
   > Set CrouchingHeaddown(CrouchingHeaddown:double=SelectFloat(A=@n21, B=0, bPickA=Get CrouchTargeting())) -> (Output_Get:double)
   n22 = GetAnimInstance(self=GetOwnerSkeletalMeshComponent())
   Is Valid(exec=@n6.then_4, InputObject=@n22)
   > n23 = Set Celebration Custom(Celebration Custom:bool=Montage_IsPlaying(self=@n22, Montage=/BRCosmetics/Animation/Game/MainPlayer/Emotes/Emote_Celebration_Female_M)) -> (Output_Get:bool)
   > Set Celebration Emote(Celebration Emote:double="To Float (Boolean)"(InBool=@n23)) -> (Output_Get:double)
   Reroute Node(InputPin=@n6.then_7)
   n24 = Get Rifle Equipped() -> (Rifle Equipped:bool)
   n25 = Get bIsCrouching()
   Set AimUp_Rifle(execute=@n6.then_8, AimUp_Rifle:double=SelectFloat(A=Get Head Up(), B=0, bPickA=AND Boolean!(A=@n25, B=@n24, C:bool) -> (ReturnValue))) -> (Output_Get:double)
   Set PlayingMeleeAnim(execute=@n6.then_9, PlayingMeleeAnim:double="To Float (Boolean)"(InBool=Get bIsPlayingMeleeAnim())) -> (Output_Get:double)
   > n26 = Branch(Condition=Get bIsSkydiveFloating())
   > Set isSkydiving_Floating(isSkydiving_Floating:double=1) -> (Output_Get:double)
   Set isSkydiving_Floating(execute=@n26.else, isSkydiving_Floating=0)
   Branch(execute=@n6.then_5, Condition=Get bIsInFrontEnd())
   n27 = GetCurveValue(self=GetOwnerFrontEndAnimInstance(), CurveName=Frontend_Idle_Bypass)
   > Set isPlayingFrontendIdle(isPlayingFrontendIdle:float=@n27) -> (Output_Get:float)
   > Set turnOffDuringCustomIdle(turnOffDuringCustomIdle:double=float - float(A=1, B=@n27)) -> (Output_Get:double)
   Event OnExitGrindRail() {is=title K2Node_Event}
   > Set GrindRail(GrindRail=0)
   Event BlueprintInitializeAnimation()
   > Set Skel Mesh(Skel Mesh=GetOwnerSkeletalMeshComponent())
   OR Boolean!(A=Get bIsCrouching(), B=@n10) -> (ReturnValue)
   OR Boolean(A=Get bIsTargeting(), B=false)
   CustomEvent_1!() -> (
    OutputDelegate
    then
    CustomBaseMesh:Skeletal Mesh Component Object Reference
    CustomParts:Array of Custom Character Part Object References
    CustomMeshParts:Array of Skeletal Mesh Component Budgeted Object References
   ) {is=title K2Node_CustomEvent}
   Get bIsCrouching()
   Get Weapon Equipped() -> (Weapon Equipped:double)
   Set AimUp_Rifle(AimUp_Rifle=0)
  ]}
]
variables [
 AimUp {type=number}
 AimUp_Rifle {type=number}
 Body_Mesh {type=SkeletalMeshComponent}
 Celebration Custom {type=boolean}
 Celebration Emote {type=number}
 CrouchingHeaddown {type=number}
 CrouchingHeadUP {type=number}
 CrouchingHeadUp_Alpha {type=number}
 CrouchTargeting {type=boolean}
 CrouchWalk {type=boolean}
 Delta Time X {type=number}
 Dyn_Alpha {type=number}
 FrontEndidle {type=number}
 GravityOverride {type=Vector}
 GrindRail {type=number}
 Head Down {type=number}
 Head Up {type=number}
 Head_Turn_L {type=number}
 Head_Turn_R {type=number}
 Is Backbling Cape {type=boolean}
 Is Targeting {type=boolean}
 Is Wearing Backbling {type=boolean}
 IsJumping {type=number}
 isPlayingFrontendIdle {type=number}
 isSkydiving {type=number}
 isSkydiving_Floating {type=number}
 isTargeting float {type=number}
 isUsingPistol {type=number}
 isWearingBackblingBool {type=boolean}
 Pistol CrouchAimup {type=number}
 PlayingMeleeAnim {type=number}
 RBAN Settings Out {type=FortRigidBodyAnimNodeParameters}
 Rifle Equipped {type=boolean}
 Sim Space Settings {
  type=AnimPhysSimSpaceSettings
  default {maxAngularAcceleration=10000, maxAngularVelocity=10000, simSpaceAngularAlpha=0, externalAngularVelocity=<0 0 0>}
 }
 Sim Space Settings_BR {
  type=AnimPhysSimSpaceSettings
  default {maxAngularAcceleration=10000, maxAngularVelocity=10000, simSpaceAngularAlpha=0, externalAngularVelocity=<0 0 0>}
 }
 Sim Space Settings_FE {
  type=AnimPhysSimSpaceSettings
  default {maxAngularAcceleration=10000, maxAngularVelocity=10000, simSpaceAngularAlpha=0.3, externalAngularVelocity=<0 0 0>}
 }
 Skel Mesh {type=SkeletalMeshComponent}
 Skydivemode {type=number}
 TacticalJump {type=number}
 TargetingGun {type=boolean}
 turnOffDuringCustomIdle {type=number}
 Weapon Equipped {type=number}
]
