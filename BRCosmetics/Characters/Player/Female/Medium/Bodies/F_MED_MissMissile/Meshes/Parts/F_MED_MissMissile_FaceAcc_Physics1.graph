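# Physics asset for the MissMissile female player character's face accessory mesh, set up for the motion of its long hair.
# Capsule bodies on upper spine, clavicles, neck and head form an angularly locked frame, along with the cape attach and side-hair base bones.
# Left, right and back-middle hair chains hang from it with limited swing and zero twist limit.
# Angularly free cross-constraints tie lower side-chain segments to matching back-chain segments.
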PhysicsAsset /BRCosmetics/Characters/Player/Female/Medium/Bodies/F_MED_MissMissile/Meshes/Parts/F_MED_MissMissile_FaceAcc_Physics1
bodies {
 attach_cape [{shapeType=Capsule, length=0.505, radius=0.505, shapeName=attach_cape_capsule, rotation=<rot 0 -90 0>}]
 back_mid_hair_1 [{shapeType=Capsule, length=0.505, radius=0.505, shapeName=back_mid_hair_1_capsule, rotation=<rot 0 -90 0>}]
 back_mid_hair_2 [{shapeType=Capsule, length=0.505, radius=0.505, shapeName=back_mid_hair_2_capsule, rotation=<rot 0 -90 0>}]
 back_mid_hair_3 [{shapeType=Capsule, length=0.505, radius=0.505, shapeName=back_mid_hair_3_capsule, rotation=<rot 0 -90 0>}]
 back_mid_hair_4 [{shapeType=Capsule, length=0.505, radius=0.505, shapeName=back_mid_hair_4_capsule, rotation=<rot 0 -90 0>}]
 back_mid_hair_5 [{shapeType=Capsule, length=0.505, radius=0.505, shapeName=back_mid_hair_5_capsule, rotation=<rot 0 -90 0>}]
 base_L_hair_main [{shapeType=Capsule, length=0.505, radius=0.505, shapeName=base_L_hair_main_capsule, rotation=<rot 0 -90 0>}]
 base_R_hair_main [{shapeType=Capsule, length=0.505, radius=0.505, shapeName=base_R_hair_main_capsule, rotation=<rot 0 -90 0>}]
 clavicle_l [{
   shapeType=Capsule
   length=6.5566
   radius=8
   shapeName=clavicle_r_capsule
   center=<6.9423 1.1136 -3.5468>
   rotation=<rot 89.6647 92.4944 87.5056>
  }]
 clavicle_r [{shapeType=Capsule, length=5.9866, radius=8, shapeName=clavicle_r_capsule, center=<-7.1743 -0.8606 3.546>, rotation=<rot 90 0 0>}]
 dyn_L_hair_main_1 [{shapeType=Capsule, length=3, radius=7, shapeName=dyn_L_hair_main_1_capsule, center=<4 -3 0>, rotation=<rot 90 0 0>}]
 dyn_L_hair_main_2 [{shapeType=Capsule, length=3, radius=5, shapeName=dyn_L_hair_main_2_capsule, center=<4 -3 3>, rotation=<rot 90 0 0>}]
 dyn_L_hair_main_3 [{shapeType=Capsule, length=3, radius=5, shapeName=dyn_L_hair_main_3_capsule, center=<4 -3 2>, rotation=<rot 90 0 0>}]
 dyn_L_hair_main_4 [{shapeType=Capsule, length=3, radius=5, shapeName=dyn_L_hair_main_4_capsule, center=<4 -3 2>, rotation=<rot 90 0 0>}]
 dyn_R_hair_main_1 [{shapeType=Capsule, length=3, radius=7, shapeName=dyn_R_hair_main_1_capsule, center=<4 3 0>, rotation=<rot 90 0 0>}]
 dyn_R_hair_main_2 [{shapeType=Capsule, length=3, radius=5, shapeName=dyn_R_hair_main_2_capsule, center=<4 3 3>, rotation=<rot 90 0 0>}]
 dyn_R_hair_main_3 [{shapeType=Capsule, length=3, radius=5, shapeName=dyn_R_hair_main_3_capsule, center=<4 3 2>, rotation=<rot 90 0 0>}]
 dyn_R_hair_main_4 [{shapeType=Capsule, length=3, radius=5, shapeName=dyn_R_hair_main_4_capsule, center=<4 3 2>, rotation=<rot 90 0 0>}]
 head [{shapeType=Capsule, length=6, radius=5, shapeName=head_capsule, center=<-0.0213 -2.1722 0>, rotation=<rot 90 13.2582 0>}]
 neck_01 [{shapeType=Capsule, length=12.8207, radius=5.1793, shapeName=neck_01_capsule}]
 neck_02 [{shapeType=Capsule, length=12.8207, radius=4.1793, shapeName=neck_02_capsule}]
 spine_03 [{shapeType=Capsule, length=10, radius=8, shapeName=spine_03_capsule, center=<0 1.032 0>}]
 spine_04 [{shapeType=Capsule, length=11.6275, radius=12.0219, shapeName=spine_04_capsule, center=<0 1.9595 0>}]
 spine_05 [{shapeType=Capsule, length=21.4663, radius=9.5152, shapeName=spine_05_capsule, center=<-0.0373 0.517 0>}]
}
constraints [
 {
  bone1Name=spine_04
  bone2Name=spine_03
  swing1LimitDegrees=45
  swing1Motion=Locked
  swing2LimitDegrees=45
  swing2Motion=Locked
  twistLimitDegrees=45
  twistMotion=Locked
 }
 {
  bone1Name=spine_05
  bone2Name=spine_04
  swing1LimitDegrees=45
  swing1Motion=Locked
  swing2LimitDegrees=45
  swing2Motion=Locked
  twistLimitDegrees=45
  twistMotion=Locked
 }
 {
  bone1Name=clavicle_l
  bone2Name=spine_05
  swing1LimitDegrees=45
  swing1Motion=Locked
  swing2LimitDegrees=45
  swing2Motion=Locked
  twistLimitDegrees=45
  twistMotion=Locked
 }
 {
  bone1Name=clavicle_r
  bone2Name=spine_05
  swing1LimitDegrees=45
  swing1Motion=Locked
  swing2LimitDegrees=45
  swing2Motion=Locked
  twistLimitDegrees=45
  twistMotion=Locked
 }
 {
  bone1Name=head
  bone2Name=neck_02
  swing1LimitDegrees=45
  swing1Motion=Locked
  swing2LimitDegrees=45
  swing2Motion=Locked
  twistLimitDegrees=45
  twistMotion=Locked
 }
 {
  bone1Name=base_L_hair_main
  bone2Name=neck_02
  swing1LimitDegrees=45
  swing1Motion=Locked
  swing2LimitDegrees=45
  swing2Motion=Locked
  twistLimitDegrees=45
  twistMotion=Locked
 }
 {
  bone1Name=base_R_hair_main
  bone2Name=neck_02
  swing1LimitDegrees=45
  swing1Motion=Locked
  swing2LimitDegrees=45
  swing2Motion=Locked
  twistLimitDegrees=45
  twistMotion=Locked
 }
 {
  bone1Name=neck_01
  bone2Name=spine_05
  swing1LimitDegrees=45
  swing1Motion=Locked
  swing2LimitDegrees=45
  swing2Motion=Locked
  twistLimitDegrees=45
  twistMotion=Locked
 }
 {
  bone1Name=neck_02
  bone2Name=neck_01
  swing1LimitDegrees=45
  swing1Motion=Locked
  swing2LimitDegrees=45
  swing2Motion=Locked
  twistLimitDegrees=45
  twistMotion=Locked
 }
 {bone1Name=dyn_L_hair_main_2, bone2Name=dyn_L_hair_main_1, swing1LimitDegrees=20, swing2LimitDegrees=30, twistLimitDegrees=0}
 {bone1Name=dyn_L_hair_main_3, bone2Name=dyn_L_hair_main_2, swing1LimitDegrees=20, swing2LimitDegrees=30, twistLimitDegrees=0}
 {bone1Name=dyn_L_hair_main_4, bone2Name=dyn_L_hair_main_3, swing1LimitDegrees=20, swing2LimitDegrees=30, twistLimitDegrees=0}
 {bone1Name=dyn_R_hair_main_1, bone2Name=base_R_hair_main, swing1LimitDegrees=20, swing2LimitDegrees=30, twistLimitDegrees=0}
 {bone1Name=dyn_R_hair_main_2, bone2Name=dyn_R_hair_main_1, swing1LimitDegrees=20, swing2LimitDegrees=30, twistLimitDegrees=0}
 {bone1Name=dyn_R_hair_main_3, bone2Name=dyn_R_hair_main_2, swing1LimitDegrees=20, swing2LimitDegrees=30, twistLimitDegrees=0}
 {bone1Name=dyn_R_hair_main_4, bone2Name=dyn_R_hair_main_3, swing1LimitDegrees=20, swing2LimitDegrees=30, twistLimitDegrees=0}
 {bone1Name=dyn_L_hair_main_1, bone2Name=base_L_hair_main, swing1LimitDegrees=20, swing2LimitDegrees=30, twistLimitDegrees=0}
 {
  bone1Name=attach_cape
  bone2Name=spine_05
  swing1LimitDegrees=45
  swing1Motion=Locked
  swing2LimitDegrees=45
  swing2Motion=Locked
  twistLimitDegrees=45
  twistMotion=Locked
 }
 {bone1Name=back_mid_hair_1, bone2Name=head, swing1LimitDegrees=40, swing2LimitDegrees=40, twistLimitDegrees=0}
 {bone1Name=back_mid_hair_2, bone2Name=back_mid_hair_1, swing1LimitDegrees=40, swing2LimitDegrees=40, twistLimitDegrees=0}
 {bone1Name=back_mid_hair_3, bone2Name=back_mid_hair_2, swing1LimitDegrees=40, swing2LimitDegrees=40, twistLimitDegrees=0}
 {bone1Name=back_mid_hair_4, bone2Name=back_mid_hair_3, swing1LimitDegrees=40, swing2LimitDegrees=40, twistLimitDegrees=0}
 {bone1Name=back_mid_hair_5, bone2Name=back_mid_hair_4, swing1LimitDegrees=40, swing2LimitDegrees=40, twistLimitDegrees=0}
 {
  bone1Name=dyn_L_hair_main_4
  bone2Name=back_mid_hair_4
  swing1LimitDegrees=45
  swing1Motion=Free
  swing2LimitDegrees=45
  swing2Motion=Free
  twistLimitDegrees=45
  twistMotion=Free
 }
 {
  bone1Name=dyn_R_hair_main_4
  bone2Name=back_mid_hair_4
  swing1LimitDegrees=45
  swing1Motion=Free
  swing2LimitDegrees=45
  swing2Motion=Free
  twistLimitDegrees=45
  twistMotion=Free
 }
 {
  bone1Name=dyn_L_hair_main_3
  bone2Name=back_mid_hair_3
  swing1LimitDegrees=45
  swing1Motion=Free
  swing2LimitDegrees=45
  swing2Motion=Free
  twistLimitDegrees=45
  twistMotion=Free
 }
 {
  bone1Name=dyn_R_hair_main_3
  bone2Name=back_mid_hair_3
  swing1LimitDegrees=45
  swing1Motion=Free
  swing2LimitDegrees=45
  swing2Motion=Free
  twistLimitDegrees=45
  twistMotion=Free
 }
 {
  bone1Name=dyn_L_hair_main_2
  bone2Name=back_mid_hair_2
  swing1LimitDegrees=45
  swing1Motion=Free
  swing2LimitDegrees=45
  swing2Motion=Free
  twistLimitDegrees=45
  twistMotion=Free
 }
 {
  bone1Name=dyn_R_hair_main_2
  bone2Name=back_mid_hair_2
  swing1LimitDegrees=45
  swing1Motion=Free
  swing2LimitDegrees=45
  swing2Motion=Free
  twistLimitDegrees=45
  twistMotion=Free
 }
]
